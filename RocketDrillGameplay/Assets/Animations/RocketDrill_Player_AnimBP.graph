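# Player animation blueprint for the Rocket Drill item, overriding Fortnite item animation layers, mainly full body and upper body.
# A full-body state machine runs burst re-entry, loop, dismount, fall, landings, skid and wall hit, using state aliases for shared transitions.
# Boost states rotate the root, and landing and skid states mask in lower-body locomotion while moving.
# Idle noise, aim offset and Dead Blending follow; upper body swaps to gendered sliding poses.
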AnimBlueprint /RocketDrillGameplay/Assets/Animations/RocketDrill_Player_AnimBP
target_skeleton=&Fortnite_M_Avg_Player_Skeleton
graphs [
 AnimGraph {nodes [Output Pose()]}
 BurstDismount {nodes [
   BurstDismount!() -> (then, Context:"Anim Update Context Structure (by ref)", Node:"Anim Node Reference Structure (by ref)") {is=title K2Node_FunctionEntry}
   > Return Node()
  ]}
 EventGraph {links [], nodes []}
 ItemAimOffset {nodes [
   Input Pose!() -> (Pose, AimOffsetAlpha:double, Yaw:double, Pitch:double) {name=InPose_AimOffset}
   Input Pose!() -> (Pose) {name=InPose_PreAimOffsets}
   Output Pose() {layerGroup=Item Override}
  ]}
 ItemFallAdditive {nodes [
   Input Pose() {name=InPoseItemFallAdditive}
   Output Pose() {layerGroup=Item Override}
  ]}
 ItemFallLandAdditive {nodes [
   Input Pose() {name=InPoseItemFallLandAdditive}
   Output Pose() {layerGroup=Item Override}
  ]}
 ItemFallLevel2Override {nodes [
   Input Pose() {name=InPose_FallLevel2}
   Output Pose() {layerGroup=Item Override}
  ]}
 ItemFinalPoseOverride {nodes [Output Pose(Result=Input Pose() {name=InFinalPose}) {layerGroup=Item Override}]}
 ItemFullBodyAimOffsets {nodes [
   Input Pose!() -> (Pose, FullBody_AimOffsetAlpha:double, Aim_Yaw:double, Aim_Pitch:double) {name=InPoseFullBodyAimOffsets}
   Input Pose!() -> (Pose) {name=InPosePreFullBodyAimOffsets}
   Output Pose() {layerGroup=Item Override}
  ]}
 ItemFullBodyOverride {nodes [
   AnimGraphNode_StateMachine_1 = RocketDrill() {is=title AnimGraphNode_StateMachine, bReinitializeOnBecomingRelevant=false}
   n1 = Apply Additive(
    Base=@AnimGraphNode_StateMachine_1
    Additive="Sequence Player (additive) 'RocketDrill_Idle_Noise'"() {sequence=&/RocketDrill_Idle_Noise}
    Alpha=Get IdleNoiseAlpha() -> (IdleNoiseAlpha:float)
   ) {alphaBoolBlend {blendOption=Linear}}
   n2 = Get AimYaw() -> (AimYaw:float)
   n3 = Get AimPitch()
   n4 = Get DrillingAO_Alpha() -> (DrillingAO_Alpha:double)
   Output Pose(Result=Dead Blending!(Source:Pose Link Structure=AimOffset Player 'RocketDrill_AO_BS'!(X:float=@n2, Y:float=@n3, BasePose:Pose Link Structure=@n1, Alpha:float=@n4) -> (Pose:Pose Link Structure) {blendSpace=/RocketDrillGameplay/Assets/Animations/AO/RocketDrill_AO_BS, alphaBoolBlend {blendOption=Linear}}) -> (Pose:Pose Link Structure) {is=title AnimGraphNode_DeadBlending, defaultBlendMode=Linear}) {layerGroup=Item Override}
   Save cached pose 'FullBody Pose'(Pose=Input Pose() {name=InPoseFullBody}) {cacheName=FullBody Pose}
  ]}
 RocketDrill {
  entry=Base
  path=ItemFullBodyOverride.AnimGraphNode_StateMachine_1.RocketDrill
  ids=id
  nodes [
   ToTakeOff() {is=title AnimStateAliasNode, also {aliases [Base, Boost_Dismount, Boost_End_Skid, Boost_Fall, Boost_Land, HardLand]}}
   FromBoostLand() {is=title AnimStateAliasNode, also {aliases [Boost_Land]}}
   ToHardLand() {is=title AnimStateAliasNode, also {aliases [Boost_Dismount, Boost_Fall, Boost_Loop, Boost_ReEnter]}}
   ToWallHit() {is=title AnimStateAliasNode, also {aliases [Boost_Dismount, Boost_Fall, Boost_Loop, Boost_ReEnter]}}
   ToBase() {is=title AnimStateAliasNode, also {aliases [Boost_Dismount, Boost_Fall, Boost_Loop]}}
   ToSkid() {is=title AnimStateAliasNode, also {aliases [Boost_Dismount, Boost_Fall, Boost_Loop, Boost_ReEnter]}}
   AnimStateNode_0 = Base() {is=title AnimStateNode}
   AnimStateNode_1 = Boost_ReEnter() {is=title AnimStateNode}
   AnimStateNode_15 = HardLand() {is=title AnimStateNode}
   AnimStateNode_17 = Boost_Loop() {is=title AnimStateNode}
   AnimStateNode_18 = Boost_Fall() {is=title AnimStateNode}
   AnimStateNode_19 = Boost_End_Skid() {is=title AnimStateNode}
   AnimStateNode_2 = Wallhit() {is=title AnimStateNode}
   AnimStateNode_3 = Boost_Dismount() {is=title AnimStateNode}
   AnimStateNode_4 = Boost_Land() {is=title AnimStateNode}
   AnimStateTransitionNode_0 = AnimStateTransitionNode()
   AnimStateTransitionNode_1 = AnimStateTransitionNode()
   AnimStateTransitionNode_13 = AnimStateTransitionNode() {automaticRuleTriggerTime=0}
   AnimStateTransitionNode_14 = AnimStateTransitionNode() {bAutomaticRuleBasedOnSequencePlayerInState=true}
   AnimStateTransitionNode_15 = AnimStateTransitionNode()
   AnimStateTransitionNode_17 = AnimStateTransitionNode() {bAutomaticRuleBasedOnSequencePlayerInState=true}
   AnimStateTransitionNode_19 = AnimStateTransitionNode() {
    bAutomaticRuleBasedOnSequencePlayerInState=true
    transitionEnd {guid=#1}
    transitionInterrupt {guid=#2}
    transitionStart {guid=#3}
   }
   AnimStateTransitionNode_2 = AnimStateTransitionNode() {crossfadeDuration=0}
   AnimStateTransitionNode_3 = AnimStateTransitionNode() {automaticRuleTriggerTime=0}
   AnimStateTransitionNode_36 = AnimStateTransitionNode() {priorityOrder=5}
   AnimStateTransitionNode_38 = AnimStateTransitionNode()
   AnimStateTransitionNode_4 = AnimStateTransitionNode() {crossfadeDuration=0.3}
   AnimStateTransitionNode_5 = AnimStateTransitionNode() {crossfadeDuration=0.1}
   AnimStateTransitionNode_6 = AnimStateTransitionNode() {bAutomaticRuleBasedOnSequencePlayerInState=true}
   AnimStateTransitionNode_7 = AnimStateTransitionNode() {crossfadeDuration=0, priorityOrder=3}
   AnimStateTransitionNode_8 = AnimStateTransitionNode() {bAutomaticRuleBasedOnSequencePlayerInState=true, crossfadeDuration=0.4}
   AnimStateTransitionNode_9 = AnimStateTransitionNode() {crossfadeDuration=0.1}
  ]
  links [
   {
    dst=Boost_Fall
    rule=ItemFullBodyOverride.AnimGraphNode_StateMachine_1.RocketDrill.AnimStateTransitionNode_6.AnimationTransitionGraph_0
    src=Boost_Dismount
    via=AnimStateTransitionNode_6
   }
   {
    dst=Boost_Land
    rule=ItemFullBodyOverride.AnimGraphNode_StateMachine_1.RocketDrill.AnimStateTransitionNode_13.Transition
    src=Boost_Dismount
    via=AnimStateTransitionNode_3
   }
   {
    dst=Base
    rule=ItemFullBodyOverride.AnimGraphNode_StateMachine_1.RocketDrill.AnimStateTransitionNode_1.Transition
    src=Boost_End_Skid
    via=AnimStateTransitionNode_1
   }
   {
    dst=Base
    rule=ItemFullBodyOverride.AnimGraphNode_StateMachine_1.RocketDrill.AnimStateTransitionNode_19.AnimationTransitionGraph_0
    src=Boost_End_Skid
    via=AnimStateTransitionNode_19
   }
   {
    dst=Boost_Land
    rule=ItemFullBodyOverride.AnimGraphNode_StateMachine_1.RocketDrill.AnimStateTransitionNode_13.Transition
    src=Boost_Fall
    via=AnimStateTransitionNode_13
   }
   {
    dst=Boost_Dismount
    rule=ItemFullBodyOverride.AnimGraphNode_StateMachine_1.RocketDrill.AnimStateTransitionNode_9.Transition
    src=Boost_Loop
    via=AnimStateTransitionNode_5
   }
   {
    dst=Boost_Dismount
    rule=ItemFullBodyOverride.AnimGraphNode_StateMachine_1.RocketDrill.AnimStateTransitionNode_9.Transition
    src=Boost_ReEnter
    via=AnimStateTransitionNode_9
   }
   {
    dst=Boost_Loop
    rule=ItemFullBodyOverride.AnimGraphNode_StateMachine_1.RocketDrill.AnimStateTransitionNode_4.AnimationTransitionGraph_0
    src=Boost_ReEnter
    via=AnimStateTransitionNode_4
   }
   {
    dst=Base
    rule=ItemFullBodyOverride.AnimGraphNode_StateMachine_1.RocketDrill.AnimStateTransitionNode_17.Transition
    src=FromBoostLand
    via=AnimStateTransitionNode_17
   }
   {
    dst=Base
    rule=ItemFullBodyOverride.AnimGraphNode_StateMachine_1.RocketDrill.AnimStateTransitionNode_14.AnimationTransitionGraph_0
    src=HardLand
    via=AnimStateTransitionNode_14
   }
   {
    dst=Base
    rule=ItemFullBodyOverride.AnimGraphNode_StateMachine_1.RocketDrill.AnimStateTransitionNode_15.AnimationTransitionGraph_0
    src=HardLand
    via=AnimStateTransitionNode_15
   }
   {
    dst=Base
    rule=ItemFullBodyOverride.AnimGraphNode_StateMachine_1.RocketDrill.AnimStateTransitionNode_36.AnimationTransitionGraph_0
    src=ToBase
    via=AnimStateTransitionNode_36
   }
   {
    dst=HardLand
    rule=ItemFullBodyOverride.AnimGraphNode_StateMachine_1.RocketDrill.AnimStateTransitionNode_7.AnimationTransitionGraph_0
    src=ToHardLand
    via=AnimStateTransitionNode_7
   }
   {
    dst=Boost_End_Skid
    rule=ItemFullBodyOverride.AnimGraphNode_StateMachine_1.RocketDrill.AnimStateTransitionNode_38.AnimationTransitionGraph_0
    src=ToSkid
    via=AnimStateTransitionNode_38
   }
   {
    dst=Boost_ReEnter
    rule=ItemFullBodyOverride.AnimGraphNode_StateMachine_1.RocketDrill.AnimStateTransitionNode_0.Transition
    src=ToTakeOff
    via=AnimStateTransitionNode_0
   }
   {
    dst=Wallhit
    rule=ItemFullBodyOverride.AnimGraphNode_StateMachine_1.RocketDrill.AnimStateTransitionNode_2.Transition
    src=ToWallHit
    via=AnimStateTransitionNode_2
   }
   {
    dst=Base
    rule=ItemFullBodyOverride.AnimGraphNode_StateMachine_1.RocketDrill.AnimStateTransitionNode_8.Transition
    src=Wallhit
    via=AnimStateTransitionNode_8
   }
  ]
 }
 Base {
  path=ItemFullBodyOverride.AnimGraphNode_StateMachine_1.RocketDrill.AnimStateNode_0.Base
  nodes [Output Animation Pose(Result=Use cached pose 'FullBody Pose'())]
 }
 Boost_ReEnter {
  path=ItemFullBodyOverride.AnimGraphNode_StateMachine_1.RocketDrill.AnimStateNode_1.Boost_ReEnter
  nodes [
   Component To Local()
   n1 = Sequence Player 'RocketDrill_In_Air_Burst_Fwd'!(StartPosition:float=0) -> (Pose) {
    bLoopAnimation=false
    groupName=Air_Burst
    method=SyncGroup
    sequence=/RocketDrillGameplay/Assets/Animations/Burst/RocketDrill_In_Air_Burst_Fwd
    startPosition=0.3
   }
   n2 = "Transform (Modify) Bone - Bone: Root"!(
    ComponentPose=Local To Component(LocalPose=@n1)
    Alpha=1
    Translation="0,0,0"
    Rotation_Roll:float=0
    Rotation_Pitch:float=0
    Rotation_Yaw:float=Get StartBoostYaw() -> (StartBoostYaw:float)
    Scale="1,1,1"
   ) -> (Pose) {
    rotationMode=BMM_Replace
    rotationSpace=BCS_WorldSpace
    translationSpace=BCS_BoneSpace
    alphaBoolBlend {blendOption=Linear}
    boneToModify {boneName=Root}
   }
   n3 = "Transform (Modify) Bone - Bone: Root"(
    ComponentPose=@n2
    Alpha=1
    Translation="0,0,0"
    Rotation_Roll=Get BoostPitch() -> (BoostPitch:double)
    Rotation_Pitch=0
    Rotation_Yaw=0
    Scale="1,1,1"
   ) {
    rotationMode=BMM_Additive
    rotationSpace=BCS_BoneSpace
    translationSpace=BCS_BoneSpace
    alphaBoolBlend {blendOption=Linear}
    boneToModify {boneName=Root}
   }
   Output Animation Pose(Result=Modify Curve(SourcePose=Component To Local(ComponentPose=@n3), CurveValues_0=1, Alpha=1) {curveNames [UserCurve01], curveValues [0]})
  ]
 }
 HardLand {
  path=ItemFullBodyOverride.AnimGraphNode_StateMachine_1.RocketDrill.AnimStateNode_15.HardLand
  nodes [
   n1 = Sequence Player 'RocketDrill_Burst_HeavyLand'() {
    bLoopAnimation=false
    groupName=HeavyLand
    method=SyncGroup
    sequence=/RocketDrillGameplay/Assets/Animations/Burst/RocketDrill_Burst_HeavyLand
   }
   n2 = Layered blend per bone(BasePose=@n1, BlendPoses_0=Use cached pose 'FullBody Pose'()) {blendMode=BlendMask, blendMasks [&Fortnite_M_Avg_Player_Skeleton:LowerBody], layerSetup []}
   n3 = Sequence Player 'RocketDrill_Burst_HeavyLand'() {
    bLoopAnimation=false
    groupName=HeavyLand
    method=SyncGroup
    sequence=&RocketDrill_Burst_HeavyLand
   }
   Output Animation Pose(Result=Blend Poses by bool(BlendPose_0=@n2, BlendPose_1=@n3, BlendTime_0=0.2, BlendTime_1=0.2, bActiveValue=Get bIsMoving2D()) {blendType=Linear}) {
    node {stateEntryFunction {functionName=OnStateEntry}, stateExitFunction {functionName=OnStateFullyBlendedOut}}
    stateEntryFunction {bSelfContext=true, memberGuid=#4, memberName=OnStateEntryWallHigHardLand}
    stateExitFunction {bSelfContext=true, memberGuid=#5, memberName=OnExitWallhitHardLand}
   }
  ]
 }
 Boost_Loop {
  path=ItemFullBodyOverride.AnimGraphNode_StateMachine_1.RocketDrill.AnimStateNode_17.Boost_Loop
  nodes [
   n1 = Get YawDeltaSmoothed() -> (YawDeltaSmoothed:float)
   n2 = "Transform (Modify) Bone - Bone: Root"(
    ComponentPose=Local To Component(LocalPose=Sequence Player 'RocketDrill_Burst_Loop'() {sequence=/RocketDrillGameplay/Assets/Animations/Burst/RocketDrill_Burst_Loop})
    Alpha=1
    Translation="0,0,0"
    Rotation_Roll=0
    Rotation_Pitch=0
    Rotation_Yaw=Get VelocityYaw() -> (VelocityYaw:double)
    Scale="1,1,1"
   ) {
    rotationMode=BMM_Replace
    rotationSpace=BCS_WorldSpace
    translationSpace=BCS_BoneSpace
    alphaBoolBlend {blendOption=Linear}
    boneToModify {boneName=Root}
   }
   n3 = "Transform (Modify) Bone - Bone: Root"(ComponentPose=@n2, Alpha=1, Translation="0,0,0", Rotation_Roll=Get BoostPitch(), Rotation_Pitch=0, Rotation_Yaw=0, Scale="1,1,1") {
    rotationMode=BMM_Additive
    rotationSpace=BCS_BoneSpace
    translationMode=BMM_Additive
    translationSpace=BCS_BoneSpace
    alphaBoolBlend {blendOption=Linear}
    boneToModify {boneName=Root}
   }
   n4 = Component To Local(ComponentPose=@n3)
   Output Animation Pose(Result=Apply Additive(
     Base=Modify Curve!(SourcePose=@n4, CurveValues_0=1, CurveValues_1:float=1, Alpha=1) -> (Pose) {curveNames [UserCurve01,UserCurve02], curveValues [0,0]}
     Additive=Blendspace Player 'RocketDrill_Burst_Leans'!(X=@n1) -> (Pose) {blendSpace=&/RocketDrill_Burst_Leans}
    ) {alphaBoolBlend {blendOption=Linear}})
  ]
 }
 Boost_Fall {
  path=ItemFullBodyOverride.AnimGraphNode_StateMachine_1.RocketDrill.AnimStateNode_18.Boost_Fall
  nodes [Output Animation Pose(Result=Modify Curve!(
     SourcePose=Sequence Player 'RocketDrill_Burst_Dismount_Fall'() {bLoopAnimation=false, sequence=/RocketDrillGameplay/Assets/Animations/Burst/RocketDrill_Burst_Dismount_Fall}
     CurveValues_0=1
     Alpha=1
    ) -> (Pose) {curveNames [UserCurve01], curveValues [0]})]
 }
 Boost_End_Skid {
  path=ItemFullBodyOverride.AnimGraphNode_StateMachine_1.RocketDrill.AnimStateNode_19.Boost_End_Skid
  nodes [
   n1 = Sequence Player 'RocketDrill_Burst_Stop'() {bLoopAnimation=false, groupName=Burst_Stop, method=SyncGroup, sequence=/RocketDrillGameplay/Assets/Animations/Burst/RocketDrill_Burst_Stop}
   n2 = Layered blend per bone(BasePose=@n1, BlendPoses_0=Use cached pose 'FullBody Pose'(), BlendWeights_0=Get Drill_Burst_Stop_Float() -> (Drill_Burst_Stop_Float:float)) {blendMode=BlendMask, blendMasks [&Fortnite_M_Avg_Player_Skeleton:LowerBody], layerSetup []}
   n3 = Sequence Player 'RocketDrill_Burst_Stop'() {bLoopAnimation=false, groupName=Burst_Stop, method=SyncGroup, sequence=&RocketDrill_Burst_Stop}
   n4 = Sequence Player 'RocketDrill_Burst_Stop'() {bLoopAnimation=false, groupName=Burst_Stop, method=SyncGroup, sequence=&RocketDrill_Burst_Stop}
   Output Animation Pose(Result=Blend Poses by bool(
     BlendPose_0=Layered blend per bone(BasePose=@n2, BlendPoses_0=@n3) {bMeshSpaceRotationBlend=true, layerSetup [{branchFilters [{boneName=Spine_01}]}]}
     BlendPose_1=@n4
     bActiveValue=Get bIsMoving() -> (bIsMoving:bool)
    ) {blendType=Linear})
  ]
 }
 Wallhit {
  path=ItemFullBodyOverride.AnimGraphNode_StateMachine_1.RocketDrill.AnimStateNode_2.Wallhit
  nodes [Output Animation Pose(Result=Rotate Root Bone!(
     BasePose:Pose Link Structure=Sequence Player 'RocketDrill_Burst_WallHit'() {bLoopAnimation=false, sequence=/RocketDrillGameplay/Assets/Animations/Burst/RocketDrill_Burst_WallHit}
     Pitch:float=Get Pitch() -> (Pitch:double)
     Yaw:float=Get Yaw() -> (Yaw:double)
     MeshToComponent:Rotator="0,-90,0"
    ) -> (Pose:Pose Link Structure) {is=title AnimGraphNode_RotateRootBone}) {
    node {stateEntryFunction {functionName=OnStateEntry}, stateExitFunction {functionName=OnStateFullyBlendedOut}}
    stateEntryFunction {bSelfContext=true, memberGuid=#4, memberName=OnStateEntryWallHigHardLand}
    stateExitFunction {bSelfContext=true, memberGuid=#5, memberName=OnExitWallhitHardLand}
   }]
 }
 Boost_Dismount {
  path=ItemFullBodyOverride.AnimGraphNode_StateMachine_1.RocketDrill.AnimStateNode_3.Boost_Dismount
  nodes [
   "Transform (Modify) Bone - Bone: Root"(Alpha=1, Translation="0,0,0", Rotation_Roll=Get BoostPitch(), Rotation_Pitch=0, Rotation_Yaw=0, Scale="1,1,1") {
    rotationMode=BMM_Additive
    rotationSpace=BCS_BoneSpace
    translationMode=BMM_Additive
    translationSpace=BCS_BoneSpace
    alphaBoolBlend {blendOption=Linear}
    boneToModify {boneName=Root}
   }
   Output Animation Pose(Result=Sequence Player 'RocketDrill_Burst_Dismount'() {bLoopAnimation=false, sequence=/RocketDrillGameplay/Assets/Animations/Burst/RocketDrill_Burst_Dismount}) {
    initialUpdateFunction {bSelfContext=true, memberGuid=#6, memberName=BurstDismount}
    node {initialUpdateFunction {functionName=OnInitialUpdate}}
   }
  ]
 }
 Boost_Land {
  path=ItemFullBodyOverride.AnimGraphNode_StateMachine_1.RocketDrill.AnimStateNode_4.Boost_Land
  nodes [
   n1 = Sequence Player 'RocketDrill_Burst_Dismount_Land'() {
    bLoopAnimation=false
    groupName=Dismount_Land
    method=SyncGroup
    sequence=/RocketDrillGameplay/Assets/Animations/Burst/RocketDrill_Burst_Dismount_Land
   }
   n2 = Layered blend per bone(BasePose=@n1, BlendPoses_0=Use cached pose 'FullBody Pose'()) {blendMode=BlendMask, blendMasks [&Fortnite_M_Avg_Player_Skeleton:LowerBody], layerSetup []}
   n3 = Sequence Player 'RocketDrill_Burst_Dismount_Land'() {
    bLoopAnimation=false
    groupName=Dismount_Land
    method=SyncGroup
    sequence=&RocketDrill_Burst_Dismount_Land
   }
   n4 = Layered blend per bone(BasePose=@n2, BlendPoses_0=@n3, BlendWeights_0=0.5) {layerSetup [{branchFilters [{boneName=pelvis}, {blendDepth=-1, boneName=Spine_01}, {blendDepth=-1, boneName=thigh_l}, {blendDepth=-1, boneName=thigh_r}]}]}
   n5 = Sequence Player 'RocketDrill_Burst_Dismount_Land'() {
    bLoopAnimation=false
    groupName=Dismount_Land
    method=SyncGroup
    sequence=&RocketDrill_Burst_Dismount_Land
   }
   Output Animation Pose(Result=Blend Poses by bool(BlendPose_0=@n4, BlendPose_1=@n5, BlendTime_0=0.2, BlendTime_1=0.2, bActiveValue=Get bIsMoving()) {blendType=Linear})
  ]
 }
 Transition {
  path=ItemFullBodyOverride.AnimGraphNode_StateMachine_1.RocketDrill.AnimStateTransitionNode_0.Transition
  nodes [Result(bCanEnterTransition=Get bDrillActive() -> (bDrillActive:bool))]
 }
 Transition {
  path=ItemFullBodyOverride.AnimGraphNode_StateMachine_1.RocketDrill.AnimStateTransitionNode_1.Transition
  nodes [Result(bCanEnterTransition=Get bIsOtherMovementMode() -> (bIsOtherMovementMode:bool))]
 }
 Transition {
  path=ItemFullBodyOverride.AnimGraphNode_StateMachine_1.RocketDrill.AnimStateTransitionNode_13.Transition
  nodes [Result(bCanEnterTransition=Get bCanEnterToLand() -> (bCanEnterToLand:bool))]
 }
 AnimationTransitionGraph_0 {path=ItemFullBodyOverride.AnimGraphNode_StateMachine_1.RocketDrill.AnimStateTransitionNode_14.AnimationTransitionGraph_0, nodes [Result()]}
 AnimationTransitionGraph_0 {
  path=ItemFullBodyOverride.AnimGraphNode_StateMachine_1.RocketDrill.AnimStateTransitionNode_15.AnimationTransitionGraph_0
  nodes [Result(bCanEnterTransition=Get bIsTacticalSprint() -> (bIsTacticalSprint:bool))]
 }
 Transition {path=ItemFullBodyOverride.AnimGraphNode_StateMachine_1.RocketDrill.AnimStateTransitionNode_17.Transition, nodes [Result()]}
 AnimationTransitionGraph_0 {path=ItemFullBodyOverride.AnimGraphNode_StateMachine_1.RocketDrill.AnimStateTransitionNode_19.AnimationTransitionGraph_0, nodes [Result()]}
 Transition {
  path=ItemFullBodyOverride.AnimGraphNode_StateMachine_1.RocketDrill.AnimStateTransitionNode_2.Transition
  nodes [Result(bCanEnterTransition=Get bIntoWallHit() -> (bIntoWallHit:bool))]
 }
 AnimationTransitionGraph_0 {
  path=ItemFullBodyOverride.AnimGraphNode_StateMachine_1.RocketDrill.AnimStateTransitionNode_36.AnimationTransitionGraph_0
  nodes [Result(bCanEnterTransition=Get bCanEnterToBase() -> (bCanEnterToBase:bool))]
 }
 AnimationTransitionGraph_0 {
  path=ItemFullBodyOverride.AnimGraphNode_StateMachine_1.RocketDrill.AnimStateTransitionNode_38.AnimationTransitionGraph_0
  nodes [Result(bCanEnterTransition=Get bDrillSkid() -> (bDrillSkid:bool))]
 }
 AnimationTransitionGraph_0 {
  path=ItemFullBodyOverride.AnimGraphNode_StateMachine_1.RocketDrill.AnimStateTransitionNode_4.AnimationTransitionGraph_0
  nodes [Result(bCanEnterTransition=Get bBurstStartBlendIn() -> (bBurstStartBlendIn:bool))]
 }
 AnimationTransitionGraph_0 {path=ItemFullBodyOverride.AnimGraphNode_StateMachine_1.RocketDrill.AnimStateTransitionNode_6.AnimationTransitionGraph_0, nodes [Result()]}
 AnimationTransitionGraph_0 {
  path=ItemFullBodyOverride.AnimGraphNode_StateMachine_1.RocketDrill.AnimStateTransitionNode_7.AnimationTransitionGraph_0
  nodes [Result(bCanEnterTransition=Get bLandingWithImpact() -> (bLandingWithImpact:bool))]
 }
 Transition {path=ItemFullBodyOverride.AnimGraphNode_StateMachine_1.RocketDrill.AnimStateTransitionNode_8.Transition, nodes [Result()]}
 Transition {
  path=ItemFullBodyOverride.AnimGraphNode_StateMachine_1.RocketDrill.AnimStateTransitionNode_9.Transition
  nodes [Result(bCanEnterTransition=Get bCanEnterIntoFall() -> (bCanEnterIntoFall:bool))]
 }
 ItemIdleOverride {nodes [
   Input Pose() {name=InPose_Idle}
   Output Pose() {layerGroup=Item Override}
  ]}
 ItemJetPackJumpAdditive {nodes [Output Pose() {layerGroup=Item Override}]}
 ItemJetPackStartAdditive {nodes [Output Pose() {layerGroup=Item Override}]}
 ItemJumpLoopAdditive {nodes [
   Input Pose() {name=InPoseItemJumpLoopAdditive}
   Output Pose() {layerGroup=Item Override}
  ]}
 ItemJumpOverride {nodes [
   Input Pose() {name=InPoseJump}
   Output Pose() {layerGroup=Item Override}
  ]}
 ItemJumpUpAdditive {nodes [
   Input Pose() {name=InPoseItemJumpUpAdditive}
   Output Pose() {layerGroup=Item Override}
  ]}
 ItemPreAimOffsets {nodes [
   Input Pose() {name=InPosePreAimOffsets}
   Output Pose() {layerGroup=Item Override}
  ]}
 ItemPreIK {nodes [
   Input Pose() {name=InPose_PreIK}
   Output Pose() {layerGroup=Item Override}
  ]}
 ItemRightHandIK {nodes [
   Input Pose() {name=InPosePreRightHandIK}
   Input Pose() {name=InPosePostRightHandIK}
   Output Pose() {layerGroup=Item Override}
  ]}
 ItemSkeletalControl {nodes [
   Input Pose()
   Output Pose() {layerGroup=Item Override}
  ]}
 ItemSwimJumpFallLoop {nodes [Output Pose() {layerGroup=Item Override}]}
 ItemSwimJumpSurfaceEnd {nodes [Output Pose() {layerGroup=Item Override}]}
 ItemSwimJumpSurfaceLoop {nodes [Output Pose() {layerGroup=Item Override}]}
 ItemUpperBody {nodes [
   n1 = Sequence Player 'RocketDrill_Sliding_Pose_CMF'() {bLoopAnimation=false, sequence=/Game/Animation/Game/MainPlayerFN/Poses/Rocket_Drill/RocketDrill_Sliding_Pose_CMF}
   n2 = Sequence Player 'RocketDrill_Sliding_Pose'() {bLoopAnimation=false, sequence=/Game/Animation/Game/MainPlayerFN/Poses/Rocket_Drill/RocketDrill_Sliding_Pose}
   n3 = Get bIsFemale() -> (bIsFemale:bool)
   Output Pose(Result=Blend Poses by bool(
     BlendPose_0=Blend Poses by bool(BlendPose_0=@n1, BlendPose_1=@n2, BlendTime_0=0.2, BlendTime_1=0.2, bActiveValue=@n3) {blendType=Linear}
     BlendPose_1=Use cached pose 'upperBody CachedPose'()
     bActiveValue=Get bIsSliding()
    ) {blendType=Linear}) {layerGroup=Item Override}
   Save cached pose 'upperBody CachedPose'(Pose=Input Pose!() -> (Pose, InputParam:Fort Anim Input Adjusted Aim Structure) {name=InPoseUpperBody}) {cacheName=upperBody CachedPose}
  ]}
 ItemZipLineStartAdditive {nodes [Output Pose() {layerGroup=Item Override}]}
 OnExitWallhitHardLand {nodes [
   OnExitWallhitHardLand!() -> (then, Context:"Anim Update Context Structure (by ref)", Node:"Anim Node Reference Structure (by ref)") {is=title K2Node_FunctionEntry}
   > Set bExitHardLand(bExitHardLand:bool=true) -> (Output_Get:bool)
   > Return Node()
  ]}
 OnStateEntryWallHigHardLand {nodes [
   OnStateEntryWallHigHardLand!() -> (then, Context:"Anim Update Context Structure (by ref)", Node:"Anim Node Reference Structure (by ref)") {is=title K2Node_FunctionEntry}
   > Set bExitHardLand(bExitHardLand=false)
   > Return Node()
  ]}
]
variables []
