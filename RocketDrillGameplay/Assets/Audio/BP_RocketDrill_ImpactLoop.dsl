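; Rocket drill impact loop actor that drives looping surface-specific sound for whatever the drill hits.
; Target data (not flying) or damage traces (flying, first non-Volume hit) supply the hit actor, which ActorToSurface classifies by vault-door tag or water/player casts, caching the last actor.
; PreviousMat's OnRep updates a surface sound component created on cosmetic-capable machines.
; On destruction while not flying, it plays a beam-end sound attached to the instigator.
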(event EventBeginPlay
  (Networking|CanExecuteCosmeticEvents
    (:True
      (bind _returnvalue (Game|AddComponentbyClass 0 "/Game/Sounds/Fort_Audio_Blueprints/SurfaceTypeSound/BP_SurfaceSoundComponent.BP_SurfaceSoundComponent_C" (Variables|Default|GetDefaultSceneRoot) (Variables|Default|GetEndpoint_Sound) (Variables|Default|GetSurface%20Sound%20Data%20Table)))
      (Variables|Default|SetSurface%20Sound%20Component _returnvalue))))

(event Custom|UpdateHitActor
  (params (HitActor :type Actor%20Object%20Reference))
  (bind (_change _surface) (CallFunction|ActortoSurface HitActor))
  (Variables|Transient|SetPreviousMat _surface))

(event EventDestroyed
  (if (not (Variables|Transient|GetisFlying))
    (Audio|SpawnSoundAttached (Variables|Default|GetBeamEnd_Sound) (Variables|/Script/Engine.Actor|GetRootComponent (Variables|Actor|GetInstigator)))))

(event Custom|OnTargetDataUpdate
  (params (TargetData :type Gameplay%20Ability%20Target%20Data%20Handle%20Structure))
  (bind (_bblockinghit _binitialoverlap _time _distance _location _impactpoint _normal _impactnormal _physmat _hitactor) (Collision|BreakHitResult (Ability|TargetData|GetHitResultfromTargetData TargetData)))
  (Variables|Transient|SetisFlying false)
  (if (Ability|TargetData|TargetDataHasHitResult TargetData)
    (CallFunction|UpdateHitActor _hitactor)
    (else
      (CallFunction|UpdateHitActor 0))))

(event Custom|UpdateFromDamageTrace
  (params (HitResults :type Array%20of%20Hit%20Result%20Structures))
  (CallFunction|OnUpdatefromDamageTrace HitResults))

(fn ActorToSurface
  (params (HitActor :type Actor%20Object%20Reference))
  (bind _hitactor (Variables|GetHitActor))
  (if (== HitActor (Variables|Transient|GetPreviousActor))
    (return false (Variables|Transient|GetPreviousMat))
    (else
      (Variables|GetHitActor
        (:then
          (Utilities|FlowControl|Sequence
            (:then_0
              (bind _output_get (Variables|Transient|SetPreviousActor (Variables|GetHitActor)))
              (bind _asbuilding_gameplay_actor (Utilities|Casting|CastToBuildingGameplayActor _output_get)
                (:then
                  (if (GameplayTags|HasMatchingGameplayTag _asbuilding_gameplay_actor "(TagName=\"Athena.Quests.Building.Vault.Door\")")
                    (LocalVariables|SetNewSurface "SurfaceType3")
                    (else
                      _)))
                (:CastFailed
                  _))
              (Utilities|Casting|CastToWaterBody _hitactor
                (:then
                  (LocalVariables|SetNewSurface "SurfaceType8"))
                (:CastFailed
                  (Utilities|Casting|CastToPlayerPawn_Athena _hitactor
                    (:then
                      (LocalVariables|SetNewSurface "SurfaceType4"))
                    (:CastFailed
                      (LocalVariables|SetNewSurface "SurfaceType10"))))))
            (:then_1
              (return (Utilities|Enum|NotEqual(Enum) (Variables|Transient|GetPreviousMat) (LocalVariables|GetNewSurface)) (LocalVariables|GetNewSurface)))))
        (:else
          (Variables|Transient|SetPreviousActor 0)
          (return true "SurfaceType_Default"))))))

(fn OnRep_PreviousMat
  (Networking|CanExecuteCosmeticEvents
    (:True
      (bind _surface_sound_component (Variables|Default|GetSurface%20Sound%20Component)
        (:then
          (Class|SurfaceSoundComponent|SurfaceTypeChanged _surface_sound_component (Variables|Transient|GetPreviousMat)))
        (:else)))))

(fn OnUpdateFromDamageTrace
  (params (HItResults :type Array%20of%20Hit%20Result%20Structures))
  (Variables|Transient|SetisFlying true)
  (for _array_element HItResults
    (Utilities|Casting|CastToVolume (.HitActor (Collision|BreakHitResult _array_element))
      (:CastFailed
        (CallFunction|UpdateHitActor (.HitActor (Collision|BreakHitResult _array_element)))
        (return))))
  (CallFunction|UpdateHitActor 0))

(fn ConstructionScript)
