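# Looping stereo MetaSound for the Rocket Drill's 1P in-air flight, mixing a random wind loop with engine and sideways layers.
# Absolute yaw is normalized and smoothed with fast rise and slow fall, driving the sideways layer, whose mixer gain is zero.
# The larger of normalized pitch and smoothed yaw drives the engine layer; both layers reuse the Glider_ThrustLoop_Direction patch.
# An attack envelope fades in the mix on play.
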MetaSoundSource /RocketDrillGameplay/Assets/Audio/MSS_RocketDrill_FlyingAir_1P
inputs [
 Assets.EngineLoop {type=WaveAsset:Array}
 Assets.SidewaysLoop {type=WaveAsset:Array}
 Assets.SidewaysOneshots {type=WaveAsset:Array}
 Assets.WindLoop {type=WaveAsset:Array}
 Pitch {type=Float}
 UE.Attenuation.Distance {type=Float}
 UE.Source.OnPlay {type=Trigger}
 Yaw {type=Float}
]
interfaces [UE.Attenuation:1.0, UE.OutputFormat.Stereo:1.0, UE.Source:1.1]
outputs [
 "UE.OutputFormat.Stereo.Audio:0" {type=Audio}
 "UE.OutputFormat.Stereo.Audio:1" {type=Audio}
]
pages [Default {
  input_defaults {
   Pitch=0
   UE.Attenuation.Distance=0
   UE.Source.OnPlay=null
   Yaw=0
   Assets.EngineLoop [
    /RocketDrillGameplay/Assets/Audio/waves/RocketDrill_InAir_Engine_loop_01
    /RocketDrillGameplay/Assets/Audio/waves/RocketDrill_InAir_Engine_loop_02
   ]
   Assets.SidewaysLoop [/RocketDrillGameplay/Assets/Audio/waves/Temp/BarnTowel_Bank_Loop_02, /RocketDrillGameplay/Assets/Audio/waves/Temp/BarnTowel_Bank_Loop_01]
   Assets.SidewaysOneshots [
    /RocketDrillGameplay/Assets/Audio/waves/Temp/BarnTowel_Rattle_01
    /RocketDrillGameplay/Assets/Audio/waves/Temp/BarnTowel_Rattle_02
    /RocketDrillGameplay/Assets/Audio/waves/Temp/BarnTowel_Rattle_03
    /RocketDrillGameplay/Assets/Audio/waves/Temp/BarnTowel_Rattle_04
   ]
   Assets.WindLoop [
    /RocketRamGameplay/Items/RocketRam/Sounds/RocketRam_Fire_InAirLoop_Wind_Up_01
    /RocketRamGameplay/Items/RocketRam/Sounds/RocketRam_Fire_InAirLoop_Wind_Up_02
    /RocketRamGameplay/Items/RocketRam/Sounds/RocketRam_Fire_InAirLoop_Wind_Up_03
   ]
  }
  nodes [
   InitVariable.Audio()
   InitVariable.Float()
   InitVariable.Float()
   InitVariable.Float()
   InitVariable.Float()
   n1 = MapRange.MapRange:Float(In=AbsoluteValue.Abs:Float(Input=^Yaw).Value) {also {defaults {In Range B=90}}}
   TriggerOnValueChange_0 = UE.Trigger On Value Change:Float(Value=^Yaw)
   n2 = VariableDeferredAccessor.Float(Variable=InitVariable.Float().Variable)
   n3 = TriggerCompare.Clamp:Float(B=@n1.Out Value, Compare=@TriggerOnValueChange_0.Trigger, A=@n2.Value) {also {defaults {Type=3}}}
   TriggerRouteTriggerRouteFloat2_1 = "TriggerRoute.Trigger Route (Float, 2):Float"(Set 1=@n3.FALSE, Set 0=@n3.TRUE) {also {defaults {Value 0=0.75, Value 1=0.1}}}
   n4 = UE.ConversionFloatToTime(In=@TriggerRouteTriggerRouteFloat2_1.Value)
   n5 = VariableAccessor.Float(Variable=VariableMutator.Float(Value=UE.InterpTo:Audio(Interp Time=@n4.out, Target=@n1.Out Value).Value, Variable=@n2.Variable).Variable)
   n6 = /FortMetasound/Cosmetics/Glider/Glider_ThrustLoop_Direction(Loop=^Assets.SidewaysLoop, Oneshots=^Assets.SidewaysOneshots, On Play=^UE.Source.OnPlay, Value=@n5.Value) {also {defaults {Interp Time=0, OneShotVolume=0.5, VolumeMin=0}}}
   n7 = InitVariable.Float()
   n8 = VariableAccessor.Float(Variable=@n5.Variable)
   TriggerOnValueChange_1 = UE.Trigger On Value Change:Float(Value=@n8.Value)
   TriggerOnValueChange_2 = UE.Trigger On Value Change:Float(Value=^Pitch)
   TriggerAnyTriggerAny3_0 = "TriggerAny.Trigger Any (3)"(In 0=^UE.Source.OnPlay, In 2=@TriggerOnValueChange_1.Trigger, In 1=@TriggerOnValueChange_2.Trigger)
   n9 = AbsoluteValue.Abs:Float(Input=^Pitch)
   n10 = VariableAccessor.Float(Variable=VariableMutator.Float(Variable=InitVariable.Float().Variable, Value=MapRange.MapRange:Float(In=@n9.Value) {also {defaults {In Range B=60}}}.Out Value).Variable)
   n11 = TriggerCompare.Clamp:Float(
    B=UE.Reroute(Value=UE.Reroute(Value=@n8.Value).Value).Value
    Compare=@TriggerAnyTriggerAny3_0.out
    A=VariableAccessor.Float(Variable=@n10.Variable).Value
   ) {also {defaults {Type=3}}}
   TriggerRouteTriggerRouteFloat2_0 = "TriggerRoute.Trigger Route (Float, 2):Float"(Value 1=UE.Reroute(Value=UE.Reroute(Value=@n8.Value).Value).Value, Set 1=@n11.FALSE, Set 0=@n11.TRUE, Value 0=@n10.Value)
   n12 = &Glider_ThrustLoop_Direction(
    Loop=^Assets.EngineLoop
    On Play=^UE.Source.OnPlay
    Value=VariableAccessor.Float(Variable=VariableMutator.Float(Variable=@n7.Variable, Value=@TriggerRouteTriggerRouteFloat2_0.Value).Variable).Value
   ) {also {defaults {Interp Time=0.35, PitchMin=1, VolumeMax=1.5, VolumeMin=1, bIsBackwards="<MetaSoundBuilderResult.SUCCEEDED: 0>"}}}
   ArrayRandomGet_0 = Array.Random Get:WaveAsset:Array(In Array=^Assets.WindLoop, Next=^UE.Source.OnPlay) {also {defaults {Weights []}}}
   WavePlayer_0 = UE.Wave Player:Stereo(Play=@ArrayRandomGet_0.On Next, Wave Asset=@ArrayRandomGet_0.Value)
   AudioMixerAudioMixerStereo3_0 = "AudioMixer.Audio Mixer (Stereo, 3)"(In 2 L=@n6.Out L, In 2 R=@n6.Out R, In 1 L=@n12.Out L, In 1 R=@n12.Out R, In 0 L=@WavePlayer_0.Out Left, In 0 R=@WavePlayer_0.Out Right) {also {defaults {Gain 0=0.5, Gain 1=0.75, Gain 2=0}}}
   ADSREnvelopeADSREnvelope_0 = ADSR Envelope.ADSR Envelope:Audio(Trigger Attack=^UE.Source.OnPlay) {also {defaults {Attack Curve=2, Attack Time=0.7, Decay Time=0, Sustain Level=1}}}
   n13 = VariableAccessor.Audio(Variable=VariableMutator.Audio(Value=@ADSREnvelopeADSREnvelope_0.Out Envelope, Variable=InitVariable.Audio().Variable).Variable)
   Multiply_0 = UE.Multiply:Audio(PrimaryOperand=@AudioMixerAudioMixerStereo3_0.Out L, AdditionalOperands=@n13.Value)
   Multiply_1 = UE.Multiply:Audio(PrimaryOperand=@AudioMixerAudioMixerStereo3_0.Out R, AdditionalOperands=@n13.Value)
  ]
  links [{dst="@UE.OutputFormat.Stereo.Audio:0", src=Multiply_0.out}, {dst="@UE.OutputFormat.Stereo.Audio:1", src=Multiply_1.out}]
 }]
properties {
 assetImportData=&self:AssetImportData
 attenuationSettings=/RocketDrillGameplay/Assets/Audio/Settings/ATT_RocketDrill_InAir
 bLooping=true
 duration=10000
 maxDistance=3740.0002
 numChannels=2
 outputFormat=Stereo
 sampleRate=48000
 soundClassObject=&Fort_Gliders_InAir
}
