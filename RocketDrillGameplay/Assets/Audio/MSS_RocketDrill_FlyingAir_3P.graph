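# Mono looping MetaSound source for the Rocket Drill's third-person in-air engine sound, attenuated with the 3P engines settings.
# On play, it picks a random loop from the 3PAirLoop wave array and starts the wave player.
# An ADSR envelope multiplies the audio for a short fade-in.
# A Doppler patch reads attenuation distance, but its output is not connected to the wave player or output.
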MetaSoundSource /RocketDrillGameplay/Assets/Audio/MSS_RocketDrill_FlyingAir_3P
inputs [
 3PAirLoop {type=WaveAsset:Array}
 UE.Attenuation.Distance {type=Float}
 UE.Source.OnPlay {type=Trigger}
 UE.Source.Orientation.Azimuth {type=Float}
 UE.Source.Orientation.Elevation {type=Float}
]
interfaces [UE.Attenuation:1.0, UE.OutputFormat.Mono:1.0, UE.Source.Orientation:1.0, UE.Source:1.1]
outputs ["UE.OutputFormat.Mono.Audio:0" {type=Audio}]
pages [Default {
  input_defaults {
   UE.Attenuation.Distance=0
   UE.Source.OnPlay=null
   UE.Source.Orientation.Azimuth=0
   UE.Source.Orientation.Elevation=0
   3PAirLoop [
    /RocketDrillGameplay/Assets/Audio/waves/RocketDrill_InAir_Engine_loop_01
    /RocketDrillGameplay/Assets/Audio/waves/RocketDrill_InAir_Engine_loop_02
   ]
  }
  nodes [
   InitVariable.Float()
   /FortMetasound/Patches/MSP_Doppler(Distance=^UE.Attenuation.Distance) {also {defaults {StrengthMultiplier=5}}}
   ADSREnvelopeADSREnvelope_0 = ADSR Envelope.ADSR Envelope:Audio(Trigger Attack=^UE.Source.OnPlay) {also {defaults {Attack Time=0.3, Decay Time=0, Release Time=0, Sustain Level=1}}}
   ArrayRandomGet_0 = Array.Random Get:WaveAsset:Array(In Array=^3PAirLoop, Next=^UE.Source.OnPlay)
   WavePlayer_0 = UE.Wave Player:Mono(Play=@ArrayRandomGet_0.On Next, Wave Asset=@ArrayRandomGet_0.Value) {also {defaults {Loop="<MetaSoundBuilderResult.SUCCEEDED: 0>"}}}
   Reroute_0 = UE.Reroute(Value=UE.Multiply:Audio(PrimaryOperand=@ADSREnvelopeADSREnvelope_0.Out Envelope, AdditionalOperands=@WavePlayer_0.Out Mono).out)
  ]
  links [{dst="@UE.OutputFormat.Mono.Audio:0", src=Reroute_0.Value}]
 }]
properties {
 assetImportData=&self:AssetImportData
 attenuationSettings=/RocketDrillGameplay/Assets/Audio/Settings/ATT_RocketDrill_Engines_3P
 bLooping=true
 duration=10000
 maxDistance=21000
 sampleRate=48000
 soundClassObject=&Fort_Gliders_InAir
 virtualizationMode=PlayWhenSilent
}
