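# Mono MetaSound Source that plays Rocket Drill fly-by one-shots as near-miss audio.
# A threshold trigger on attenuation distance, apparently firing as it drops below the limit, picks a random fly-by wave for a Wave Player.
# OnPlay starts nothing, and the Speed and air-loop inputs are unconnected.
# Overridden concurrency allows one instance, preventing new ones, with a retrigger time.
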MetaSoundSource /RocketDrillGameplay/Assets/Audio/MSS_RocketDrill_NearMiss
inputs [
 3P Bys {type=WaveAsset:Array}
 3PAirLoop {type=WaveAsset:Array}
 Speed {type=Float}
 UE.Attenuation.Distance {type=Float}
 UE.Source.OnPlay {type=Trigger}
 UE.Source.Orientation.Azimuth {type=Float}
 UE.Source.Orientation.Elevation {type=Float}
]
interfaces [UE.Attenuation:1.0, UE.OutputFormat.Mono:1.0, UE.Source.Orientation:1.0, UE.Source:1.1]
outputs ["UE.OutputFormat.Mono.Audio:0" {type=Audio}]
pages [Default {
  input_defaults {
   Speed=95.9999
   UE.Attenuation.Distance=0
   UE.Source.OnPlay=null
   UE.Source.Orientation.Azimuth=0
   UE.Source.Orientation.Elevation=0
   3P Bys [
    /RocketDrillGameplay/Assets/Audio/waves/RocketDrill_FlyBy_02
    /RocketDrillGameplay/Assets/Audio/waves/RocketDrill_FlyBy_03
    /RocketDrillGameplay/Assets/Audio/waves/RocketDrill_FlyBy_04
    /RocketDrillGameplay/Assets/Audio/waves/RocketDrill_FlyBy_05
   ]
   3PAirLoop [/RocketRamGameplay/Items/RocketRam/Sounds/RocketRam_Fire_InAirLoop_Rocket_01]
  }
  nodes [
   InitVariable.Float()
   ArrayRandomGet_0 = Array.Random Get:WaveAsset:Array(In Array=^3P Bys, Next=UE.TriggerOnThreshold:Float(In=^UE.Attenuation.Distance) {also {defaults {Threshold=1000, Type=1}}}.Out)
   WavePlayer_0 = UE.Wave Player:Mono(Play=@ArrayRandomGet_0.On Next, Wave Asset=@ArrayRandomGet_0.Value)
  ]
  links [{dst="@UE.OutputFormat.Mono.Audio:0", src=WavePlayer_0.Out Mono}]
 }]
properties {
 assetImportData=&self:AssetImportData
 attenuationSettings=/RocketDrillGameplay/Assets/Audio/Settings/ATT_RocketDrill_NearMiss
 bLooping=true
 bOverrideConcurrency=true
 duration=10000
 maxDistance=4500
 pitch=1.2
 sampleRate=48000
 soundClassObject=&Fort_Gliders_InAir
 concurrencyOverrides {maxCount=1, resolutionRule=PreventNew, retriggerTime=5}
}
