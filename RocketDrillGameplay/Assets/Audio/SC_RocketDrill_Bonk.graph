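# Sound cue for the Rocket Drill "bonk", apparently played when the airborne drill hits something indestructible, judging by its wave names.
# A Random node picks one of four equally weighted impact waves, a Modulator raises its pitch by a random amount, and an Enveloper swells then fades the volume.
# It uses in-air glider sound class and drill attenuation, with per-owner concurrency that stops the oldest instance.
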SoundCue /RocketDrillGameplay/Assets/Audio/SC_RocketDrill_Bonk
root=Enveloper_0
nodes [
 {
  id=Enveloper_0
  children [Modulator_1]
  properties {
   pitchCurve {editorCurveData {keys [{time=0}]}}
   volumeCurve {editorCurveData {keys [
      <at 0 1>
      {arriveTangent=-1, interpMode=RCIM_Cubic, leaveTangent=-1, tangentMode=RCTM_User, time=0.2, value=1.3403}
      <at 0.3311 0.5829>
      <at 1 0>
     ]}}
  }
 }
 {id=Modulator_1, children [Random_1], properties {pitchMax=1.2, pitchMin=1.4}}
 {id=Random_1, children [WavePlayer_6,WavePlayer_7,WavePlayer_8,WavePlayer_9], properties {weights [1,1,1,1]}}
 {id=WavePlayer_6, properties {soundWaveAssetPtr=/RocketDrillGameplay/Assets/Audio/waves/RocketDrill_Damage_Airborn_InDestructible_01}}
 {id=WavePlayer_7, properties {soundWaveAssetPtr=/RocketDrillGameplay/Assets/Audio/waves/RocketDrill_Damage_Airborn_InDestructible_02}}
 {id=WavePlayer_8, properties {soundWaveAssetPtr=/RocketDrillGameplay/Assets/Audio/waves/RocketDrill_Damage_Airborn_InDestructible_03}}
 {id=WavePlayer_9, properties {soundWaveAssetPtr=/RocketDrillGameplay/Assets/Audio/waves/RocketDrill_Damage_Airborn_InDestructible_04}}
]
properties {
 attenuationSettings=/RocketDrillGameplay/Assets/Audio/Settings/ATT_RocketDrill_InAir
 bOverrideConcurrency=true
 maxDistance=3740.0002
 soundClassObject=&Fort_Gliders_InAir
 volumeMultiplier=0.5
 concurrencyOverrides {bLimitToOwner=true, maxCount=2, resolutionRule=StopOldest, retriggerTime=1}
}
