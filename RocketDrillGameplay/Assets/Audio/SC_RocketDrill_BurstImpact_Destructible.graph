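# Sound cue for the Rocket Drill's burst impact on destructible objects, played as a melee weapon sound with in-air impact attenuation.
# A mixer layers a random drill airborne environment damage wave with a slightly delayed random RocketRam wall-break wave borrowed from the RocketRam feature.
# Each layer gets slight random pitch and volume modulation, and the whole cue is pitched down and quieted.
# A concurrency override caps simultaneous instances and limits rapid retriggering.
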SoundCue /RocketDrillGameplay/Assets/Audio/SC_RocketDrill_BurstImpact_Destructible
root=Mixer_0
nodes [
 {id=Delay_0, children [Modulator_3], properties {delayMax=0.05, delayMin=0.05}}
 {id=Mixer_0, children [Delay_0,Modulator_0], properties {inputVolume [1,1]}}
 {id=Modulator_0, children [Random_0], properties {pitchMax=1.1, pitchMin=0.9, volumeMax=1.1, volumeMin=0.9}}
 {id=Modulator_3, children [Random_4], properties {pitchMax=1.1, pitchMin=0.9, volumeMax=1.1, volumeMin=0.9}}
 {id=Random_0, children [WavePlayer_0,WavePlayer_1,WavePlayer_2,WavePlayer_3], properties {weights [1,1,1,1]}}
 {id=Random_4, children [WavePlayer_25, WavePlayer_26, WavePlayer_27, WavePlayer_28, WavePlayer_29], properties {weights [1,1,1,1,1]}}
 {id=WavePlayer_0, properties {soundWaveAssetPtr=/RocketDrillGameplay/Assets/Audio/waves/RocketDrill_Damage_Airborn_Enviornment_04}}
 {id=WavePlayer_1, properties {soundWaveAssetPtr=/RocketDrillGameplay/Assets/Audio/waves/RocketDrill_Damage_Airborn_Enviornment_03}}
 {id=WavePlayer_2, properties {soundWaveAssetPtr=/RocketDrillGameplay/Assets/Audio/waves/RocketDrill_Damage_Airborn_Enviornment_02}}
 {id=WavePlayer_25, properties {soundWaveAssetPtr=/RocketRamGameplay/Items/RocketRam/Sounds/RocketRam_Impact_BreakWall_Default_01}}
 {id=WavePlayer_26, properties {soundWaveAssetPtr=/RocketRamGameplay/Items/RocketRam/Sounds/RocketRam_Impact_BreakWall_Default_02}}
 {id=WavePlayer_27, properties {soundWaveAssetPtr=/RocketRamGameplay/Items/RocketRam/Sounds/RocketRam_Impact_BreakWall_Default_03}}
 {id=WavePlayer_28, properties {soundWaveAssetPtr=/RocketRamGameplay/Items/RocketRam/Sounds/RocketRam_Impact_BreakWall_Default_04}}
 {id=WavePlayer_29, properties {soundWaveAssetPtr=/RocketRamGameplay/Items/RocketRam/Sounds/RocketRam_Impact_BreakWall_Default_05}}
 {id=WavePlayer_3, properties {soundWaveAssetPtr=/RocketDrillGameplay/Assets/Audio/waves/RocketDrill_Damage_Airborn_Enviornment_01}}
]
properties {
 attenuationSettings=/RocketDrillGameplay/Assets/Audio/Settings/ATT_RocketDrill_InAir_Impact
 bOverrideConcurrency=true
 maxDistance=12000
 pitchMultiplier=0.8
 soundClassObject=&Fort_WeaponMelee
 volumeMultiplier=0.4
 concurrencyOverrides {maxCount=2, retriggerTime=0.15}
}
