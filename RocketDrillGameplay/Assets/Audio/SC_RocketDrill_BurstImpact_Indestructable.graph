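# Sound cue for the Rocket Drill burst impact on indestructible targets, using airborne damage sounds.
# A Random node picks one of four equally weighted wave variants, and a Modulator adds slight pitch and volume variation.
# It plays in the melee weapon sound class with an in-air impact attenuation asset.
# Overridden concurrency allows one instance per owner, blocking new plays and limiting rapid retriggers.
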SoundCue /RocketDrillGameplay/Assets/Audio/SC_RocketDrill_BurstImpact_Indestructable
root=Modulator_3
nodes [
 {id=Modulator_3, children [Random_0], properties {pitchMax=1.1, pitchMin=0.9, volumeMax=1.1, volumeMin=0.9}}
 {id=Random_0, children [WavePlayer_6,WavePlayer_7,WavePlayer_8,WavePlayer_9], properties {weights [1,1,1,1]}}
 {id=WavePlayer_6, properties {soundWaveAssetPtr=/RocketDrillGameplay/Assets/Audio/waves/RocketDrill_Damage_Airborn_InDestructible_01}}
 {id=WavePlayer_7, properties {soundWaveAssetPtr=/RocketDrillGameplay/Assets/Audio/waves/RocketDrill_Damage_Airborn_InDestructible_02}}
 {id=WavePlayer_8, properties {soundWaveAssetPtr=/RocketDrillGameplay/Assets/Audio/waves/RocketDrill_Damage_Airborn_InDestructible_03}}
 {id=WavePlayer_9, properties {soundWaveAssetPtr=/RocketDrillGameplay/Assets/Audio/waves/RocketDrill_Damage_Airborn_InDestructible_04}}
]
properties {
 attenuationSettings=/RocketDrillGameplay/Assets/Audio/Settings/ATT_RocketDrill_InAir_Impact
 bOverrideConcurrency=true
 maxDistance=12000
 soundClassObject=&Fort_WeaponMelee
 volumeMultiplier=1
 concurrencyOverrides {bLimitToOwner=true, maxCount=1, resolutionRule=PreventNew, retriggerTime=0.15}
}
