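# Sound cue for the Rocket Drill burst impact, player variant, routed to the melee weapon sound class.
# A mixer layers a randomly picked Rocket Ram player impact, at reduced volume, under a randomly picked Rocket Drill airborne damage wave.
# Each layer has its own slight random pitch and volume modulation.
# In-air impact attenuation shapes distance falloff, while a concurrency override caps simultaneous instances and sets a short retrigger time.
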SoundCue /RocketDrillGameplay/Assets/Audio/SC_RocketDrill_BurstImpact_Player
root=Mixer_0
nodes [
 {id=Mixer_0, children [Modulator_3,Modulator_0], properties {inputVolume [0.6,1]}}
 {id=Modulator_0, children [Random_0], properties {pitchMax=1.1, pitchMin=0.9, volumeMax=1.1, volumeMin=0.9}}
 {id=Modulator_3, children [Random_1], properties {pitchMax=1.1, pitchMin=0.9, volumeMax=1.1, volumeMin=0.9}}
 {id=Random_0, children [WavePlayer_0,WavePlayer_1,WavePlayer_2,WavePlayer_3], properties {weights [1,1,1,1]}}
 {
  id=Random_1
  children [WavePlayer_6, WavePlayer_7, WavePlayer_8, WavePlayer_9, WavePlayer_10, WavePlayer_11]
  properties {weights [1,1,1,1,1,1]}
 }
 {id=WavePlayer_0, properties {soundWaveAssetPtr=/RocketDrillGameplay/Assets/Audio/waves/RocketDrill_Damage_Airborn_Player_01}}
 {id=WavePlayer_1, properties {soundWaveAssetPtr=/RocketDrillGameplay/Assets/Audio/waves/RocketDrill_Damage_Airborn_Player_02}}
 {id=WavePlayer_10, properties {soundWaveAssetPtr=/RocketRamGameplay/Items/RocketRam/Sounds/RocketRam_Impact_Player_05}}
 {id=WavePlayer_11, properties {soundWaveAssetPtr=/RocketRamGameplay/Items/RocketRam/Sounds/RocketRam_Impact_Player_06}}
 {id=WavePlayer_2, properties {soundWaveAssetPtr=/RocketDrillGameplay/Assets/Audio/waves/RocketDrill_Damage_Airborn_Player_03}}
 {id=WavePlayer_3, properties {soundWaveAssetPtr=/RocketDrillGameplay/Assets/Audio/waves/RocketDrill_Damage_Airborn_Player_04}}
 {id=WavePlayer_6, properties {soundWaveAssetPtr=/RocketRamGameplay/Items/RocketRam/Sounds/RocketRam_Impact_Player_01}}
 {id=WavePlayer_7, properties {soundWaveAssetPtr=/RocketRamGameplay/Items/RocketRam/Sounds/RocketRam_Impact_Player_02}}
 {id=WavePlayer_8, properties {soundWaveAssetPtr=/RocketRamGameplay/Items/RocketRam/Sounds/RocketRam_Impact_Player_03}}
 {id=WavePlayer_9, properties {soundWaveAssetPtr=/RocketRamGameplay/Items/RocketRam/Sounds/RocketRam_Impact_Player_04}}
]
properties {
 attenuationSettings=/RocketDrillGameplay/Assets/Audio/Settings/ATT_RocketDrill_InAir_Impact
 bOverrideConcurrency=true
 maxDistance=12000
 soundClassObject=&Fort_WeaponMelee
 volumeMultiplier=1
 concurrencyOverrides {maxCount=2, retriggerTime=0.15}
}
