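# Sound cue for the Rocket Drill charge-up, layering a random launch spool-up with a random drill-start sound.
# An IsLocal branch routes the shared mix to a first-person sound class, or to a third-person class with attenuation.
# Quality-level nodes switch each layer between full and reduced variation pools.
# Concurrency is limited to one instance per owner, alongside shared weapon gunfire concurrency sets.
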SoundCue /RocketDrillGameplay/Assets/Audio/SC_RocketDrill_ChargeUp
root=Branch_0
nodes [
 {
  id=Attenuation_11
  children [Mixer_5]
  properties {attenuationSettings=/RocketDrillGameplay/Assets/Audio/Settings/ATT_RocketDrill_Engines_On_Off}
 }
 {id=Branch_0, children [SoundClass_1,SoundClass_6,Mixer_3], properties {boolParameterName=IsLocal}}
 {id=Mixer_3, children [Modulator_1,Modulator_2], properties {inputVolume [1,1]}}
 {id=Mixer_5, children [Mixer_3], properties {inputVolume [1]}}
 {id=Modulator_1, children [QualityLevel_0], properties {volumeMax=1, volumeMin=1}}
 {id=Modulator_2, children [QualityLevel_1], properties {volumeMax=1, volumeMin=1}}
 {id=QualityLevel_0, children [Random_5,Random_5,Random_6,Random_6]}
 {id=QualityLevel_1, children [Random_7,Random_7,Random_8,Random_8]}
 {id=Random_5, children [WavePlayer_39, WavePlayer_40, WavePlayer_41, WavePlayer_42, WavePlayer_43], properties {weights [1,1,1,1,1]}}
 {id=Random_6, children [WavePlayer_39, WavePlayer_41, WavePlayer_43], properties {bShouldExcludeFromBranchCulling=true, weights [1,1,1]}}
 {
  id=Random_7
  children [WavePlayer_19, WavePlayer_20, WavePlayer_21, WavePlayer_22, WavePlayer_23, WavePlayer_24]
  properties {weights [1,1,1,1,1,1]}
 }
 {id=Random_8, children [WavePlayer_19, WavePlayer_22, WavePlayer_24], properties {bShouldExcludeFromBranchCulling=true, weights [1,1,1]}}
 {id=SoundClass_1, children [Mixer_3], properties {soundClassOverride=&Fort_WeaponRanged_Fire_1P_PSM}}
 {
  id=SoundClass_6
  children [Attenuation_11]
  properties {soundClassOverride=&Fort_WeaponRanged_Fire_3P_PSM}
 }
 {id=WavePlayer_19, properties {soundWaveAssetPtr=/RocketDrillGameplay/Assets/Audio/waves/RocketDrill_DrillStart_01}}
 {id=WavePlayer_20, properties {soundWaveAssetPtr=/RocketDrillGameplay/Assets/Audio/waves/RocketDrill_DrillStart_02}}
 {id=WavePlayer_21, properties {soundWaveAssetPtr=/RocketDrillGameplay/Assets/Audio/waves/RocketDrill_DrillStart_03}}
 {id=WavePlayer_22, properties {soundWaveAssetPtr=/RocketDrillGameplay/Assets/Audio/waves/RocketDrill_DrillStart_04}}
 {id=WavePlayer_23, properties {soundWaveAssetPtr=/RocketDrillGameplay/Assets/Audio/waves/RocketDrill_DrillStart_05}}
 {id=WavePlayer_24, properties {soundWaveAssetPtr=/RocketDrillGameplay/Assets/Audio/waves/RocketDrill_DrillStart_06}}
 {id=WavePlayer_39, properties {soundWaveAssetPtr=/RocketDrillGameplay/Assets/Audio/waves/RocketDrill_Launch_SpoolUp_01}}
 {id=WavePlayer_40, properties {soundWaveAssetPtr=/RocketDrillGameplay/Assets/Audio/waves/RocketDrill_Launch_SpoolUp_02}}
 {id=WavePlayer_41, properties {soundWaveAssetPtr=/RocketDrillGameplay/Assets/Audio/waves/RocketDrill_Launch_SpoolUp_03}}
 {id=WavePlayer_42, properties {soundWaveAssetPtr=/RocketDrillGameplay/Assets/Audio/waves/RocketDrill_Launch_SpoolUp_04}}
 {id=WavePlayer_43, properties {soundWaveAssetPtr=/RocketDrillGameplay/Assets/Audio/waves/RocketDrill_Launch_SpoolUp_05}}
]
properties {
 maxDistance=28000
 volumeMultiplier=1.2
 concurrencyOverrides {bLimitToOwner=true, maxCount=1}
 concurrencySet [&Weapon_Gunfire_Global_Conc, &Weapon_Gunfire_Local_Conc]
}
