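# Sound cue for the RocketDrill charge-up blast, layering one of five random launch bursts with a quieter, briefly delayed air layer.
# The mix passes through a modulator, then branches on IsLocal into first-person or third-person ranged-fire sound classes, or plays without either when unset.
# Only the third-person path adds attenuation, borrowed from the drill's engine on/off settings.
# Concurrency combines gunfire groups with a single-instance, per-owner override.
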SoundCue /RocketDrillGameplay/Assets/Audio/SC_RocketDrill_ChargeUp_Blast
root=Branch_0
nodes [
 {
  id=Attenuation_11
  children [Mixer_1]
  properties {attenuationSettings=/RocketDrillGameplay/Assets/Audio/Settings/ATT_RocketDrill_Engines_On_Off}
 }
 {id=Branch_0, children [SoundClass_1, SoundClass_11, Modulator_2], properties {boolParameterName=IsLocal}}
 {id=Delay_0, children [WavePlayer_0], properties {delayMax=0.3, delayMin=0.3}}
 {id=Mixer_1, children [Modulator_2], properties {inputVolume [1]}}
 {id=Mixer_3, children [Delay_0,Random_5], properties {inputVolume [0.5,2]}}
 {id=Modulator_2, children [Mixer_3]}
 {id=Random_5, children [WavePlayer_27, WavePlayer_28, WavePlayer_30, WavePlayer_31, WavePlayer_32], properties {weights [1,1,1,1,1]}}
 {id=SoundClass_1, children [Modulator_2], properties {soundClassOverride=&Fort_WeaponRanged_Fire_1P_PSM}}
 {
  id=SoundClass_11
  children [Attenuation_11]
  properties {soundClassOverride=&Fort_WeaponRanged_Fire_3P_PSM}
 }
 {id=WavePlayer_0, properties {soundWaveAssetPtr=/RocketDrillGameplay/Assets/Audio/waves/RocketDrill_ChargeUp_Blast_Layer_Air_01}}
 {id=WavePlayer_27, properties {soundWaveAssetPtr=/RocketDrillGameplay/Assets/Audio/waves/RocketDrill_Launch_Burst_01}}
 {id=WavePlayer_28, properties {soundWaveAssetPtr=/RocketDrillGameplay/Assets/Audio/waves/RocketDrill_Launch_Burst_02}}
 {id=WavePlayer_30, properties {soundWaveAssetPtr=/RocketDrillGameplay/Assets/Audio/waves/RocketDrill_Launch_Burst_03}}
 {id=WavePlayer_31, properties {soundWaveAssetPtr=/RocketDrillGameplay/Assets/Audio/waves/RocketDrill_Launch_Burst_04}}
 {id=WavePlayer_32, properties {soundWaveAssetPtr=/RocketDrillGameplay/Assets/Audio/waves/RocketDrill_Launch_Burst_05}}
]
properties {
 maxDistance=28000
 volumeMultiplier=0.5
 concurrencyOverrides {bLimitToOwner=true, maxCount=1}
 concurrencySet [&Weapon_Gunfire_Global_Conc, &Weapon_Gunfire_Local_Conc]
}
