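# Sound cue for the Rocket Drill's drilling sound, in the melee weapon sound class with melee attenuation and capped concurrency.
# It mixes a randomly chosen drill loop, an oil-rig drill rotation loop and a one-shot drill start, each through a modulator.
# Envelopers fade in both loops and shape their pitch over the first half second.
# A quality-level node picks the start sound from either six or three random variations.
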SoundCue /RocketDrillGameplay/Assets/Audio/SC_RocketDrill_Drilling
root=Mixer_0
nodes [
 {
  id=Enveloper_0
  children [Modulator_1]
  properties {
   pitchCurve {editorCurveData {keys [<at 0 0.75>, {time=0.5}]}}
   volumeCurve {editorCurveData {keys [<at 0 0 RCIM_Cubic>, {arriveTangent=2.4785, interpMode=RCIM_Cubic, leaveTangent=2.4785, tangentMode=RCTM_User, time=0.5, value=1}]}}
  }
 }
 {
  id=Enveloper_4
  children [Modulator_0]
  properties {
   pitchCurve {editorCurveData {keys [<at 0 0.75>, {time=0.5}]}}
   volumeCurve {editorCurveData {keys [<at 0 0 RCIM_Cubic>, {arriveTangent=2.4785, interpMode=RCIM_Cubic, leaveTangent=2.4785, tangentMode=RCTM_User, time=0.5, value=1}]}}
  }
 }
 {id=Mixer_0, children [Enveloper_4,Enveloper_0,Modulator_3], properties {inputVolume [1,0.5,1.2]}}
 {id=Modulator_0, children [Random_0]}
 {id=Modulator_1, children [WavePlayer_5]}
 {id=Modulator_3, children [QualityLevel_0]}
 {id=QualityLevel_0, children [Random_1,Random_1,Random_2,Random_2]}
 {id=Random_0, children [WavePlayer_0,WavePlayer_1,WavePlayer_2], properties {weights [1,1,1]}}
 {id=Random_1, children [WavePlayer_3,WavePlayer_4,WavePlayer_6,WavePlayer_7,WavePlayer_8,WavePlayer_9], properties {weights [1,1,1,1,1,1]}}
 {id=Random_2, children [WavePlayer_3,WavePlayer_6,WavePlayer_9], properties {bShouldExcludeFromBranchCulling=true, weights [1,1,1]}}
 {id=WavePlayer_0, properties {bLooping=true, soundWaveAssetPtr=/RocketDrillGameplay/Assets/Audio/waves/RocketDrill_DrillLoop_03}}
 {id=WavePlayer_1, properties {bLooping=true, soundWaveAssetPtr=/RocketDrillGameplay/Assets/Audio/waves/RocketDrill_DrillLoop_02}}
 {id=WavePlayer_2, properties {bLooping=true, soundWaveAssetPtr=/RocketDrillGameplay/Assets/Audio/waves/RocketDrill_DrillLoop_01}}
 {id=WavePlayer_3, properties {soundWaveAssetPtr=/RocketDrillGameplay/Assets/Audio/waves/RocketDrill_DrillStart_01}}
 {id=WavePlayer_4, properties {soundWaveAssetPtr=/RocketDrillGameplay/Assets/Audio/waves/RocketDrill_DrillStart_02}}
 {
  id=WavePlayer_5
  properties {bLooping=true, soundWaveAssetPtr=/Game/Sounds/Helios/Environment_Actors/OilRigDrill/Spot_NitroOilRig_Pneumatic_Drill_Rotate_Loop_mono}
 }
 {id=WavePlayer_6, properties {soundWaveAssetPtr=/RocketDrillGameplay/Assets/Audio/waves/RocketDrill_DrillStart_03}}
 {id=WavePlayer_7, properties {soundWaveAssetPtr=/RocketDrillGameplay/Assets/Audio/waves/RocketDrill_DrillStart_04}}
 {id=WavePlayer_8, properties {soundWaveAssetPtr=/RocketDrillGameplay/Assets/Audio/waves/RocketDrill_DrillStart_05}}
 {id=WavePlayer_9, properties {soundWaveAssetPtr=/RocketDrillGameplay/Assets/Audio/waves/RocketDrill_DrillStart_06}}
]
properties {
 attenuationSettings=/RocketDrillGameplay/Assets/Audio/Settings/ATT_RocketDrill_Melee
 bOverrideConcurrency=true
 maxDistance=6400
 soundClassObject=&Fort_WeaponMelee
 volumeMultiplier=1
 attenuationOverrides {attenuationShapeExtents {x=0}}
 concurrencyOverrides {maxCount=3}
}
