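# Sound Cue for the impact of a dropped Rocket Drill landing.
# A Random node picks one of four equally weighted DropLand waves and feeds a Modulator left at its usual settings.
# Playback uses the shared weapon-drop attenuation and stop-oldest concurrency set, routed through the loot-collect sound class at reduced volume.
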SoundCue /RocketDrillGameplay/Assets/Audio/SC_RocketDrill_Drop_Impact
root=Modulator_1
nodes [
 {id=Modulator_1, children [Random_0]}
 {id=Random_0, children [WavePlayer_3,WavePlayer_4,WavePlayer_5,WavePlayer_6], properties {weights [1,1,1,1]}}
 {id=WavePlayer_3, properties {soundWaveAssetPtr=/RocketDrillGameplay/Assets/Audio/waves/RocketDrill_DropLand_01}}
 {id=WavePlayer_4, properties {soundWaveAssetPtr=/RocketDrillGameplay/Assets/Audio/waves/RocketDrill_DropLand_02}}
 {id=WavePlayer_5, properties {soundWaveAssetPtr=/RocketDrillGameplay/Assets/Audio/waves/RocketDrill_DropLand_03}}
 {id=WavePlayer_6, properties {soundWaveAssetPtr=/RocketDrillGameplay/Assets/Audio/waves/RocketDrill_DropLand_04}}
]
properties {
 attenuationSettings=&Fort_Weapon_Drop_Att
 maxDistance=3000
 soundClassObject=&Fort_Loot_Collect
 volumeMultiplier=0.65
 concurrencyOverrides {resolutionRule=PreventNew}
 concurrencySet [&Weapon_Drop_StopOldest]
}
