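# Sound cue for the Rocket Drill engine stop, playing the RocketDrill_PickUp_01 wave through a modulator node.
# It uses the in-air Rocket Drill attenuation and the weapon melee sound class.
# Overridden concurrency limits it to one instance per owner.
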SoundCue /RocketDrillGameplay/Assets/Audio/SC_RocketDrill_Engine_Stop
root=Modulator_0
nodes [
 {id=Modulator_0, children [WavePlayer_0]}
 {id=WavePlayer_0, properties {soundWaveAssetPtr=/RocketDrillGameplay/Assets/Audio/waves/RocketDrill_PickUp_01}}
]
properties {
 attenuationSettings=/RocketDrillGameplay/Assets/Audio/Settings/ATT_RocketDrill_InAir
 bOverrideConcurrency=true
 maxDistance=3740.0002
 soundClassObject=&Fort_WeaponMelee
 volumeMultiplier=1
 concurrencyOverrides {bLimitToOwner=true, maxCount=1}
}
