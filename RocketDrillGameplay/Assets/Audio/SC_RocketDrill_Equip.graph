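# Sound cue for equipping the Rocket Drill, played through the melee weapon sound class with generic player-equip attenuation.
# A short fixed delay feeds a modulator, then a quality-level switch selects one of two random pools.
# The first two quality levels draw from all six equip waves, while the last two use only the final three.
# A concurrency override caps simultaneous instances.
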SoundCue /RocketDrillGameplay/Assets/Audio/SC_RocketDrill_Equip
root=Delay_0
nodes [
 {id=Delay_0, children [Modulator_0], properties {delayMax=0.1, delayMin=0.1}}
 {id=Modulator_0, children [QualityLevel_0]}
 {id=QualityLevel_0, children [Random_0,Random_0,Random_1,Random_1]}
 {id=Random_0, children [WavePlayer_0,WavePlayer_1,WavePlayer_5,WavePlayer_6,WavePlayer_7,WavePlayer_8], properties {weights [1,1,1,1,1,1]}}
 {id=Random_1, children [WavePlayer_6,WavePlayer_7,WavePlayer_8], properties {bShouldExcludeFromBranchCulling=true, weights [1,1,1]}}
 {id=WavePlayer_0, properties {soundWaveAssetPtr=/RocketDrillGameplay/Assets/Audio/waves/RocketDrill_Equip_01}}
 {id=WavePlayer_1, properties {soundWaveAssetPtr=/RocketDrillGameplay/Assets/Audio/waves/RocketDrill_Equip_02}}
 {id=WavePlayer_5, properties {soundWaveAssetPtr=/RocketDrillGameplay/Assets/Audio/waves/RocketDrill_Equip_03}}
 {id=WavePlayer_6, properties {soundWaveAssetPtr=/RocketDrillGameplay/Assets/Audio/waves/RocketDrill_Equip_04}}
 {id=WavePlayer_7, properties {soundWaveAssetPtr=/RocketDrillGameplay/Assets/Audio/waves/RocketDrill_Equip_05}}
 {id=WavePlayer_8, properties {soundWaveAssetPtr=/RocketDrillGameplay/Assets/Audio/waves/RocketDrill_Equip_06}}
]
properties {
 attenuationSettings=/Game/Sounds/Attenuation/Fort_Player_Equip_Generic_Att
 bOverrideConcurrency=true
 maxDistance=1800
 pitchMultiplier=1.05
 soundClassObject=&Fort_WeaponMelee
 volumeMultiplier=0.4
 concurrencyOverrides {maxCount=4}
}
