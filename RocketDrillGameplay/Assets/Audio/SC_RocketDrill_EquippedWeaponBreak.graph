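# Sound cue for the rocket drill weapon breaking while equipped, in the RocketDrillGameplay plugin.
# A single wave player plays the shared 3D item durability break sound at reduced volume, using weapon reload attenuation and the weapon foley sound class.
# A concurrency override allows one instance at a time, stopping the oldest, and virtualization is disabled.
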SoundCue /RocketDrillGameplay/Assets/Audio/SC_RocketDrill_EquippedWeaponBreak
root=WavePlayer_0
nodes [{id=WavePlayer_0, properties {soundWaveAssetPtr=/Game/Sounds/Fort_UI_Sounds/Gameplay/Inventory/UI_Item_Durability_Break_01_3D}}]
properties {
 attenuationSettings=&Fort_Weapon_Reload_Att
 bOverrideConcurrency=true
 maxDistance=2250
 priority=0
 soundClassObject=&Fort_WeaponFoley
 virtualizationMode=Disabled
 volumeMultiplier=0.5
 concurrencyOverrides {maxCount=1, resolutionRule=StopOldest}
}
