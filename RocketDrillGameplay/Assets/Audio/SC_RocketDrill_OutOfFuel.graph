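# Sound cue for the rocket drill's out-of-fuel feedback, classed as melee weapon audio.
# An IsLocal branch silences it when false; when true or unset, it mixes a no-fuel sound, one of four random pops and a nested IsLocal branch.
# That branch adds a pitched-up pistol foley click and a UI "can't do" error ping.
# A concurrency override limits it to one instance.
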SoundCue /RocketDrillGameplay/Assets/Audio/SC_RocketDrill_OutOfFuel
root=Branch_1
nodes [
 {id=Branch_0, children [Mixer_2,null,Mixer_2], properties {boolParameterName=IsLocal}}
 {id=Branch_1, children [Mixer_0,null,Mixer_0], properties {boolParameterName=IsLocal}}
 {id=Mixer_0, children [Modulator_3,Modulator_2,Branch_0], properties {inputVolume [0.5,0.5,0.8]}}
 {id=Mixer_2, children [Modulator_1,WavePlayer_6], properties {inputVolume [1.25,0.5]}}
 {id=Modulator_1, children [WavePlayer_0], properties {pitchMax=1.3, pitchMin=1.3}}
 {id=Modulator_2, children [Random_0]}
 {id=Modulator_3, children [WavePlayer_1]}
 {id=Random_0, children [WavePlayer_2,WavePlayer_3,WavePlayer_4,WavePlayer_5], properties {weights [1,1,1,1]}}
 {id=WavePlayer_0, properties {soundWaveAssetPtr=/Game/Sounds/Fort_Weapons/pistol/pistol_foley/auto_sdk_end_01}}
 {id=WavePlayer_1, properties {soundWaveAssetPtr=/RocketDrillGameplay/Assets/Audio/waves/RocketDrill_NoFuel}}
 {id=WavePlayer_2, properties {soundWaveAssetPtr=/RocketDrillGameplay/Assets/Audio/waves/RocketDrill_NoFuel_Pop_01}}
 {id=WavePlayer_3, properties {soundWaveAssetPtr=/RocketDrillGameplay/Assets/Audio/waves/RocketDrill_NoFuel_Pop_02}}
 {id=WavePlayer_4, properties {soundWaveAssetPtr=/RocketDrillGameplay/Assets/Audio/waves/RocketDrill_NoFuel_Pop_03}}
 {id=WavePlayer_5, properties {soundWaveAssetPtr=/RocketDrillGameplay/Assets/Audio/waves/RocketDrill_NoFuel_Pop_04}}
 {id=WavePlayer_6, properties {soundWaveAssetPtr=/Game/Sounds/Fort_UI_Sounds/Gameplay/UI_Error_Ping_Cant_Do_Default_01}}
]
properties {
 bOverrideConcurrency=true
 maxDistance=2097152
 soundClassObject=&Fort_WeaponMelee
 volumeMultiplier=0.4
 concurrencyOverrides {maxCount=1}
}
