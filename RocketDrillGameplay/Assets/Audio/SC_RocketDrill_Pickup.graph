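# SC_RocketDrill_Pickup is the Sound Cue for picking up the Rocket Drill.
# A Random node picks one of three equally weighted pickup waves and feeds it through a Modulator at default settings.
# It uses weapon pickup attenuation and the loot collect sound class at reduced volume.
# A concurrency override allows only one instance at a time and blocks new plays while one is active.
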SoundCue /RocketDrillGameplay/Assets/Audio/SC_RocketDrill_Pickup
root=Modulator_1
nodes [
 {id=Modulator_1, children [Random_1]}
 {id=Random_1, children [WavePlayer_0,WavePlayer_1,WavePlayer_3], properties {weights [1,1,1]}}
 {id=WavePlayer_0, properties {soundWaveAssetPtr=/RocketDrillGameplay/Assets/Audio/waves/RocketDrill_PickUp_01}}
 {id=WavePlayer_1, properties {soundWaveAssetPtr=/RocketDrillGameplay/Assets/Audio/waves/RocketDrill_PickUp_02}}
 {id=WavePlayer_3, properties {soundWaveAssetPtr=/RocketDrillGameplay/Assets/Audio/waves/RocketDrill_PickUp_03}}
]
properties {
 attenuationSettings=&Fort_Weapon_Pickup_Att
 bOverrideConcurrency=true
 maxDistance=3500
 soundClassObject=&Fort_Loot_Collect
 volumeMultiplier=0.65
 concurrencyOverrides {maxCount=1, resolutionRule=PreventNew}
}
