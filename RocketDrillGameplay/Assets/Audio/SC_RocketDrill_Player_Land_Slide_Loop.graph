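# Sound cue for the Rocket Drill's player land-slide loop, classed as melee weapon audio.
# A Modulator feeds a quality-level switch: the first two levels randomly pick from four slide loop waves, the last two from just two of them.
# The smaller random node is excluded from branch culling.
# It borrows the drill's engine-loop attenuation and overrides concurrency to allow one instance per owner.
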SoundCue /RocketDrillGameplay/Assets/Audio/SC_RocketDrill_Player_Land_Slide_Loop
root=Modulator_0
nodes [
 {id=Modulator_0, children [QualityLevel_0]}
 {id=QualityLevel_0, children [Random_0,Random_0,Random_1,Random_1]}
 {id=Random_0, children [WavePlayer_9, WavePlayer_10, WavePlayer_11, WavePlayer_12], properties {weights [1,1,1,1]}}
 {id=Random_1, children [WavePlayer_9, WavePlayer_12], properties {bShouldExcludeFromBranchCulling=true, weights [1,1]}}
 {id=WavePlayer_10, properties {soundWaveAssetPtr=/RocketDrillGameplay/Assets/Audio/waves/RocketDrill_LandSlide_Loop_03}}
 {id=WavePlayer_11, properties {soundWaveAssetPtr=/RocketDrillGameplay/Assets/Audio/waves/RocketDrill_LandSlide_Loop_02}}
 {id=WavePlayer_12, properties {soundWaveAssetPtr=/RocketDrillGameplay/Assets/Audio/waves/RocketDrill_LandSlide_Loop_01}}
 {id=WavePlayer_9, properties {soundWaveAssetPtr=/RocketDrillGameplay/Assets/Audio/waves/RocketDrill_LandSlide_Loop_04}}
]
properties {
 attenuationSettings=/RocketDrillGameplay/Assets/Audio/Settings/ATT_RocketDrill_Engine_Loop
 bOverrideConcurrency=true
 maxDistance=1500
 soundClassObject=&Fort_WeaponMelee
 volumeMultiplier=1
 concurrencyOverrides {bLimitToOwner=true, maxCount=1}
}
