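# Sound cue for the player's RocketDrill land-slide stop, layering a slide-loop stop sound with a landing impact.
# A Mixer blends two Modulator branches: one picks randomly from four slide-stop waves, the other picks landing waves through a Quality Level node.
# The first two quality slots use six landing variants, while the last two use a three-wave subset.
# It reuses the drill engine-loop attenuation and limits concurrency to one instance per owner.
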SoundCue /RocketDrillGameplay/Assets/Audio/SC_RocketDrill_Player_Land_Slide_Stop
root=Mixer_0
nodes [
 {id=Mixer_0, children [Modulator_1,Modulator_0], properties {inputVolume [1,1]}}
 {id=Modulator_0, children [Random_0]}
 {id=Modulator_1, children [QualityLevel_0]}
 {id=QualityLevel_0, children [Random_1,Random_1,Random_2,Random_2]}
 {id=Random_0, children [WavePlayer_0,WavePlayer_1,WavePlayer_2,WavePlayer_9], properties {weights [1,1,1,1]}}
 {id=Random_1, children [WavePlayer_3,WavePlayer_4,WavePlayer_5,WavePlayer_6,WavePlayer_7,WavePlayer_8], properties {weights [1,1,1,1,1,1]}}
 {id=Random_2, children [WavePlayer_3,WavePlayer_6,WavePlayer_8], properties {bShouldExcludeFromBranchCulling=true, weights [1,1,1]}}
 {id=WavePlayer_0, properties {soundWaveAssetPtr=/RocketDrillGameplay/Assets/Audio/waves/RocketDrill_LandSlide_Loop_Stop_01}}
 {id=WavePlayer_1, properties {soundWaveAssetPtr=/RocketDrillGameplay/Assets/Audio/waves/RocketDrill_LandSlide_Loop_Stop_02}}
 {id=WavePlayer_2, properties {soundWaveAssetPtr=/RocketDrillGameplay/Assets/Audio/waves/RocketDrill_LandSlide_Loop_Stop_03}}
 {id=WavePlayer_3, properties {soundWaveAssetPtr=/RocketDrillGameplay/Assets/Audio/waves/RocketDrill_Land_06}}
 {id=WavePlayer_4, properties {soundWaveAssetPtr=/RocketDrillGameplay/Assets/Audio/waves/RocketDrill_Land_05}}
 {id=WavePlayer_5, properties {soundWaveAssetPtr=/RocketDrillGameplay/Assets/Audio/waves/RocketDrill_Land_04}}
 {id=WavePlayer_6, properties {soundWaveAssetPtr=/RocketDrillGameplay/Assets/Audio/waves/RocketDrill_Land_03}}
 {id=WavePlayer_7, properties {soundWaveAssetPtr=/RocketDrillGameplay/Assets/Audio/waves/RocketDrill_Land_02}}
 {id=WavePlayer_8, properties {soundWaveAssetPtr=/RocketDrillGameplay/Assets/Audio/waves/RocketDrill_Land_01}}
 {id=WavePlayer_9, properties {soundWaveAssetPtr=/RocketDrillGameplay/Assets/Audio/waves/RocketDrill_LandSlide_Loop_Stop_04}}
]
properties {
 attenuationSettings=/RocketDrillGameplay/Assets/Audio/Settings/ATT_RocketDrill_Engine_Loop
 bOverrideConcurrency=true
 maxDistance=1500
 soundClassObject=&Fort_WeaponMelee
 volumeMultiplier=1
 concurrencyOverrides {bLimitToOwner=true, maxCount=1}
}
