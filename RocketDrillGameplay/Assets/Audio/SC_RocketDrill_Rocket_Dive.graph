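# Sound cue for the rocket drill's dive, built from in-air engine-off waves.
# A shared mix layers a random Off wave with a random, briefly delayed Off Burst wave, each through a modulator.
# An IsLocal branch routes it to the first-person fire sound class, or to the third-person class with attenuation and lower volume; unset plays the plain mix.
# Overridden concurrency allows one instance per owner.
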SoundCue /RocketDrillGameplay/Assets/Audio/SC_RocketDrill_Rocket_Dive
root=Branch_0
nodes [
 {
  id=Attenuation_11
  children [Mixer_1]
  properties {attenuationSettings=/RocketDrillGameplay/Assets/Audio/Settings/ATT_RocketDrill_Engines_On_Off}
 }
 {id=Branch_0, children [SoundClass_1, SoundClass_11, Mixer_0], properties {boolParameterName=IsLocal}}
 {id=Delay_0, children [Modulator_3], properties {delayMax=0.15, delayMin=0.15}}
 {id=Mixer_0, children [Modulator_2,Delay_0], properties {inputVolume [2,1.3]}}
 {id=Mixer_1, children [Mixer_0], properties {inputVolume [0.5]}}
 {id=Modulator_2, children [Random_4]}
 {id=Modulator_3, children [Random_5]}
 {
  id=Random_4
  children [WavePlayer_13, WavePlayer_14, WavePlayer_15, WavePlayer_16, WavePlayer_17, WavePlayer_18]
  properties {weights [1,1,1,1,1,1]}
 }
 {id=Random_5, children [WavePlayer_19, WavePlayer_20, WavePlayer_21], properties {weights [1,1,1]}}
 {id=SoundClass_1, children [Mixer_0], properties {soundClassOverride=&Fort_WeaponRanged_Fire_1P_PSM}}
 {
  id=SoundClass_11
  children [Attenuation_11]
  properties {soundClassOverride=&Fort_WeaponRanged_Fire_3P_PSM}
 }
 {id=WavePlayer_13, properties {soundWaveAssetPtr=/RocketDrillGameplay/Assets/Audio/waves/RocketDrill_InAir_Off_01}}
 {id=WavePlayer_14, properties {soundWaveAssetPtr=/RocketDrillGameplay/Assets/Audio/waves/RocketDrill_InAir_Off_02}}
 {id=WavePlayer_15, properties {soundWaveAssetPtr=/RocketDrillGameplay/Assets/Audio/waves/RocketDrill_InAir_Off_03}}
 {id=WavePlayer_16, properties {soundWaveAssetPtr=/RocketDrillGameplay/Assets/Audio/waves/RocketDrill_InAir_Off_04}}
 {id=WavePlayer_17, properties {soundWaveAssetPtr=/RocketDrillGameplay/Assets/Audio/waves/RocketDrill_InAir_Off_05}}
 {id=WavePlayer_18, properties {soundWaveAssetPtr=/RocketDrillGameplay/Assets/Audio/waves/RocketDrill_InAir_Off_06}}
 {id=WavePlayer_19, properties {soundWaveAssetPtr=/RocketDrillGameplay/Assets/Audio/waves/RocketDrill_InAir_Off_Burst_01}}
 {id=WavePlayer_20, properties {soundWaveAssetPtr=/RocketDrillGameplay/Assets/Audio/waves/RocketDrill_InAir_Off_Burst_02}}
 {id=WavePlayer_21, properties {soundWaveAssetPtr=/RocketDrillGameplay/Assets/Audio/waves/RocketDrill_InAir_Off_Burst_03}}
]
properties {bOverrideConcurrency=true, maxDistance=28000, volumeMultiplier=1.5, concurrencyOverrides {bLimitToOwner=true, maxCount=1}}
