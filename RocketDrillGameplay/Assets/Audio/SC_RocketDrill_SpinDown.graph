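# Sound cue for the rocket drill spinning down, using melee weapon class and attenuation with gunfire concurrency limits.
# The main layer plays a random drill-stop wave pitched down, choosing from six variants at the first two quality levels and three at the last two.
# It is mixed with a very quiet, slightly delayed, pitched-up spin-down wave borrowed from the Goin' Commando weapon.
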SoundCue /RocketDrillGameplay/Assets/Audio/SC_RocketDrill_SpinDown
root=Mixer_0
nodes [
 {id=Delay_0, children [Modulator_3], properties {delayMax=0.3, delayMin=0.3}}
 {id=Mixer_0, children [Modulator_0,Delay_0], properties {inputVolume [1.1,0.05]}}
 {id=Modulator_0, children [QualityLevel_0], properties {pitchMax=0.8, pitchMin=0.8}}
 {id=Modulator_3, children [WavePlayer_3], properties {pitchMax=1.25, pitchMin=1.25}}
 {id=QualityLevel_0, children [Random_1,Random_1,Random_2,Random_2]}
 {id=Random_1, children [WavePlayer_0,WavePlayer_1,WavePlayer_2,WavePlayer_4,WavePlayer_5,WavePlayer_6], properties {weights [1,1,1,1,1,1]}}
 {id=Random_2, children [WavePlayer_0,WavePlayer_4,WavePlayer_6], properties {bShouldExcludeFromBranchCulling=true, weights [1,1,1]}}
 {id=WavePlayer_0, properties {soundWaveAssetPtr=/RocketDrillGameplay/Assets/Audio/waves/RocketDrill_DrillStop_01}}
 {id=WavePlayer_1, properties {soundWaveAssetPtr=/RocketDrillGameplay/Assets/Audio/waves/RocketDrill_DrillStop_02}}
 {id=WavePlayer_2, properties {soundWaveAssetPtr=/RocketDrillGameplay/Assets/Audio/waves/RocketDrill_DrillStop_03}}
 {id=WavePlayer_3, properties {soundWaveAssetPtr=/Game/Sounds/Fort_Weapons/AutoRifle/WPN_GoinCommando_SpinDown}}
 {id=WavePlayer_4, properties {soundWaveAssetPtr=/RocketDrillGameplay/Assets/Audio/waves/RocketDrill_DrillStop_04}}
 {id=WavePlayer_5, properties {soundWaveAssetPtr=/RocketDrillGameplay/Assets/Audio/waves/RocketDrill_DrillStop_05}}
 {id=WavePlayer_6, properties {soundWaveAssetPtr=/RocketDrillGameplay/Assets/Audio/waves/RocketDrill_DrillStop_06}}
]
properties {
 attenuationSettings=/RocketDrillGameplay/Assets/Audio/Settings/ATT_RocketDrill_Melee
 bDebug=true
 maxDistance=6400
 soundClassObject=&Fort_WeaponMelee
 volumeMultiplier=1
 concurrencySet [&Weapon_Gunfire_Local_Conc, &Weapon_Gunfire_Global_Conc]
}
