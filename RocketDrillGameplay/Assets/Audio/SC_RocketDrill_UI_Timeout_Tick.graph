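# Sound cue for the Rocket Drill's timeout tick, named as a UI sound and built on a temporary fuel countdown wave.
# A single wave player feeds an Enveloper that holds full volume briefly, then fades to silence by 0.2 seconds, trimming the wave into a short tick.
# It plays at reduced volume in the gadgets sound class and uses the jetpack attenuation settings.
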SoundCue /RocketDrillGameplay/Assets/Audio/SC_RocketDrill_UI_Timeout_Tick
root=Enveloper_0
nodes [
 {
  id=Enveloper_0
  children [WavePlayer_0]
  properties {pitchCurve {editorCurveData {keys [{time=0}]}}, volumeCurve {editorCurveData {keys [<at 0 1>,<at 0.1 1>,<at 0.2 0>]}}}
 }
 {id=WavePlayer_0, properties {soundWaveAssetPtr=/RocketDrillGameplay/Assets/Audio/waves/Temp/BarnTowel_Fuel_CountDown}}
]
properties {
 attenuationSettings=/Game/Sounds/Attenuation/Fort_Jetpack_Att
 maxDistance=13067
 soundClassObject=&Fort_Gadgets
 volumeMultiplier=0.5
}
