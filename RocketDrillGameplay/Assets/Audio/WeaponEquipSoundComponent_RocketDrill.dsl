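; Rocket drill weapon equip sound component that starts and stops the weapon's looping audio through the Sound Library.
; UpdateLoopStatus plays the loop event and stores the result when Play is true.
; Otherwise, if the loop is active, it plays the looping-stop event, snaps the resulting audio components onto the player pawn's root, then stops the loop.
; BeginPlay and Tick add no logic.
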(event EventBeginPlay
  (|CallParentFunction) ; This node is disabled and will not be called.\nDrag off pins to build functionality.
)

(event EventTick
  (params (DeltaSeconds :type Float%20%28single-precision%29)))

(fn UpdateLoopStatus
  (params (Play :type Boolean))
  (if Play
    (bind _result (CallFunction|PlayValidEvent (Variables|Event|GetLoopEventName)))
    (Variables|Default|SetSoundLibraryEvent _result)
    (else
      (if (SoundLibrary|IsEventActive (Variables|Private|GetSoundLibraryComponent) (Variables|Event|GetLoopEventName))
        (bind _result_2 (CallFunction|PlayValidEvent (Variables|Event|GetLoopingStopEventName)))
        (for _array_element (.AudioComponents (Utilities|Struct|BreakSoundLibraryPlayResult _result_2))
          (Transformation|AttachComponentToComponent _array_element (Variables|/Script/Engine.Actor|GetRootComponent (Variables|Private|GetPlayerPawn)) "None" "SnapToTarget" "SnapToTarget" "SnapToTarget" true))
        (SoundLibrary|StopEvent (Variables|Private|GetSoundLibraryComponent) (Variables|Event|GetLoopEventName))))))
