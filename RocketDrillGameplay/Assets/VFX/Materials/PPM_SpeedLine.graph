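# Post-process material that draws animated pale grey speed lines over the screen, strongest toward the edges.
# It converts screen UVs to polar coordinates around the center and pans two samples of a packed noise texture at different angular tilings and speeds.
# The maximum of both samples, scaled by a cubed radial falloff, forms a mask that blends the tinted streaks over the scene color.
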Material /RocketDrillGameplay/Assets/VFX/Materials/PPM_SpeedLine
graph {
 properties {bAutomaticallySetUsageInEditor=false, materialDomain=MD_PostProcess, usedShadingModels=Default Lit}
 nodes [
  n1 = Subtract(A=TextureCoordinate()) {constB=0.5}
  n2 = Length(None=@n1) {is=class MaterialExpressionLength}
  n3 = Multiply(A=@n2) {constB=0.6}
  n4 = ComponentMask(None=@n1) {g=true}
  n5 = ComponentMask(None=@n1) {r=true}
  n6 = Frac(None=Divide(A=Arctangent2!(Y=@n4, X=@n5)) {constB=6.48})
  n7 = Multiply(A=@n6) {constB=6}
  n8 = TextureSample(UVs=Panner(Coordinate=AppendVector(A=@n3, B=@n7)) {speedX=-0.5, speedY=0.7}) {samplerType=SAMPLERTYPE_Masks, texture=/Game/Effects/Fort_Effects/Textures/T_PackedNoise}
  n9 = Multiply(A=@n6) {constB=12}
  n10 = TextureSample(UVs=Panner(Coordinate=AppendVector(A=@n3, B=@n9)) {speedX=-1, speedY=0.5}) {samplerType=SAMPLERTYPE_Masks, texture=&T_PackedNoise}
  n11 = Multiply(A=Multiply(A=Power(Base=@n2) {constExponent=3}, B=Max(A=@n8.B, B=@n10.G)))
  LinearInterpolate(
   A=Multiply(A=Constant4Vector() {constant=<rgba 0.4691 0.4832 0.5 1>}.RGBA, B=@n11)
   B=SceneTexture() {sceneTextureId=PPI_PostProcessInput0}.Color
   Alpha=Saturate(None=OneMinus(None=@n11))
  )
 ]
}
