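# Niagara burst effect for the rocket drill boost: air, wind and orange flare visuals that follow a skeletal mesh.
# Air_Pressure and Wind burst short-lived, curve-scaled meshes, with Air_Pressure alpha also scaled by time of day.
# Combined_Burst_Once packs wind cone, air glow, corona, streak and flare into per-renderer attributes on shared particles.
# Its two spawn groups set VisibilityTag, which selects either a fixed lifetime or User.uDuration.
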NiagaraSystem /RocketDrillGameplay/Assets/VFX/NS_RocketDrill_Boost_Burst
dependencies {
 usedDataInterfaces [&NiagaraDataInterfaceCurve, &NiagaraDataInterfaceColorCurve, &NiagaraDataInterfaceSkeletalMesh]
 usedDynamicInputs [&ColorFromCurve, &RemapRange]
 usedModules [
  &self:SystemSpawnScript.SystemScriptSource.SystemScriptGraph.NiagaraNodeAssignment_4.SetVariables_#1
  &SystemState
  &self:SystemSpawnScript.SystemScriptSource.SystemScriptGraph.NiagaraNodeAssignment_2.SetVariables_#2
  &self:SystemSpawnScript.SystemScriptSource.SystemScriptGraph.NiagaraNodeAssignment_3.SetVariables_#3
 ]
}
emitters {
 Air_Pressure {
  emitterUpdateScript [
   &EmitterState()
   &SpawnBurst_Instantaneous(Spawn Count=5, Use Spawn Probability=false, UseLoopCountLimit=false)
  ]
  particleSpawnScript [
   &InitializeParticle(
    Color Maximum=<rgba 0.2292 0.212 0.1946 0.25>
    Color Minimum=<rgba 0.1354 0.1253 0.115 0.15>
    Color Mode=::NewEnumerator2
    Lifetime Max=0.4
    Lifetime Min=0.25
    Lifetime Mode=::NewEnumerator1
    Mesh Scale=&RandomRangeVector(Minimum=<1.8 1.8 2>, Maximum=<2 2 2>)
    Mesh Scale Mode=::NewEnumerator4
    Position Offset Coordinate Space=::World
    UsePositionOffset=true
    Write Material Random=true
   )
   &InitialMeshOrientation(
    Mesh Axis To Orient=<0 -1 0>
    Mesh Orientation Mode=::NewEnumerator2
    Orientation Coordinate Space=::Local
    Orientation Vector=<0 0 1>
    Rotation=&Lerp_Vector(A=<0 0 -0.25>, B=<0 0 0.25>, Alpha=&ReturnNormalizedExecIndex())
    Use Rotation Vector=true
   )
  ]
  particleUpdateScript [
   &ParticleState()
   &SkeletalMeshLocation(
    Bone Sampling Mode=::NewEnumerator4
    MaskSpawnGroup=false
    Orientation Sampling=::NewEnumerator1
    Sampled Position Offset=<-75 0 0>
    Skeletal Mesh=$System.SkeletalMesh
   )
   &ScaleMeshSize(Scale Factor=&MakeVector(
     X=&FloatFromCurve(FloatCurve=DataInterface() {propertyValues {Curve {keys [
          {arriveTangent=4.1785, interpMode=RCIM_Cubic, leaveTangent=4.1785, tangentMode=RCTM_User, time=0, value=0}
          {arriveTangent=1.1179, interpMode=RCIM_Cubic, leaveTangent=1.1179, tangentMode=RCTM_User, time=0.2042, value=0.6228}
          {arriveTangent=0.0224, interpMode=RCIM_Cubic, leaveTangent=0.0224, tangentMode=RCTM_User, time=1, value=1}
         ]}}})
     Y=&FloatFromCurve(FloatCurve=DataInterface() {propertyValues {Curve {keys [
          {arriveTangent=4.1785, interpMode=RCIM_Cubic, leaveTangent=4.1785, tangentMode=RCTM_User, time=0, value=0}
          {arriveTangent=1.1179, interpMode=RCIM_Cubic, leaveTangent=1.1179, tangentMode=RCTM_User, time=0.2042, value=0.6228}
          {arriveTangent=0.0224, interpMode=RCIM_Cubic, leaveTangent=0.0224, tangentMode=RCTM_User, time=1, value=1}
         ]}}})
     Z=1
    ))
   &ScaleColor(
    Linear Color Curve=DataInterface() {propertyValues {
      AlphaCurve {keys [<at 0 1>,<at 0.4544 0.5>,<at 0.9945 0>]}
      BlueCurve {keys [<at 0 1>,<at 1 1>]}
      GreenCurve {keys [<at 0 1>,<at 1 1>]}
      RedCurve {keys [<at 0 1>,<at 1 1>]}
     }}
    Scale Mode=::NewEnumerator2
   )
   &SolveForcesAndVelocity(Clamp Velocity=false, Limit Acceleration=false)
   &ScaleColor(
    Scale Alpha=&FNTimeOfDayScalar(Night Scale=1, Morning Scale=1.2, Day Scale=1.5, Evening Scale=1.2)
    ScaleA=true
    ScaleRGB=false
   ) {module=ScaleColor001}
   &ScaleColor(
    Scale Alpha=&FloatFromCurve(
     FloatCurve=DataInterface() {propertyValues {Curve {keys [
         {arriveTangent=4.1785, interpMode=RCIM_Cubic, leaveTangent=4.1785, tangentMode=RCTM_User, time=0, value=0}
         {arriveTangent=1.1179, interpMode=RCIM_Cubic, leaveTangent=1.1179, tangentMode=RCTM_User, time=0.2042, value=0.6228}
         {arriveTangent=0.0224, interpMode=RCIM_Cubic, leaveTangent=0.0224, tangentMode=RCTM_User, time=1, value=1}
        ]}}}
     CurveIndex=$Emitter.NormalizedLoopAge
     Scale Curve=2
    )
    ScaleA=true
    ScaleRGB=false
   ) {module=ScaleColor002}
  ]
  rendererClasses [&NiagaraMeshRendererProperties]
  renderers [{rendererClass=&NiagaraMeshRendererProperties, rendererIndex=0}]
 }
 Combined_Burst_Once {
  emitterUpdateScript [
   &EmitterState()
   &SpawnBurst_Instantaneous(Use Spawn Probability=false, UseLoopCountLimit=false)
   &SpawnBurst_Instantaneous(Spawn Group=1, Use Spawn Probability=false, UseLoopCountLimit=false) {module=SpawnBurst_Instantaneous001}
  ]
  particleSpawnScript [
   /SetVariables(Particles.VisibilityTag=$Emitter.SpawnGroup) {isSetParameters=true}
   &InitializeParticle(
    Color Mode=::NewEnumerator0
    Lifetime=&MergeHelper_FloatSwitch(0=2.5, 1=$User.uDuration, 2=?, 3=?, Index=$Particles.VisibilityTag, Num=::NewEnumerator0)
    Position Mode=::NewEnumerator0
    Write Material Random=true
   )
   /SetVariables(
    Particles.Color_WindCone:LinearColor=<rgba 0.4583 0.4032 0.3558 0.75>
    Particles.DynamicParameter_WindCone:Vector4f=<0.7 2.5 0 0>
    Particles.Orientation_WindCone:Quat4f=&OrientMeshToVector(Look At Direction=<0 1 0>, Reference Axis Vector=<1 0 0>)
    Particles.Size_WindCone:Vector3f=<0.5 0.5 0.5>
   ) {isSetParameters=true}
   /SetVariables(
    Particles.Color_AirGlow:LinearColor=<rgba 0.125 0.125 0.125 1>
    Particles.DynamicParameter_AirGlow:Vector4f=<1 1 0 1>
    Particles.Orientation_AirGlow:Quat4f=&OrientMeshToVector(Look At Direction=<0 1 0>, Reference Axis Vector=<1 0 0>)
    Particles.Size_AirGlow:Vector3f=&Multiply_VectorByFloat(Vector=<7.5 2.25 2.25>, Float=0.6)
   ) {isSetParameters=true}
   /SetVariables(Particles.Color_GlowCorona:LinearColor=<rgba 1 0.1667 0 1>, Particles.Size_GlowCorona:Vector2f=<315 210>) {isSetParameters=true}
   /SetVariables(Particles.Streak_01_Color:LinearColor=<rgba 2.5 0.1667 0 1>, Particles.Streak_01_Size:Vector2f=<350 25>) {isSetParameters=true}
   /SetVariables(
    Particles.Color_FlareIncrease:LinearColor=<rgba 1 0.125 0 1>
    Particles.DynamicParameter_FlareIncrease:Vector4f=<0.5 1 0 0>
    Particles.DynamicParameter_FlareIncrease_01:Vector4f=<5 0 0 0>
    Particles.Size_FlareIncrease:Vector2f=<35 35>
   ) {isSetParameters=true}
   /SetVariables(Particles.Sprite_CameraOffset=25) {isSetParameters=true}
  ]
  particleUpdateScript [
   &ParticleState()
   &SkeletalMeshLocation(
    Bone Sampling Mode=::NewEnumerator4
    MaskSpawnGroup=false
    Orientation Sampling=::NewEnumerator1
    Sampled Position Offset=<-5 0 0>
    Skeletal Mesh=$System.SkeletalMesh
   )
   /SetVariables(Particles.Color_WindCone:LinearColor=&MultiplyLinearColor(
     Linear Color=&LerpLinearColors(
      StartColor=<rgba 0.4583 0.4032 0.3558 1>
      EndColor=<rgba 0.4583 0.4032 0.3558 1>
      LerpFactor=&FloatFromCurve(
       FloatCurve=DataInterface() {propertyValues {Curve {keys [
           {interpMode=RCIM_Cubic, tangentMode=RCTM_User, time=0, value=0}
           {interpMode=RCIM_Cubic, tangentMode=RCTM_User, time=0.5, value=1}
           {tangentMode=RCTM_User, time=1, value=0}
          ]}}}
       CurveIndex=&Multiply_Float(A=$Particles.NormalizedAge, B=2.5)
      )
     )
     Scale Factor=&MakeFloat4FromVectorAndFloat(
      "Vector (RGB)"=<1 1 1>
      "Float (Alpha)"=&FloatFromCurve(
       FloatCurve=DataInterface() {propertyValues {Curve {keys [
           {arriveTangent=31.8867, interpMode=RCIM_Cubic, leaveTangent=31.8867, tangentMode=RCTM_User, time=0, value=0}
           {arriveTangent=-2.5249, interpMode=RCIM_Cubic, leaveTangent=-2.5249, tangentMode=RCTM_User, time=0.091, value=0.9714}
           {arriveTangent=-0.7632, interpMode=RCIM_Cubic, leaveTangent=-0.7632, tangentMode=RCTM_User, time=0.1741, value=0.2056}
           {arriveTangent=-0.0405, interpMode=RCIM_Cubic, leaveTangent=-0.0405, tangentMode=RCTM_User, time=1, value=0}
          ]}}}
       CurveIndex=&Multiply_Float(A=$Particles.NormalizedAge, B=2.5)
      )
     )
    )) {isSetParameters=true}
   /SetVariables(
    Particles.Color_GlowCorona:LinearColor=&MultiplyLinearColor(
     Linear Color=$Particles.Initial.Color_GlowCorona
     Scale Factor=&MakeFloat4FromVectorAndFloat(
      "Vector (RGB)"=&VectorFromFloat(Value=&FloatFromCurve(
        FloatCurve=DataInterface() {propertyValues {Curve {keys [
            {arriveTangent=47.5885, interpMode=RCIM_Cubic, time=0, value=0}
            {arriveTangent=3.5714, interpMode=RCIM_Cubic, leaveTangent=3.5714, time=0.1, value=5}
            <at 0.7 2.5>
            <at 1 0 RCIM_Cubic>
           ]}}}
        CurveIndex=&Multiply_Float(A=$Particles.NormalizedAge, B=9.6)
        Scale Curve=2
       ))
      "Float (Alpha)"=&FloatFromCurve(
       FloatCurve=DataInterface() {propertyValues {Curve {keys [
           <at 0 1 RCIM_Cubic>
           {arriveTangent=-1.8333, interpMode=RCIM_Cubic, leaveTangent=-1.8333, time=0.1, value=1.25}
           {arriveTangent=-1.3889, interpMode=RCIM_Cubic, leaveTangent=-1.3889, time=0.3, value=0.45}
           <at 1 0 RCIM_Cubic>
          ]}}}
       CurveIndex=&Multiply_Float(A=$Particles.NormalizedAge, B=9.6)
      )
     )
    )
    Particles.Size_GlowCorona:Vector2f=&Multiply_Vector2D_ByFloat(
     Float=&FloatFromCurve(
      FloatCurve=DataInterface() {propertyValues {Curve {keys [
          {arriveTangent=0.225, interpMode=RCIM_Cubic, time=0, value=0.025}
          {arriveTangent=1.2, interpMode=RCIM_Cubic, leaveTangent=1.2, time=0.1, value=0.575}
          <at 1 1.225 RCIM_Cubic>
         ]}}}
      CurveIndex=&Multiply_Float(A=$Particles.NormalizedAge, B=9.6)
     )
     Vector2D=$Particles.Initial.Size_GlowCorona
    )
   ) {isSetParameters=true}
   /SetVariables(Particles.Streak_01_Color:LinearColor=&ScaleLinearColorByCurve(
     InitialValue=$Particles.Initial.Streak_01_Color
     ScaleCurve=DataInterface() {propertyValues {
       AlphaCurve {keys [
         {interpMode=RCIM_Cubic, tangentMode=RCTM_User, time=0, value=1}
         {arriveTangent=-1.8333, arriveTangentWeight=0.0333, interpMode=RCIM_Cubic, leaveTangent=-1.8333, tangentMode=RCTM_User, time=0.1, value=1.25}
         {arriveTangent=-1.3889, arriveTangentWeight=0.0667, interpMode=RCIM_Cubic, leaveTangent=-1.3889, tangentMode=RCTM_User, time=0.3, value=0.45}
         {arriveTangentWeight=0.2333, interpMode=RCIM_Cubic, tangentMode=RCTM_User, time=1, value=0}
        ]}
       BlueCurve {keys [
         {interpMode=RCIM_Cubic, tangentMode=RCTM_User, time=0, value=0}
         {arriveTangent=3.5714, arriveTangentWeight=0.0333, interpMode=RCIM_Cubic, leaveTangent=3.5714, tangentMode=RCTM_User, time=0.1, value=5}
         <at 0.7 2.5>
         {arriveTangentWeight=0.1, interpMode=RCIM_Cubic, tangentMode=RCTM_User, time=1, value=0}
        ]}
       GreenCurve {keys [
         {interpMode=RCIM_Cubic, tangentMode=RCTM_User, time=0, value=0}
         {arriveTangent=3.5714, arriveTangentWeight=0.0333, interpMode=RCIM_Cubic, leaveTangent=3.5714, tangentMode=RCTM_User, time=0.1, value=5}
         <at 0.7 2.5>
         {arriveTangentWeight=0.1, interpMode=RCIM_Cubic, tangentMode=RCTM_User, time=1, value=0}
        ]}
       RedCurve {keys [
         {interpMode=RCIM_Cubic, tangentMode=RCTM_User, time=0, value=0}
         {arriveTangent=3.5714, arriveTangentWeight=0.0333, interpMode=RCIM_Cubic, leaveTangent=3.5714, tangentMode=RCTM_User, time=0.1, value=5}
         <at 0.7 2.5>
         {arriveTangentWeight=0.1, interpMode=RCIM_Cubic, tangentMode=RCTM_User, time=1, value=0}
        ]}
      }}
     CurveIndex=&Multiply_Float(A=$Particles.NormalizedAge, B=9.6)
    )) {isSetParameters=true}
   /SetVariables(
    Particles.Color_FlareIncrease:LinearColor=&MultiplyLinearColor(
     Linear Color=$Particles.Initial.Color_FlareIncrease
     Scale Factor=&MakeFloat4FromVectorAndFloat(
      "Vector (RGB)"=&VectorFromFloat(Value=&FloatFromCurve(
        FloatCurve=DataInterface() {propertyValues {Curve {keys [
            {arriveTangent=9.0725, interpMode=RCIM_Cubic, leaveTangent=9.0725, tangentMode=RCTM_User, time=0, value=0}
            {arriveTangent=-2.5249, interpMode=RCIM_Cubic, leaveTangent=-2.5249, tangentMode=RCTM_User, time=0.1408, value=0.9847}
            {arriveTangent=-0.0405, interpMode=RCIM_Cubic, leaveTangent=-0.0405, tangentMode=RCTM_User, time=1, value=0}
           ]}}}
        CurveIndex=&NormalizeFloat(Value=&Multiply_Float(A=$Particles.NormalizedAge, B=10), Min=0, Max=1)
        Scale Curve=10
       ))
      "Float (Alpha)"=1
     )
    )
    Particles.Size_FlareIncrease:Vector2f=&Multiply_Vector2D_ByFloat(
     Float=&Multiply_Float(
      A=1
      B=&Lerp_Float(
       A=1
       B=6
       Alpha=&FloatFromCurve(
        FloatCurve=DataInterface() {propertyValues {Curve {keys [
            {arriveTangent=9.0725, interpMode=RCIM_Cubic, leaveTangent=9.0725, tangentMode=RCTM_User, time=0, value=0}
            {arriveTangent=-2.5249, interpMode=RCIM_Cubic, leaveTangent=-2.5249, tangentMode=RCTM_User, time=0.1408, value=0.9847}
            {arriveTangent=-0.0405, interpMode=RCIM_Cubic, leaveTangent=-0.0405, tangentMode=RCTM_User, time=1, value=0}
           ]}}}
        CurveIndex=&NormalizeFloat(Value=&Multiply_Float(A=$Particles.NormalizedAge, B=10), Min=0, Max=1)
       )
      )
     )
     Vector2D=$Particles.Initial.Size_FlareIncrease
    )
   ) {isSetParameters=true}
  ]
  rendererClasses [&NiagaraMeshRendererProperties, &NiagaraSpriteRendererProperties]
  renderers [
   {rendererClass=&NiagaraMeshRendererProperties, rendererIndex=0}
   {rendererClass=&NiagaraSpriteRendererProperties, rendererIndex=1}
   {rendererClass=&NiagaraSpriteRendererProperties, rendererIndex=2}
   {rendererClass=&NiagaraSpriteRendererProperties, rendererIndex=3}
  ]
 }
 Wind {
  emitterUpdateScript [
   &EmitterState()
   &SpawnBurst_Instantaneous(Use Spawn Probability=false, UseLoopCountLimit=false)
  ]
  particleSpawnScript [
   &PartitionParticles()
   &InitializeParticle(
    Color Maximum=<rgba 0.2 0.1667 0.15 1>
    Color Minimum=<rgba 0.25 0.1979 0.1875 1>
    Color Mode=::NewEnumerator2
    Lifetime Max=0.5
    Lifetime Min=0.25
    Lifetime Mode=::NewEnumerator1
    Mesh Scale Mode=::NewEnumerator3
    Mesh Uniform Scale=&Multiply_Float(A=&RandomRangeFloat(Minimum=0.25, Maximum=0.4), B=0.5)
    Position Offset=<5 0 0>
    UsePositionOffset=true
    Write Material Random=true
   )
   &InitialMeshOrientation(Mesh Orientation Mode=::NewEnumerator2, Orientation Vector=<0 1 0>, Use Rotation Vector=true)
  ]
  particleUpdateScript [
   &ParticleState()
   &SkeletalMeshLocation(Bone Sampling Mode=::NewEnumerator4, MaskSpawnGroup=false, Orientation Sampling=::NewEnumerator1, Skeletal Mesh=$System.SkeletalMesh)
   &ScaleMeshSize(Scale Factor=&MakeVector(
     X=&FloatFromCurve(
      FloatCurve=DataInterface() {propertyValues {Curve {keys [
          {arriveTangent=4.1785, interpMode=RCIM_Cubic, leaveTangent=4.1785, tangentMode=RCTM_User, time=0, value=0}
          {arriveTangent=1.1179, interpMode=RCIM_Cubic, leaveTangent=1.1179, tangentMode=RCTM_User, time=0.2042, value=0.6228}
          {arriveTangent=0.0224, interpMode=RCIM_Cubic, leaveTangent=0.0224, tangentMode=RCTM_User, time=1, value=1}
         ]}}}
      Scale Curve=&RandomRangeFloat(Minimum=2, Maximum=2.75)
     )
     Y=1.5
     Z=1.5
    ))
   &ScaleColor(
    Scale Alpha=&FloatFromCurve(
     FloatCurve=DataInterface() {propertyValues {Curve {keys [
         {arriveTangent=0.6385, interpMode=RCIM_Cubic, leaveTangent=0.6385, tangentMode=RCTM_User, time=0, value=0}
         {
          arriveTangent=1.9289
          arriveTangentWeight=0.1253
          interpMode=RCIM_Cubic
          leaveTangent=1.9289
          leaveTangentWeight=0.2429
          tangentMode=RCTM_User
          time=0.1
          value=0.2
         }
         {arriveTangent=0.0224, interpMode=RCIM_Cubic, leaveTangent=0.0224, tangentMode=RCTM_User, time=0.5, value=1}
         {arriveTangent=-3.0357, interpMode=RCIM_Cubic, leaveTangent=-3.0357, tangentMode=RCTM_User, time=1, value=0.2}
        ]}}}
     CurveIndex=$Emitter.NormalizedLoopAge
     Scale Curve=16
    )
    ScaleA=true
    ScaleRGB=false
   )
  ]
  rendererClasses [&NiagaraMeshRendererProperties]
  renderers [{rendererClass=&NiagaraMeshRendererProperties, rendererIndex=0}]
 }
}
summary {
 systemName=NS_RocketDrill_Boost_Burst
 emitters [
  {
   emitterName=Combined_Burst_Once
   rendererClasses [&NiagaraMeshRendererProperties, &NiagaraSpriteRendererProperties]
  }
  {emitterName=Wind, rendererClasses [&NiagaraMeshRendererProperties]}
  {emitterName=Air_Pressure, rendererClasses [&NiagaraMeshRendererProperties]}
 ]
 userVariables [User.uDuration {type=NiagaraFloat, defaultValue=5, description=""}]
}
