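# Niagara one-shot dust puff for a RocketDrill landing, built from burst smoke, air streaks and debris.
# One mesh emitter packs four animated layers via spawn-group visibility tags, with time-of-day material scalars on the fourth.
# Drag-slowed smoke, radial and upward streaks, distance-culled sprite chips and colliding mesh debris complete the impact.
# System.ScaleFX scales all emitters, and User.FX_Scale_Impact_Center additionally scales the upward streaks.
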NiagaraSystem /RocketDrillGameplay/Assets/VFX/NS_RocketDrill_DustPuff_OnLanding
dependencies {
 usedDataInterfaces [
  &NiagaraDataInterfaceVectorCurve
  &NiagaraDataInterfaceCurve
  &NiagaraDataInterfaceVector2DCurve
  &NiagaraDataInterfaceArrayFloat
 ]
 usedDynamicInputs [&SelectFloatFromArray]
 usedModules [
  &self:SystemSpawnScript.SystemScriptSource.SystemScriptGraph.NiagaraNodeAssignment_0.SetVariables_#1
  &SystemState
 ]
}
emitters {
 AirStreaks {
  emitterUpdateScript [
   &EmitterState(Life Cycle Mode=::NewEnumerator1, Loop Behavior=::NewEnumerator1, Loop Duration=1, UseLoopDelay=false)
   &SpawnBurst_Instantaneous(Spawn Count=18, Spawn Time=0.1, Use Spawn Probability=false, UseLoopCountLimit=false)
  ]
  particleSpawnScript [
   /SetVariables(Particles.RandomFloat=&UniformRangedFloat(Minimum=0, Maximum=3)) {isSetParameters=true, module=SetVariables_#2}
   &InitializeParticle(
    Color Maximum=<rgba 0.25 0.215 0.2057 1>
    Color Minimum=<rgba 0.15 0.1459 0.1409 1>
    Color Mode=::NewEnumerator2
    Lifetime Max=0.3
    Lifetime Min=0.15
    Lifetime Mode=::NewEnumerator1
    Sprite Size Max=<60 200>
    Sprite Size Min=<50 150>
    Sprite Size Mode=::NewEnumerator2
    UsePositionOffset=false
    Write Material Random=true
   ) {module=InitializeParticle001}
   /SetVariables(Particles.SpriteSize=&Multiply_Vector2D_ByFloat(Float=&Multiply_Float(A=$System.ScaleFX, B=1), Vector2D=$Particles.SpriteSize)) {isSetParameters=true}
   &ShapeLocation(
    Handle Radius=15
    Large Radius=50
    MaskSpawnGroup=false
    Non Uniform Scale=&Multiply_VectorByFloat(Vector=<1 1 0>, Float=$System.ScaleFX)
    Offset=<0 0 20>
    Offset Mode=::NewEnumerator0
    Override Seed=false
    Shape Primitive=::NewEnumerator3
    UseOwnerScale=true
   )
   &AddVelocity(Module.Use Falloff Distance=false, Override Seed=false, Velocity Mode=::NewEnumerator1, Velocity Speed=500)
   &ApplyInitialForces()
   &SubUVAnimation(End Frame=3, SubUV Animation Mode:ENiagara_SubUVLookupMode=::NewEnumerator2)
  ]
  particleUpdateScript [
   &ParticleState() {module=UpdateAge}
   &SolveForcesAndVelocity(Clamp Velocity=false, Limit Acceleration=false)
   &ScaleSpriteSize(Scale Factor=&Vector2DFromCurve(
     Vector2Curve=DataInterface() {propertyValues {XCurve {keys [<at 0.0036 1.6897>,<at 0.9967 0.0582>]}, YCurve {keys [<at 0 1>,<at 1.002 3.7227>]}}}
     ScaleCurve=&Multiply_Vector2D_ByFloat(Float=0.65, Vector2D=<1 1>)
    ))
   &ScaleColor(
    Scale Alpha=&FloatFromCurve(
     FloatCurve=DataInterface() {propertyValues {Curve {keys [{interpMode=RCIM_Cubic, tangentMode=RCTM_Break, time=0, value=1}, {interpMode=RCIM_Cubic, tangentMode=RCTM_Break, time=1, value=0}]}}}
     Scale Curve=&RandomRangeFloat(Minimum=0.05, Maximum=0.2)
    )
    ScaleA=true
    ScaleRGB=false
   )
   &DynamicMaterialParameters(
    Index 0 Param 2=1
    Index 0 Param 3=1
    Index 0 Param 4=&FloatFromCurve(FloatCurve=DataInterface() {propertyValues {Curve {keys [<at 0.3 1>,<at 1 0.25>]}}})
    Module.Param0WriteEnabled=true
    Module.Param1WriteEnabled=true
    Module.Param2WriteEnabled=true
    Module.Param3WriteEnabled=true
   )
  ]
  rendererClasses [&NiagaraSpriteRendererProperties]
  renderers [{rendererClass=&NiagaraSpriteRendererProperties}]
 }
 AirStreaks_upward {
  emitterUpdateScript [
   &EmitterState()
   &SpawnBurst_Instantaneous(Spawn Count=12, Spawn Time=0.1, Use Spawn Probability=false, UseLoopCountLimit=false)
  ]
  particleSpawnScript [
   /SetVariables(Particles.RandomFloat=&UniformRangedFloat(Minimum=0, Maximum=3)) {isSetParameters=true, module=SetVariables_#2}
   &InitializeParticle(
    Color Maximum=<rgba 0.4 0.3329 0.3092 1>
    Color Minimum=<rgba 0.2604 0.2292 0.2121 1>
    Color Mode=::NewEnumerator2
    Lifetime Max=0.25
    Lifetime Min=0.15
    Lifetime Mode=::NewEnumerator1
    Sprite Size Max=<50 280>
    Sprite Size Min=<40 180>
    Sprite Size Mode=::NewEnumerator2
    UsePositionOffset=false
    Write Material Random=true
   ) {module=InitializeParticle001}
   /SetVariables(Particles.SpriteSize=&Multiply_Vector2D_ByFloat(Float=&Multiply_Float(A=$System.ScaleFX, B=$User.FX_Scale_Impact_Center), Vector2D=$Particles.SpriteSize)) {isSetParameters=true}
   &ShapeLocation(
    Large Radius=25
    MaskSpawnGroup=false
    Non Uniform Scale=&Multiply_VectorByFloat(Vector=<1 1 0>, Float=&Multiply_Float(A=$System.ScaleFX, B=$User.FX_Scale_Impact_Center))
    Override Seed=false
    Shape Primitive=::NewEnumerator3
    UseOwnerScale=true
   )
   &AddVelocity(Module.Use Falloff Distance=false, Origin Offset=<0 0 -500>, Override Seed=false, Velocity Mode=::NewEnumerator1)
   &ApplyInitialForces()
   &SubUVAnimation(End Frame=3, SubUV Animation Mode:ENiagara_SubUVLookupMode=::NewEnumerator2)
  ]
  particleUpdateScript [
   &ParticleState() {module=UpdateAge}
   &SolveForcesAndVelocity(Clamp Velocity=false, Limit Acceleration=false)
   &ScaleSpriteSize(Scale Factor=&Vector2DFromCurve(
     Vector2Curve=DataInterface() {propertyValues {XCurve {keys [<at 0.0036 1.6897>,<at 0.9967 0.0582>]}, YCurve {keys [<at 0 1>,<at 1.002 3.7227>]}}}
     ScaleCurve=&Multiply_Vector2D_ByFloat(Float=0.65, Vector2D=<1 1>)
    ))
   &ScaleColor(
    Scale Alpha=&FloatFromCurve(
     FloatCurve=DataInterface() {propertyValues {Curve {keys [
         {interpMode=RCIM_Cubic, leaveTangent=-1.9294, tangentMode=RCTM_Break, time=0, value=1}
         {interpMode=RCIM_Cubic, tangentMode=RCTM_Break, time=1, value=0}
        ]}}}
     Scale Curve=&RandomRangeFloat(Minimum=0.03, Maximum=0.12)
    )
    ScaleA=true
    ScaleRGB=true
   )
   &DynamicMaterialParameters(
    Index 0 Param 2=1
    Index 0 Param 3=1
    Index 0 Param 4=&FloatFromCurve(FloatCurve=DataInterface() {propertyValues {Curve {keys [<at 0.3 1>,<at 1 0.25>]}}})
    Module.Param0WriteEnabled=true
    Module.Param1WriteEnabled=true
    Module.Param2WriteEnabled=true
    Module.Param3WriteEnabled=true
   )
  ]
  rendererClasses [&NiagaraSpriteRendererProperties]
  renderers [{rendererClass=&NiagaraSpriteRendererProperties}]
 }
 Combined_Emitter_Mesh {
  emitterUpdateScript [
   &EmitterState(Life Cycle Mode=::NewEnumerator1, Loop Behavior=::NewEnumerator1, Loop Duration=1, UseLoopDelay=false)
   &SpawnBurst_Instantaneous(Use Spawn Probability=false, UseLoopCountLimit=false)
   &SpawnBurst_Instantaneous(Spawn Group=1, Use Spawn Probability=false, UseLoopCountLimit=false) {module=SpawnBurst_Instantaneous001}
   &SpawnBurst_Instantaneous(Spawn Count=2, Spawn Group=2, Spawn Time=0.1, Use Spawn Probability=false, UseLoopCountLimit=false) {module=SpawnBurst_Instantaneous002}
   &SpawnBurst_Instantaneous(Spawn Count=4, Spawn Group=3, Spawn Time=0.1, Use Spawn Probability=false, UseLoopCountLimit=false) {module=SpawnBurst_Instantaneous003}
  ]
  particleSpawnScript [
   /SetVariables(Particles.SpawnGroup=$Emitter.SpawnGroup) {isSetParameters=true}
   /SetVariables(Particles.VisibilityTag=$Particles.SpawnGroup) {isSetParameters=true}
   &InitializeParticle(Color Mode=::NewEnumerator0, UsePositionOffset=false, Write Material Random=true)
   /SetVariables(
    Particles.Mesh_01_CameraOffset=100
    Particles.Mesh_01_Color:LinearColor=<rgba 0.7091 0.5906 0.4849 2>
    Particles.Mesh_01_Orientation:Quat4f=&OrientMeshToVector(Look At Direction=<1 0 0>, Reference Axis Vector=<1 0 0>)
    Particles.Mesh_01_Size:Vector3f=&VectorFromFloat(Value=&Multiply_Float(A=3, B=$System.ScaleFX))
   ) {isSetParameters=true}
   /SetVariables(
    Particles.Mesh_02_Color:LinearColor=<rgba 0.3162 0.3155 0.315 1>
    Particles.Mesh_02_Orientation:Quat4f=&OrientMeshToVector(Look At Direction=<0 1 0>, Reference Axis Vector=<1 0 0>)
    Particles.Mesh_02_Size:Vector3f=&Multiply_VectorByFloat(Vector=<0.6 0.05 0.6>, Float=$System.ScaleFX)
   ) {isSetParameters=true}
   /SetVariables(
    Particles.Mesh_03_Color:LinearColor=<rgba 0.8609 0.7779 0.6695 0.375>
    Particles.Mesh_03_Orientation:Quat4f=&MultiplyQuaternion(
     Quaternion A=&OrientMeshToVector(Look At Direction=<1 0 0>, Reference Axis Vector=<1 0 0>)
     Quaternion B=&MakeQuatFromAxisAngle(Axis=<0 0 1>, AngleInDegrees=&RandomRangeFloat(Minimum=-180, Maximum=180))
    )
    Particles.Mesh_03_Size:Vector3f=&VectorFromFloat(Value=&Multiply_Float(A=1.8, B=$System.ScaleFX))
   ) {isSetParameters=true}
   /SetVariables(
    Particles.Mesh_04_Color:LinearColor=<rgba 0.9143 0.8098 0.7338 0.275>
    Particles.Mesh_04_Orientation:Quat4f=&MultiplyQuaternion(
     Quaternion A=&MultiplyQuaternion(
      Quaternion A=&OrientMeshToVector(Look At Direction=<1 0 0>, Reference Axis Vector=<1 0 0>)
      Quaternion B=&MakeQuatFromAxisAngle(Axis=<0 0 1>, AngleInDegrees=&RandomRangeFloat(Minimum=-180, Maximum=180))
     )
     Quaternion B=&MakeQuatFromAxisAngle(
      Axis=<1 0 0>
      AngleInDegrees=&RandomRangeFloat(Minimum=-20, Maximum=20, Randomness Mode=::NewEnumerator1)
     )
    )
    Particles.Mesh_04_Size:Vector3f=&VectorFromFloat(Value=&Multiply_Float(A=&RandomRangeFloat(Minimum=1.25, Maximum=4), B=$System.ScaleFX))
   ) {isSetParameters=true}
   &SystemLocation(MaskSpawnGroup=false, Offset=<0 0 10>)
   &CameraOffset(Camera Offset Amount=25)
  ]
  particleUpdateScript [
   &ParticleState()
   /SetVariables(
    Particles.Mesh_01_Color:LinearColor=&MultiplyLinearColor(
     Linear Color=$Particles.Initial.Mesh_01_Color
     Scale Factor=&MakeFloat4FromVectorAndFloat(
      "Vector (RGB)"=&VectorFromCurve(
       VectorCurve=DataInterface() {propertyValues {XCurve {keys [<at 0 1>,<at 1 0.526>]}, YCurve {keys [<at 0 1>,<at 1 0.526>]}, ZCurve {keys [<at 0 1>,<at 1 0.526>]}}}
       CurveIndex=&Divide_Float(A=$Particles.NormalizedAge, B=0.7)
      )
      "Float (Alpha)"=&FloatFromCurve(
       FloatCurve=DataInterface() {propertyValues {Curve {keys [<at 0 1>,<at 0.1 0.5>,<at 1 0>]}}}
       CurveIndex=&Divide_Float(A=$Particles.NormalizedAge, B=0.7)
       Scale Curve=0.075
      )
     )
    )
    Particles.Mesh_01_DynamicParameter:Vector4f=&MakeVector4(
     X=&FloatFromCurve(
      FloatCurve=DataInterface() {propertyValues {Curve {keys [{arriveTangent=2.5068, interpMode=RCIM_Cubic, leaveTangent=2.5068, tangentMode=RCTM_User, time=0, value=-0.4}, <at 1 0.55>]}}}
      CurveIndex=&Divide_Float(A=$Particles.NormalizedAge, B=0.7)
     )
     Y=1
     Z=0
     W=0
    )
    Particles.Mesh_01_Size:Vector3f=&Multiply_Vector(
     A=$Particles.Initial.Mesh_01_Size
     B=&VectorFromFloat(Value=&FloatFromCurve(
       FloatCurve=DataInterface() {propertyValues {Curve {keys [
           {arriveTangent=3.5711, interpMode=RCIM_Cubic, leaveTangent=3.5711, tangentMode=RCTM_User, time=0, value=0}
           {interpMode=RCIM_Cubic, tangentMode=RCTM_User, time=1, value=1.5}
          ]}}}
       CurveIndex=&Divide_Float(A=$Particles.NormalizedAge, B=0.7)
       Scale Curve=1.25
      ))
    )
   ) {isSetParameters=true}
   /SetVariables(
    Particles.Mesh_02_Color:LinearColor=&MultiplyLinearColor(
     Linear Color=$Particles.Initial.Mesh_02_Color
     Scale Factor=&MakeFloat4FromVectorAndFloat(
      "Vector (RGB)"=&VectorFromCurve(
       VectorCurve=DataInterface() {propertyValues {
         XCurve {keys [
           <at 0 1 RCIM_Cubic>
           {arriveTangent=9.97, interpMode=RCIM_Cubic, leaveTangent=9.97, time=0.1536, value=1}
           <at 0.776 8.7362 RCIM_Cubic>
           <at 1 1 RCIM_Cubic>
          ]}
         YCurve {keys [
           <at 0 1 RCIM_Cubic>
           {arriveTangent=8.4018, interpMode=RCIM_Cubic, leaveTangent=8.4018, time=0.1536, value=0.6932}
           {arriveTangent=0.3625, interpMode=RCIM_Cubic, leaveTangent=0.3625, time=0.7783, value=7.5392}
           <at 1 1 RCIM_Cubic>
          ]}
         ZCurve {keys [
           <at 0 1 RCIM_Cubic>
           {arriveTangent=7.6288, interpMode=RCIM_Cubic, leaveTangent=7.6288, time=0.1536, value=0.1681}
           {arriveTangent=0.9829, interpMode=RCIM_Cubic, leaveTangent=0.9829, time=0.7781, value=6.9361}
           <at 1 1 RCIM_Cubic>
          ]}
        }}
       CurveIndex=&Divide_Float(A=$Particles.NormalizedAge, B=0.3)
       Scale Curve=&VectorFromFloat(Value=10)
      )
      "Float (Alpha)"=&FloatFromCurve(
       FloatCurve=DataInterface() {propertyValues {Curve {keys [
           {interpMode=RCIM_Cubic, leaveTangentWeight=0.3333, tangentMode=RCTM_User, time=0, value=0}
           {
            arriveTangent=-3.9986
            arriveTangentWeight=0.0551
            interpMode=RCIM_Cubic
            leaveTangent=-3.9986
            leaveTangentWeight=0.2783
            tangentMode=RCTM_User
            time=0.1652
            value=1
           }
           {arriveTangentWeight=0.2783, interpMode=RCIM_Cubic, tangentMode=RCTM_User, time=1, value=0}
          ]}}}
       CurveIndex=&Divide_Float(A=$Particles.NormalizedAge, B=0.3)
      )
     )
    )
    Particles.Mesh_02_DynamicParameter:Vector4f=&MakeVector4(
     X=&FloatFromCurve(
      FloatCurve=DataInterface() {propertyValues {Curve {keys [
          {arriveTangent=4.1785, interpMode=RCIM_Cubic, leaveTangent=4.1785, tangentMode=RCTM_User, time=0, value=0}
          {arriveTangent=1.1179, interpMode=RCIM_Cubic, leaveTangent=1.1179, tangentMode=RCTM_User, time=0.2042, value=0.6228}
          {arriveTangent=0.0224, interpMode=RCIM_Cubic, leaveTangent=0.0224, tangentMode=RCTM_User, time=1, value=1}
         ]}}}
      CurveIndex=&Divide_Float(A=$Particles.NormalizedAge, B=0.3)
     )
     Y=0
     Z=0
     W=1
    )
    Particles.Mesh_02_Size:Vector3f=&Multiply_Vector(
     A=$Particles.Initial.Mesh_02_Size
     B=&VectorFromFloat(Value=&FloatFromCurve(
       FloatCurve=DataInterface() {propertyValues {Curve {keys [
           {arriveTangent=4.1785, interpMode=RCIM_Cubic, leaveTangent=4.1785, tangentMode=RCTM_User, time=0, value=0}
           {arriveTangent=0.9995, interpMode=RCIM_Cubic, leaveTangent=0.9995, tangentMode=RCTM_User, time=0.2223, value=0.6825}
           {arriveTangent=-0.0405, interpMode=RCIM_Cubic, leaveTangent=-0.0405, tangentMode=RCTM_User, time=1, value=1}
          ]}}}
       CurveIndex=&Divide_Float(A=$Particles.NormalizedAge, B=0.3)
       Scale Curve=2.5
      ))
    )
   ) {isSetParameters=true}
   /SetVariables(
    Particles.Mesh_03_Color:LinearColor=&MultiplyLinearColor(
     Linear Color=$Particles.Initial.Mesh_03_Color
     Scale Factor=&MakeFloat4FromVectorAndFloat(
      "Vector (RGB)"=<1 1 1>
      "Float (Alpha)"=&FloatFromCurve(
       FloatCurve=DataInterface() {propertyValues {Curve {keys [
           {arriveTangent=6.2301, interpMode=RCIM_Cubic, leaveTangent=6.2301, tangentMode=RCTM_User, time=0, value=0.5}
           {arriveTangent=-0.7357, interpMode=RCIM_Cubic, leaveTangent=-0.7357, tangentMode=RCTM_User, time=0.1, value=2}
           <at 0.4 0.2>
           {arriveTangent=0.4048, interpMode=RCIM_Cubic, leaveTangent=0.4048, tangentMode=RCTM_User, time=0.7, value=0}
          ]}}}
       CurveIndex=&Divide_Float(A=$Particles.NormalizedAge, B=0.5)
       Scale Curve=0.5
      )
     )
    )
    Particles.Mesh_03_Size:Vector3f=&Multiply_Vector(
     A=$Particles.Initial.Mesh_03_Size
     B=&Multiply_VectorByFloat(
      Vector=&MakeVector(
       X=1
       Y=1
       Z=&FloatFromCurve(
        FloatCurve=DataInterface() {propertyValues {Curve {keys [<at 0 1>,<at 1 0>]}}}
        CurveIndex=&Divide_Float(A=$Particles.NormalizedAge, B=0.5)
       )
      )
      Float=&FloatFromCurve(
       FloatCurve=DataInterface() {propertyValues {Curve {keys [
           {arriveTangent=7.2197, interpMode=RCIM_Cubic, leaveTangent=7.2197, tangentMode=RCTM_User, time=0, value=0}
           {arriveTangent=1.9952, interpMode=RCIM_Cubic, leaveTangent=1.9952, time=0.4917, value=1.9432}
           {arriveTangent=-0.0405, interpMode=RCIM_Cubic, leaveTangent=-0.0405, tangentMode=RCTM_User, time=1.0024, value=2}
          ]}}}
       CurveIndex=&Divide_Float(A=$Particles.NormalizedAge, B=0.5)
      )
     )
    )
   ) {isSetParameters=true}
   /SetVariables(
    Particles.Mesh_04_Color:LinearColor=&MultiplyLinearColor(
     Linear Color=$Particles.Initial.Mesh_04_Color
     Scale Factor=&MakeFloat4FromVectorAndFloat(
      "Vector (RGB)"=<1 1 1>
      "Float (Alpha)"=&FloatFromCurve(
       FloatCurve=DataInterface() {propertyValues {Curve {keys [
           {arriveTangent=8.2356, interpMode=RCIM_Cubic, leaveTangent=8.2356, tangentMode=RCTM_User, time=0, value=0}
           {arriveTangent=0.1969, interpMode=RCIM_Cubic, leaveTangent=0.1969, tangentMode=RCTM_User, time=0.2223, value=1}
           {arriveTangent=-3.0486, interpMode=RCIM_Cubic, leaveTangent=-3.0486, tangentMode=RCTM_User, time=0.4781, value=0.509}
           {arriveTangent=-0.2378, interpMode=RCIM_Cubic, leaveTangent=-0.2378, tangentMode=RCTM_User, time=1, value=0}
          ]}}}
       CurveIndex=&Divide_Float(A=$Particles.NormalizedAge, B=&RandomRangeFloat(Minimum=0.5, Maximum=0.75))
      )
     )
    )
    Particles.Mesh_04_DynamicParameter:Vector4f=&MakeVector4(
     X=&FloatFromCurve(
      FloatCurve=DataInterface() {propertyValues {Curve {keys [{interpMode=RCIM_Cubic, tangentMode=RCTM_User, time=0.5, value=0}, {interpMode=RCIM_Cubic, tangentMode=RCTM_User, time=1, value=1}]}}}
      CurveIndex=&Divide_Float(A=$Particles.NormalizedAge, B=&RandomRangeFloat(Minimum=0.5, Maximum=0.75))
     )
     Y=&FNTimeOfDayScalar(Night Scale=0.05, Morning Scale=-0.26, Day Scale=0, Evening Scale=-0.35)
     Z=&FNTimeOfDayScalar(Night Scale=0.5, Morning Scale=0.23, Evening Scale=0.2)
     W=&FNTimeOfDayScalar(Night Scale=0.02, Morning Scale=-0.01, Day Scale=0, Evening Scale=-0.01)
    )
    Particles.Mesh_04_Orientation:Quat4f=&MultiplyQuaternion(
     Quaternion A=$Particles.Initial.Mesh_04_Orientation
     Quaternion B=&MakeQuatFromAxisAngle(
      Axis=&RandomRangeVector(Minimum=<0 0 -1>, Maximum=<0 0 1>)
      AngleInDegrees=&RandomRangeFloat(Minimum=-50, Maximum=50)
     )
    )
    Particles.Mesh_04_Size:Vector3f=&Multiply_Vector(
     A=$Particles.Initial.Mesh_04_Size
     B=&Multiply_VectorByFloat(
      Vector=&MakeVector(
       X=&FloatFromCurve(
        FloatCurve=DataInterface() {propertyValues {Curve {keys [
            {arriveTangent=0.8546, interpMode=RCIM_Cubic, leaveTangent=3.1486, tangentMode=RCTM_Break, time=0, value=0}
            {arriveTangent=1.4415, interpMode=RCIM_Cubic, leaveTangent=1.4415, tangentMode=RCTM_User, time=0.3242, value=0.7}
            {arriveTangent=1.2727, interpMode=RCIM_Cubic, leaveTangent=1.2727, tangentMode=RCTM_User, time=1, value=1.5}
           ]}}}
        CurveIndex=&Divide_Float(A=$Particles.NormalizedAge, B=&RandomRangeFloat(Minimum=0.5, Maximum=0.75))
       )
       Y=&FloatFromCurve(
        FloatCurve=DataInterface() {propertyValues {Curve {keys [
            {arriveTangent=0.8546, interpMode=RCIM_Cubic, leaveTangent=3.1486, tangentMode=RCTM_Break, time=0, value=0}
            {arriveTangent=1.4415, interpMode=RCIM_Cubic, leaveTangent=1.4415, tangentMode=RCTM_User, time=0.3242, value=0.7}
            {arriveTangent=1.2727, interpMode=RCIM_Cubic, leaveTangent=1.2727, tangentMode=RCTM_User, time=1, value=1.5}
           ]}}}
        CurveIndex=&Divide_Float(A=$Particles.NormalizedAge, B=&RandomRangeFloat(Minimum=0.5, Maximum=0.75))
       )
       Z=&FloatFromCurve(
        FloatCurve=DataInterface() {propertyValues {Curve {keys [
            {arriveTangent=1.8479, interpMode=RCIM_Cubic, leaveTangent=1.8479, tangentMode=RCTM_User, time=0, value=0}
            {arriveTangent=0.9995, interpMode=RCIM_Cubic, leaveTangent=0.9995, tangentMode=RCTM_User, time=0.4, value=0.4}
            {arriveTangent=0.4855, interpMode=RCIM_Cubic, leaveTangent=0.4855, tangentMode=RCTM_User, time=1, value=0.6}
           ]}}}
        CurveIndex=&Divide_Float(A=$Particles.NormalizedAge, B=&RandomRangeFloat(Minimum=0.5, Maximum=0.75))
       )
      )
      Float=&FloatFromCurve(
       FloatCurve=DataInterface() {propertyValues {Curve {keys [
           {interpMode=RCIM_Cubic, leaveTangent=6.5404, tangentMode=RCTM_Break, time=0, value=1.25}
           {arriveTangent=0.0345, interpMode=RCIM_Cubic, leaveTangent=0.0345, tangentMode=RCTM_User, time=0.0648, value=1.4045}
           {arriveTangent=0.0556, interpMode=RCIM_Cubic, leaveTangent=0.0556, tangentMode=RCTM_User, time=1, value=1.4}
          ]}}}
       CurveIndex=&Divide_Float(A=$Particles.NormalizedAge, B=&RandomRangeFloat(Minimum=0.5, Maximum=0.75))
      )
     )
    )
   ) {isSetParameters=true}
   /SetVariables(Particles.Mesh_Position:NiagaraPosition=&Add_Vector(
     A=Linked() {linkedVariable {name=Particles.Position, type=NiagaraPosition}}
     B=&RandomRangeVector(Minimum=<0 0 25>, Maximum=<0 0 50>)
    )) {isSetParameters=true}
  ]
  rendererClasses [&NiagaraMeshRendererProperties]
  renderers [
   {rendererClass=&NiagaraMeshRendererProperties, rendererIndex=0}
   {rendererClass=&NiagaraMeshRendererProperties, rendererIndex=1}
   {rendererClass=&NiagaraMeshRendererProperties, rendererIndex=2}
   {rendererClass=&NiagaraMeshRendererProperties, rendererIndex=3}
  ]
 }
 Debris {
  emitterUpdateScript [
   &EmitterState(Enable Distance Culling=true, MaxDistance=4000, Module.bUseMaxDistance=true, Module.bUseMinDistance=false, Scalability Mode=::NewEnumerator1)
   &SpawnBurst_Instantaneous(Spawn Count=12, Spawn Time=0.2, Use Spawn Probability=false)
  ]
  particleSpawnScript [
   /SetVariables(Particles.bCollides:NiagaraBool=&Boolean_AndOperation(Bool1=&RandomBool(), Bool2=&RandomBool())) {isSetParameters=true}
   /SetVariables(
    Particles.ScalePicker:NiagaraFloat=&Lerp_Float(
     A=5
     B=20
     Alpha=&Power(
      a=&RandomRangeFloat(Minimum=0, Maximum=&MakeCustomFloatFromBool(Bool=$Particles.bCollides, True Float=0.85, False Float=1))
      b=3
     )
    )
    Particles.SpriteVisibility:NiagaraInt32=&MakeIntFromFloat(Float=0)
    Particles.VelocityPicker:NiagaraFloat=&Power(
     a=&MakeCustomFloatFromBool(
      Bool=true
      True Float=&RandomRangeFloat(Minimum=0.25, Maximum=0.75)
      False Float=&RandomRangeFloat(Minimum=0, Maximum=1)
     )
     b=3
    )
   ) {isSetParameters=true}
   &InitializeParticle(
    Color=&MakeLinearColorFromVectorAndFloat(
     "Vector (RGB)"=&VectorFromFloat(Value=&RandomRangeFloat(Minimum=0, Maximum=0.5))
     "Float (Alpha)"=0.8
    )
    Lifetime=&ClampFloat(
     Float=&Add_Float(
      A=&Lerp_Float(A=0.3, B=0.75, Alpha=$Particles.VelocityPicker)
      B=&MakeCustomFloatFromBool(
       Bool=$Particles.bCollides
       True Float=&RandomRangeFloat(Minimum=0.5, Maximum=1)
       False Float=&RandomRangeFloat(Minimum=-0.2, Maximum=0.2)
      )
     )
     Min=0.3
     Max=1.5
    )
    Sprite Rotation Mode=::NewEnumerator2
    Sprite Size Mode=::NewEnumerator3
    Uniform Sprite Size=$Particles.ScalePicker
    UsePositionOffset=false
    Write Material Random=true
   ) {module=InitializeParticle001}
   /SetVariables(Particles.SpriteSize=&Multiply_Vector2D_ByFloat(Float=$System.ScaleFX, Vector2D=$Particles.SpriteSize)) {isSetParameters=true}
   &SubUVAnimation(End Frame=4, SubUV Animation Mode:ENiagara_SubUVLookupMode=::NewEnumerator2) {module=SubUVAnimation001}
   &SphereLocation(
    MaskSpawnGroup=false
    Non Uniform Scale=&Multiply_VectorByFloat(Vector=<1 1 1>, Float=$System.ScaleFX)
    Override Seed=false
    Sphere Radius=120
    Surface Only=false
    UseNonUniformScale=true
    UseOffset=false
    UseSphereOrientation=false
   ) {module=SphereLocation001}
   &AddVelocityFromPoint(
    Module.Use Falloff Distance=false
    Origin Offset=<-2 0 0>
    Velocity Strength=&Lerp_Float(A=250, B=1250, Alpha=$Particles.VelocityPicker)
   )
   &AddVelocityInCone(
    Cone Axis=<0 0 1>
    Cone Axis Coordinate Space=::World
    Module.Use Velocity Falloff On Cone Axis=false
    Velocity Strength=&Lerp_Float(A=&RandomRangeFloat(Minimum=0, Maximum=300), B=0, Alpha=$Particles.VelocityPicker)
   )
   &CameraOffset(Camera Offset Amount=30)
  ]
  particleUpdateScript [
   &ParticleState()
   &Drag(
    Drag=&MakeCustomFloatFromBool(
     Bool=$Particles.bCollides
     True Float=0
     False Float=&ClampFloat(Float=&Add_Float(
       A=&Lerp_Float(A=0, B=0.5, Alpha=$Particles.VelocityPicker)
       B=&RandomRangeFloat(Minimum=-0.5, Maximum=0)
      ))
    )
    Use Linear Drag=true
    Use Rotational Drag=false
   )
   &GravityForce(Gravity=&Lerp_Vector(
     A=<0 0 -500>
     B=<0 0 -2000>
     Alpha=&Multiply_Float(
      A=&MakeCustomFloatFromBool(Bool=$Particles.bCollides, True Float=0.75, False Float=0.5)
      B=&Add_Float(
       A=$Particles.VelocityPicker
       B=&Add_Float(A=-0.3043, B=&Multiply_Float(A=0.0435, B=$Particles.ScalePicker))
      )
     )
    ))
   &SolveForcesAndVelocity(Clamp Velocity=false, Limit Acceleration=false)
   &SpriteRotationRate() {module=SpriteRotationRate001}
   &ScaleSpriteSize(Scale Factor=&Vector2DFromFloat(Value=&FloatFromCurve(FloatCurve=DataInterface() {propertyValues {Curve {keys [<at 0.6918 1>,<at 1 0>]}}})))
  ]
  rendererClasses [&NiagaraSpriteRendererProperties]
  renderers [{rendererClass=&NiagaraSpriteRendererProperties}]
 }
 Debris001 {
  emitterUpdateScript [
   &EmitterState(Life Cycle Mode=::NewEnumerator1, Loop Behavior=::NewEnumerator1, Loop Duration=2, UseLoopDelay=false)
   &SpawnBurst_Instantaneous(
    Spawn Count=&RandomRangeInt(Minimum=6, Maximum=8)
    Use Spawn Probability=false
    UseLoopCountLimit=false
   )
  ]
  particleSpawnScript [
   &InitializeParticle(
    Color=<rgba 0.4 0.3423 0.3222 1>
    Lifetime Max=2.5
    Lifetime Min=2
    Lifetime Mode=::NewEnumerator1
    Mass Max=2
    Mass Mode=::NewEnumerator3
    Mesh Scale Max=<5 5 5>
    Mesh Scale Mode=::NewEnumerator2
    Position Offset=<0 0 15>
    Sprite Rotation Mode=::NewEnumerator2
    UsePositionOffset=true
    Write Material Random=true
   )
   /SetVariables(Particles.Scale=&Multiply_Vector(
     A=$Particles.Scale
     B=&VectorFromFloat(Value=&Multiply_Float(A=$System.ScaleFX, B=0.1))
    )) {isSetParameters=true}
   &ShapeLocation(
    Large Radius=120
    MaskSpawnGroup=false
    Non Uniform Scale=&Multiply_VectorByFloat(Vector=<0.5 0.5 0.5>, Float=$System.ScaleFX)
    Override Seed=false
    Rotation Axis=<0 0 0>
    Rotation Mode=::NewEnumerator1
    Shape Primitive=::NewEnumerator3
    UseOwnerScale=true
   )
   &AddVelocity(
    Cone Axis=<0 0 1>
    Module.UseConeFalloff=false
    Module.UseVelDistribution=false
    Override Seed=false
    Velocity Mode=::NewEnumerator2
    Velocity Speed=&Multiply_Float(
     A=&RandomRangeFloat(Minimum=750, Maximum=2500)
     B=&Multiply_Float(A=$System.ScaleFX, B=0.75)
    )
   )
   &MeshRotationForce(
    Coordinate Space:ENiagaraMeshTransforms=::NewEnumerator2
    Rotation=&RandomRangeVector(Minimum=<-2 -2 -2>, Maximum=<3 3 3>)
   )
   &ApplyInitialForces()
  ]
  particleUpdateScript [
   &ParticleState()
   &GravityForce(Gravity=&Multiply_VectorByFloat(Vector=&RandomRangeVector(Minimum=<0 0 -2376.6189>, Maximum=<0 0 -2909.2666>), Float=$System.ScaleFX))
   &Drag(Use Linear Drag=true, Use Rotational Drag=false)
   &ScaleColor(
    Scale Alpha=&FloatFromCurve(FloatCurve=DataInterface() {propertyValues {Curve {keys [<at -0.0046 1.0113 RCIM_Cubic>,<at 1.0011 0.002 RCIM_Cubic>]}}})
    ScaleA=true
    ScaleRGB=false
   )
   &ScaleMeshSize(Scale Factor=&VectorFromFloat(Value=&FloatFromCurve(FloatCurve=DataInterface() {propertyValues {Curve {keys [
          {tangentMode=RCTM_User, time=0, value=1}
          {arriveTangent=-0.0405, interpMode=RCIM_Cubic, leaveTangent=-0.0405, tangentMode=RCTM_User, time=0.7009, value=1}
          {arriveTangent=-8.0089, interpMode=RCIM_Cubic, leaveTangent=-8.0089, tangentMode=RCTM_User, time=1, value=0}
         ]}}})))
   &Collision(
    Age Colliding Particles=false
    Control Roll On Collision=true
    Friction=1
    Friction During a Bounce=0.5
    Module.Control Rest State Via Surface Normal=false
    Module.EnableMaxCollisionCount=false
    Module.Randomize Collision Normal Vector=false
    Restitution=0.5
    ScaleRateOfRotation=false
   )
   &SolveForcesAndVelocity(Clamp Velocity=false, Limit Acceleration=false)
  ]
  rendererClasses [&NiagaraMeshRendererProperties]
  renderers [{rendererClass=&NiagaraMeshRendererProperties, rendererIndex=0}]
 }
 Smoke {
  emitterUpdateScript [
   &EmitterState(Life Cycle Mode=::NewEnumerator1, Loop Behavior=::NewEnumerator1, Loop Duration=1, UseLoopDelay=true)
   &SpawnBurst_Instantaneous(Spawn Count=&UniformRangedInt(Minimum=16, Maximum=32), Spawn Time=0.1, Use Spawn Probability=false)
  ]
  particleSpawnScript [
   /SetVariables(Particles.VelCurveRand:NiagaraFloat=&FloatFromCurve(
     FloatCurve=DataInterface() {propertyValues {Curve {keys [
         {interpMode=RCIM_Cubic, leaveTangent=0.1132, tangentMode=RCTM_Break, time=0, value=0}
         {arriveTangent=1.9858, interpMode=RCIM_Cubic, leaveTangent=1.7203, tangentMode=RCTM_Break, time=0.3, value=0.4}
         {interpMode=RCIM_Cubic, tangentMode=RCTM_Break, time=1, value=1}
        ]}}}
     CurveIndex=&RandomRangeFloat(Minimum=0, Maximum=1)
     Scale Curve=5
    )) {isSetParameters=true}
   /SetVariables(Particles.InitSpriteRotRate=&RandomRangeFloat(Minimum=-1, Maximum=1)) {isSetParameters=true}
   &InitializeParticle(
    Color Maximum=<rgba 0.25 0.2436 0.2146 0.5>
    Color Minimum=<rgba 0.3 0.2741 0.2542 0.5>
    Color Mode=::NewEnumerator2
    Lifetime Max=1
    Lifetime Min=0.5
    Lifetime Mode=::NewEnumerator1
    Sprite Rotation Angle=&RandomRangeFloat(Minimum=0, Maximum=360)
    Sprite Rotation Mode=::NewEnumerator0
    Sprite Size Mode=::NewEnumerator1
    Uniform Sprite Size Max=250
    Uniform Sprite Size Min=200
    UsePositionOffset=false
    Write Material Random=true
   )
   /SetVariables(Particles.SpriteSize=&Multiply_Vector2D_ByFloat(Float=$System.ScaleFX, Vector2D=$Particles.SpriteSize)) {isSetParameters=true}
   &CameraOffset(Camera Offset Amount=&UniformRangedFloat(Minimum=50, Maximum=60))
   &ShapeLocation(
    Handle Radius=15
    Large Radius=75
    MaskSpawnGroup=false
    Non Uniform Scale=&Multiply_VectorByFloat(Vector=<1 1 1>, Float=$System.ScaleFX)
    Override Seed=false
    Shape Primitive=::NewEnumerator3
    UseOwnerScale=true
   )
   &AddVelocityFromPoint(Module.Use Falloff Distance=false, Velocity Strength=&Lerp_Float(A=1400, B=1500, Alpha=$Particles.VelCurveRand))
  ]
  particleUpdateScript [
   &ParticleState()
   &Drag(Drag=8, Use Linear Drag=true, Use Rotational Drag=false)
   &SpriteRotationRate(Rotation Rate=&Multiply_Float(
     A=$Particles.InitSpriteRotRate
     B=&FloatFromCurve(
      FloatCurve=DataInterface() {propertyValues {Curve {keys [
          {arriveTangent=-4.4624, interpMode=RCIM_Cubic, leaveTangent=-4.4624, tangentMode=RCTM_User, time=0, value=1}
          {arriveTangentWeight=0.1718, interpMode=RCIM_Cubic, tangentMode=RCTM_User, time=0.5, value=0.15}
         ]}}}
      Scale Curve=1500
     )
    ))
   &ScaleColor(
    Scale Alpha=&RandomRangeFloat(Minimum=0.2, Maximum=0.4)
    Scale RGB=&VectorFromFloat(Value=1)
    ScaleA=true
    ScaleRGB=false
   )
   &ScaleSpriteSize(Scale Factor=&Vector2DFromFloat(Value=&FloatFromCurve(
      FloatCurve=DataInterface() {propertyValues {Curve {keys [
          {arriveTangent=0.0418, interpMode=RCIM_Cubic, leaveTangent=0.0418, tangentMode=RCTM_User, time=0, value=0.5}
          {arriveTangent=0.8771, interpMode=RCIM_Cubic, leaveTangent=0.8771, tangentMode=RCTM_User, time=0.7, value=0.775}
          {arriveTangent=0.7996, arriveTangentWeight=0.3375, interpMode=RCIM_Cubic, leaveTangent=0.7996, tangentMode=RCTM_User, time=1, value=1}
         ]}}}
      Scale Curve=0.9
     )))
   &DynamicMaterialParameters(
    Index 0 Param 1=&Add_Float(
     A=&RandomRangeFloat(Minimum=0, Maximum=1)
     B=&FloatFromCurve(
      FloatCurve=DataInterface() {propertyValues {Curve {keys [
          {arriveTangent=2.7533, interpMode=RCIM_Cubic, leaveTangent=1.8958, tangentMode=RCTM_User, time=0, value=1}
          {arriveTangent=0.8292, interpMode=RCIM_Cubic, leaveTangent=0.8292, tangentMode=RCTM_User, time=0.321, value=1.3453}
          {arriveTangent=0.0529, interpMode=RCIM_Cubic, leaveTangent=0.2021, tangentMode=RCTM_User, time=1, value=1.5}
         ]}}}
      CurveIndex=&Add_Float(A=&Multiply_Float(A=$Particles.Initial.MaterialRandom, B=0.5), B=$Particles.NormalizedAge)
     )
    )
    Index 0 Param 2=&FloatFromCurve(FloatCurve=DataInterface() {propertyValues {Curve {keys [
         {arriveTangent=0.0453, interpMode=RCIM_Cubic, tangentMode=RCTM_Break, time=0, value=-1}
         {arriveTangent=0.0604, interpMode=RCIM_Cubic, leaveTangent=0.0604, tangentMode=RCTM_User, time=1, value=-1}
        ]}}})
    Index 0 Param 4=&FloatFromCurve(FloatCurve=DataInterface() {propertyValues {Curve {keys [<at 0.5 1>,<at 1 0>]}}})
    Index 1 Param 1=&FloatFromCurve(
     FloatCurve=DataInterface() {propertyValues {Curve {keys [
         {interpMode=RCIM_Cubic, tangentMode=RCTM_Break, time=0, value=1}
         {arriveTangentWeight=0.0833, interpMode=RCIM_Cubic, tangentMode=RCTM_Break, time=0.25, value=0.3}
        ]}}}
     Scale Curve=&RandomRangeFloat(Minimum=1.5, Maximum=3)
    )
    Index 1 Param 2=&FloatFromCurve(
     FloatCurve=DataInterface() {propertyValues {Curve {keys [
         {interpMode=RCIM_Cubic, leaveTangent=-2.5489, tangentMode=RCTM_Break, time=0, value=1}
         {interpMode=RCIM_Cubic, leaveTangent=0.9066, tangentMode=RCTM_Break, time=1, value=0}
        ]}}}
     Scale Curve=&RandomRangeFloat(Minimum=0.01, Maximum=0.1)
    )
    Index 1 Param 3=&FloatFromCurve(FloatCurve=DataInterface() {propertyValues {Curve {keys [
         {interpMode=RCIM_Cubic, leaveTangent=6.3917, tangentMode=RCTM_Break, time=0, value=0.2}
         {arriveTangent=1.7549, interpMode=RCIM_Cubic, leaveTangent=10.6767, tangentMode=RCTM_User, time=0.25, value=1}
        ]}}})
    Index 1 Param 4=&FloatFromCurve(
     FloatCurve=DataInterface() {propertyValues {Curve {keys [{arriveTangent=-0.136, interpMode=RCIM_Cubic, leaveTangent=-0.136, tangentMode=RCTM_User, time=0, value=1}, <at 1 0.95>]}}}
     Scale Curve=&RandomRangeFloat(Minimum=1.3, Maximum=2)
    )
    Index 2 Param 1=&FloatFromCurve(FloatCurve=DataInterface() {propertyValues {Curve {keys [
         {interpMode=RCIM_Cubic, leaveTangent=4.2485, tangentMode=RCTM_Break, time=0, value=0}
         {arriveTangent=1.0548, interpMode=RCIM_Cubic, leaveTangent=1.0548, tangentMode=RCTM_User, time=0.17, value=0.5715}
         <at 1 1>
        ]}}})
    Index 2 Param 2=&FloatFromCurve(
     FloatCurve=DataInterface() {propertyValues {Curve {keys [
         {arriveTangent=1.9056, interpMode=RCIM_Cubic, leaveTangent=1.9056, tangentMode=RCTM_User, time=0, value=0.2}
         {arriveTangent=0.0259, interpMode=RCIM_Cubic, leaveTangent=0.0259, tangentMode=RCTM_User, time=1, value=1.4}
        ]}}}
     Scale Curve=&RandomRangeFloat(Minimum=0.1, Maximum=0.2)
    )
    Index 2 Param 3=&FloatFromCurve(FloatCurve=DataInterface() {propertyValues {Curve {keys [
         {interpMode=RCIM_Cubic, leaveTangent=2.3064, tangentMode=RCTM_Break, time=0, value=1}
         {interpMode=RCIM_Cubic, leaveTangent=11.1233, tangentMode=RCTM_Break, time=1, value=2}
        ]}}})
    Index 2 Param 4=&FloatFromCurve(
     FloatCurve=DataInterface() {propertyValues {Curve {keys [
         {arriveTangent=-0.6867, interpMode=RCIM_Cubic, leaveTangent=-0.6867, tangentMode=RCTM_User, time=0, value=1}
         {interpMode=RCIM_Cubic, leaveTangent=-1.2956, tangentMode=RCTM_Break, time=1, value=0.75}
        ]}}}
     Scale Curve=1.5
    )
    Module.Param0WriteEnabled=true
    Module.Param0WriteEnabled1=true
    Module.Param0WriteEnabled2=true
    Module.Param1WriteEnabled=false
    Module.Param1WriteEnabled1=true
    Module.Param1WriteEnabled2=true
    Module.Param2WriteEnabled=false
    Module.Param2WriteEnabled1=true
    Module.Param2WriteEnabled2=true
    Module.Param3WriteEnabled=true
    Module.Param3WriteEnabled1=true
    Module.Param3WriteEnabled2=true
    Write Parameter Index 1=true
    Write Parameter Index 2=true
   )
   &SolveForcesAndVelocity(Clamp Velocity=false, Limit Acceleration=false)
  ]
  rendererClasses [&NiagaraSpriteRendererProperties]
  renderers [
   {rendererClass=&NiagaraSpriteRendererProperties}
   {rendererClass=&NiagaraSpriteRendererProperties, rendererIndex=1}
  ]
 }
}
summary {
 systemName=NS_RocketDrill_DustPuff_OnLanding
 emitters [
  {emitterName=Combined_Emitter_Mesh, rendererClasses [&NiagaraMeshRendererProperties]}
  {emitterName=Smoke, rendererClasses [&NiagaraSpriteRendererProperties]}
  {emitterName=AirStreaks, rendererClasses [&NiagaraSpriteRendererProperties]}
  {emitterName=AirStreaks_upward, rendererClasses [&NiagaraSpriteRendererProperties]}
  {emitterName=Debris, rendererClasses [&NiagaraSpriteRendererProperties]}
  {emitterName=Debris001, rendererClasses [&NiagaraMeshRendererProperties]}
 ]
 userVariables [
  User.FX_Scale_Global {type=NiagaraFloat, defaultValue=0, description=""}
  User.FX_Scale_Impact_Center {type=NiagaraFloat, defaultValue=1.75, description=""}
 ]
}
