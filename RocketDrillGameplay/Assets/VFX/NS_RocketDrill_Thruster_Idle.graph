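# Niagara system for the rocket drill's idle thruster, layering mesh flames, a flickering light, fake-lit glow sprites, ribbon and sprite smoke, and distortion sprites.
# User.FuelAmount scales flame length, light, opacity and spawn probability for smoke and distortion.
# A system-level NormalizedVelocityRange brightens the light and glow and lengthens flames as it rises, while shortening and fading the ribbon smoke.
# Both smoke emitters inherit source movement and use curl noise.
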NiagaraSystem /RocketDrillGameplay/Assets/VFX/NS_RocketDrill_Thruster_Idle
dependencies {
 usedDataInterfaces [&NiagaraDataInterfaceCurve, &NiagaraDataInterfaceColorCurve]
 usedDynamicInputs [
  &Add_Vector
  /Niagara/DynamicInputs/Helpers/SimulationPosition
  &ClampFloat
  &RemapRange
  &VectorLength
 ]
 usedModules [
  &SystemState
  &InterpolateOverTime
  &self:SystemSpawnScript.SystemScriptSource.SystemScriptGraph.NiagaraNodeAssignment_1.SetVariables_#1
  &self:SystemSpawnScript.SystemScriptSource.SystemScriptGraph.NiagaraNodeAssignment_0.SetVariables_#2
 ]
}
emitters {
 Distort {
  emitterUpdateScript [
   &EmitterState(Life Cycle Mode=::NewEnumerator1, Loop Behavior=::NewEnumerator1, Loop Duration Mode=::NewEnumerator1, UseLoopDelay=false)
   &SpawnRate(Spawn Probability=$User.FuelAmount, SpawnRate=6, Use Spawn Probability=true)
  ]
  particleSpawnScript [
   &InitializeParticle(
    Lifetime Max=0.5
    Lifetime Min=0.4
    Lifetime Mode=::NewEnumerator1
    Mass Mode=::NewEnumerator3
    Position Offset=<5 0 0>
    Position Offset Coordinate Space=::Local
    Sprite Rotation Mode=::NewEnumerator2
    Sprite Size Mode=::NewEnumerator1
    Uniform Sprite Size Max=60
    Uniform Sprite Size Min=30
    UsePositionOffset=true
    Write Material Random=true
   )
   &AddVelocity(
    Cone Angle=5
    Cone Axis=<0 0 -1>
    Distribution Along Cone Axis=0.25
    Module.UseConeFalloff=true
    Module.UseVelDistribution=true
    Override Seed=false
    Speed Falloff From Cone Axis=0.5
    Velocity Mode=::NewEnumerator2
    Velocity Speed=&RandomRangeFloat(Minimum=5, Maximum=10)
   )
  ]
  particleUpdateScript [
   &ParticleState()
   &ScaleSpriteSize(Uniform Curve Sprite Scale=DataInterface() {propertyValues {Curve {keys [
        {arriveTangent=4.1785, interpMode=RCIM_Cubic, leaveTangent=4.1785, tangentMode=RCTM_User, time=0, value=0}
        {arriveTangent=1.1179, interpMode=RCIM_Cubic, leaveTangent=1.1179, tangentMode=RCTM_User, time=0.2042, value=0.6228}
        {arriveTangent=0.0224, interpMode=RCIM_Cubic, leaveTangent=0.0224, tangentMode=RCTM_User, time=1, value=1}
       ]}}})
   &Drag(Use Linear Drag=true, Use Rotational Drag=false)
   &ScaleColor(
    Scale Alpha=&FloatFromCurve(
     FloatCurve=DataInterface() {propertyValues {Curve {keys [<at -0.0046 1.0113 RCIM_Cubic>,<at 1.0011 0.002 RCIM_Cubic>]}}}
     Scale Curve=$User.FuelAmount
    )
    ScaleA=true
    ScaleRGB=true
   )
   &GravityForce(Gravity=&RandomRangeVector(Minimum=<0 0 200>, Maximum=<0 0 100>))
   &SolveForcesAndVelocity(Clamp Velocity=false, Limit Acceleration=false)
   &DynamicMaterialParameters(
    Index 0 Param 1=&FloatFromCurve(
     FloatCurve=DataInterface() {propertyValues {Curve {keys [
         {interpMode=RCIM_Cubic, tangentMode=RCTM_User, time=0, value=1}
         {interpMode=RCIM_Cubic, tangentMode=RCTM_User, time=0.5, value=1.01}
         {interpMode=RCIM_Cubic, tangentMode=RCTM_User, time=1, value=1}
        ]}}}
     Scale Curve=0.9978
    )
    Module.Index0Param0Write=true
    Module.Index0Param1Write=false
    Module.Index0Param2Write=false
    Module.Index0Param3Write=false
   )
  ]
  rendererClasses [&NiagaraSpriteRendererProperties]
  renderers [{rendererClass=&NiagaraSpriteRendererProperties}]
 }
 FakeLighting {
  emitterUpdateScript [
   &EmitterState(
    Inactive Response=::NewEnumerator1
    Life Cycle Mode=::NewEnumerator1
    Loop Behavior=::NewEnumerator1
    Loop Duration Mode=::NewEnumerator1
    UseLoopDelay=true
   )
   &SpawnBurst_Instantaneous(Spawn Count=2, Use Spawn Probability=false, UseLoopCountLimit=true)
  ]
  particleSpawnScript [
   &FN_SetSpawnGroup(Number of Spawn Groups=2)
   /SetVariables(Particles.VisibilityTag=$Particles.SpawnGroup) {isSetParameters=true}
   &InitializeParticle(
    Color=<rgba 25 2.0833 0 1>
    Position Offset=&MaskVectorBySpawnGroup(Masked Vector=<10 0 0>, Passthrough Vector=<17.5 0 0>)
    Sprite Size Mode=::NewEnumerator3
    Uniform Sprite Size=12.5
    UsePositionOffset=true
    Write Material Random=true
   )
   &DynamicMaterialParameters(
    Index 0 Param 1=1
    Index 0 Param 2=1
    Index 0 Param 3=1
    Index 0 Param 4=1
    Module.Param0WriteEnabled=true
    Module.Param1WriteEnabled=true
    Module.Param2WriteEnabled=true
    Module.Param3WriteEnabled=true
   )
  ]
  particleUpdateScript [
   &ParticleState(
    Kill Particles When Lifetime Has Elapsed=false
    Let Infinitely Lived Particles Die When Emitter Deactivates=true
    Loop Particles Lifetime=true
   )
   &CameraOffset()
   /SetVariables(Particles.SpriteSize_FakeLit:Vector2f=&Multiply_Vector2D_ByFloat(Float=1.75, Vector2D=$Particles.Initial.SpriteSize)) {isSetParameters=true}
   &ScaleColor(
    Scale Alpha=&FloatFromCurve(
     FloatCurve=DataInterface() {propertyValues {Curve {keys [
         <at 0 1 RCIM_Cubic>
         {arriveTangent=0.1177, interpMode=RCIM_Cubic, leaveTangent=0.1177, time=0.0746, value=0.8987}
         {arriveTangent=-0.0227, interpMode=RCIM_Cubic, leaveTangent=-0.0227, tangentMode=RCTM_User, time=0.1682, value=1.0198}
         {arriveTangent=-0.7897, interpMode=RCIM_Cubic, leaveTangent=-0.7897, time=0.2019, value=0.7756}
         {arriveTangent=0.8345, interpMode=RCIM_Cubic, leaveTangent=0.8345, time=0.2748, value=0.9356}
         {arriveTangent=0.5127, interpMode=RCIM_Cubic, leaveTangent=0.5127, time=0.3464, value=0.8961}
         {arriveTangent=-0.6889, interpMode=RCIM_Cubic, leaveTangent=-0.6889, time=0.4074, value=1.0036}
         {arriveTangent=-0.8737, interpMode=RCIM_Cubic, leaveTangent=-0.8737, time=0.4365, value=0.834}
         {arriveTangent=-0.4832, interpMode=RCIM_Cubic, leaveTangent=-0.4832, time=0.4871, value=0.9339}
         {arriveTangent=0.1715, interpMode=RCIM_Cubic, leaveTangent=0.1715, time=0.5331, value=0.7873}
         {arriveTangent=0.795, interpMode=RCIM_Cubic, leaveTangent=0.795, time=0.6112, value=0.9552}
         {arriveTangent=0.9015, interpMode=RCIM_Cubic, leaveTangent=0.9015, time=0.69, value=0.912}
         {arriveTangent=-0.4254, interpMode=RCIM_Cubic, leaveTangent=-0.4254, time=0.7536, value=1.0835}
         {arriveTangent=-1.4701, interpMode=RCIM_Cubic, leaveTangent=-1.4701, time=0.7995, value=0.8654}
         {arriveTangent=-0.4912, interpMode=RCIM_Cubic, leaveTangent=-0.4912, time=0.8674, value=0.9161}
         {arriveTangent=0.6326, interpMode=RCIM_Cubic, leaveTangent=0.6326, time=0.9138, value=0.8093}
         {interpMode=RCIM_Cubic, leaveTangent=0.1355, time=1, value=1}
        ]}}}
     Scale Curve=$User.FuelAmount
    )
    Scale RGB=&VectorFromFloat(Value=&FloatFromCurve(
      FloatCurve=DataInterface() {propertyValues {Curve {keys [
          <at 0 1 RCIM_Cubic>
          {arriveTangent=0.1177, interpMode=RCIM_Cubic, leaveTangent=0.1177, time=0.0685, value=0.5809}
          {arriveTangent=-0.0227, interpMode=RCIM_Cubic, leaveTangent=-0.0227, tangentMode=RCTM_User, time=0.1682, value=1.0198}
          {arriveTangent=-0.7897, interpMode=RCIM_Cubic, leaveTangent=-0.7897, time=0.1958, value=0.4578}
          {arriveTangent=0.8345, interpMode=RCIM_Cubic, leaveTangent=0.8345, time=0.2748, value=0.9356}
          {arriveTangent=0.5127, interpMode=RCIM_Cubic, leaveTangent=0.5127, time=0.3403, value=0.5784}
          {arriveTangent=-0.6889, interpMode=RCIM_Cubic, leaveTangent=-0.6889, time=0.4074, value=1.0036}
          {arriveTangent=-3.5031, interpMode=RCIM_Cubic, leaveTangent=-3.5031, time=0.4304, value=0.5163}
          {arriveTangent=-0.4832, interpMode=RCIM_Cubic, leaveTangent=-0.4832, time=0.4871, value=0.7243}
          {arriveTangent=1.8607, interpMode=RCIM_Cubic, leaveTangent=1.8607, time=0.527, value=0.4696}
          {arriveTangent=0.795, interpMode=RCIM_Cubic, leaveTangent=0.795, time=0.6112, value=0.9552}
          {arriveTangent=0.9015, interpMode=RCIM_Cubic, leaveTangent=0.9015, time=0.6839, value=0.5943}
          {arriveTangent=-0.4254, interpMode=RCIM_Cubic, leaveTangent=-0.4254, time=0.7536, value=1.0835}
          {arriveTangent=-3.3116, interpMode=RCIM_Cubic, leaveTangent=-3.3116, time=0.7934, value=0.5477}
          {arriveTangent=-0.4912, interpMode=RCIM_Cubic, leaveTangent=-0.4912, time=0.8674, value=0.7065}
          {arriveTangent=2.2144, interpMode=RCIM_Cubic, leaveTangent=2.2144, time=0.9076, value=0.4916}
          {interpMode=RCIM_Cubic, leaveTangent=0.1355, time=1, value=1}
         ]}}}
      Scale Curve=&FloatFromCurve(
       FloatCurve=DataInterface() {propertyValues {Curve {keys [
           {arriveTangent=0.0801, interpMode=RCIM_Cubic, leaveTangent=0.0801, tangentMode=RCTM_User, time=0, value=1}
           {arriveTangent=0.097, interpMode=RCIM_Cubic, leaveTangent=0.097, tangentMode=RCTM_User, time=0.7, value=1}
           {arriveTangent=4.5194, interpMode=RCIM_Cubic, leaveTangent=4.5194, tangentMode=RCTM_User, time=1, value=1.5}
          ]}}}
       CurveIndex=$System.NormalizedVelocityRange
       Scale Curve=10
      )
     ))
    ScaleA=true
    ScaleRGB=true
   )
   /SetVariables(Particles.Color_FakeLit:LinearColor=&MultiplyLinearColor(
     Linear Color=$Particles.Color
     Scale Factor=&MakeFloat4FromVectorAndFloat(
      "Vector (RGB)"=&VectorFromFloat(Value=&FloatFromCurve(
        FloatCurve=DataInterface() {propertyValues {Curve {keys [
            {arriveTangent=0.0801, interpMode=RCIM_Cubic, leaveTangent=0.0801, tangentMode=RCTM_User, time=0, value=0.05}
            {arriveTangent=0.6029, interpMode=RCIM_Cubic, leaveTangent=0.6029, tangentMode=RCTM_User, time=0.7, value=0.25}
            {arriveTangent=6.6165, interpMode=RCIM_Cubic, leaveTangent=6.6165, tangentMode=RCTM_User, time=1, value=1}
           ]}}}
        CurveIndex=$System.NormalizedVelocityRange
        Scale Curve=0.1
       ))
      "Float (Alpha)"=$User.FuelAmount
     )
    )) {isSetParameters=true}
  ]
  rendererClasses [&NiagaraSpriteRendererProperties]
  renderers [
   {rendererClass=&NiagaraSpriteRendererProperties}
   {rendererClass=&NiagaraSpriteRendererProperties, rendererIndex=1}
  ]
 }
 FireMesh {
  emitterUpdateScript [
   &EmitterState(
    Inactive Response=::NewEnumerator1
    Life Cycle Mode=::NewEnumerator1
    Loop Behavior=::NewEnumerator1
    Loop Duration Mode=::NewEnumerator1
    UseLoopDelay=false
   )
   &SpawnRate(SpawnRate=20, Use Spawn Probability=false)
  ]
  particleSpawnScript [
   &InitializeParticle(
    Lifetime Max=0.27
    Lifetime Min=0.14
    Lifetime Mode=::NewEnumerator1
    Mesh Scale=<0.0425 0.0425 0.4>
    Mesh Scale Mode=::NewEnumerator4
    Position Offset=<12.5 0 0>
    UsePositionOffset=true
    Write Material Random=true
   )
   &InitialMeshOrientation(
    Mesh Axis To Orient=<0 0 1>
    Mesh Orientation Mode=::NewEnumerator2
    Rotation=&RandomRangeVector(Minimum=<0 0 0>, Maximum=<0 0 1>)
    Use Rotation Vector=true
   )
  ]
  particleUpdateScript [
   &ParticleState()
   &ScaleMeshSize(Scale Factor=&MakeVector(
     X=&FloatFromCurve(
      FloatCurve=DataInterface() {propertyValues {Curve {keys [{tangentMode=RCTM_User, time=0, value=0.75}, {tangentMode=RCTM_User, time=1, value=1.2}]}}}
      Scale Curve=1.5
     )
     Y=&FloatFromCurve(
      FloatCurve=DataInterface() {propertyValues {Curve {keys [{tangentMode=RCTM_User, time=0, value=0.75}, {tangentMode=RCTM_User, time=1, value=1.2}]}}}
      Scale Curve=1.5
     )
     Z=&Multiply_Float(
      A=&FloatFromCurve(
       FloatCurve=DataInterface() {propertyValues {Curve {keys [<at 0 0>,<at 1 1>]}}}
       Scale Curve=&Lerp_Float(
        A=2
        B=3.5
        Alpha=&FloatFromCurve(
         FloatCurve=DataInterface() {propertyValues {Curve {keys [
             {arriveTangent=0.0801, interpMode=RCIM_Cubic, leaveTangent=0.0801, tangentMode=RCTM_User, time=0, value=0.1}
             {arriveTangent=0.097, interpMode=RCIM_Cubic, leaveTangent=0.097, tangentMode=RCTM_User, time=0.7, value=0.1}
             {arriveTangent=12.9805, interpMode=RCIM_Cubic, leaveTangent=12.9805, tangentMode=RCTM_User, time=1, value=1.5}
            ]}}}
         CurveIndex=$System.NormalizedVelocityRange
        )
       )
      )
      B=$User.FuelAmount
     )
    ))
   &ScaleColor(
    Scale Alpha=&FloatFromCurve(
     FloatCurve=DataInterface() {propertyValues {Curve {keys [
         {arriveTangent=9.0725, interpMode=RCIM_Cubic, leaveTangent=9.0725, tangentMode=RCTM_User, time=0, value=0}
         {arriveTangent=-1.6443, interpMode=RCIM_Cubic, leaveTangent=-1.6443, tangentMode=RCTM_User, time=0.1408, value=0.9847}
         {arriveTangent=-0.0405, interpMode=RCIM_Cubic, leaveTangent=-0.0405, tangentMode=RCTM_User, time=1, value=0}
        ]}}}
     Scale Curve=1.5
    )
    Scale RGB=&MakeVectorFromLinearColorRGB(Color=&ColorFromCurve(ColorCurve=DataInterface() {propertyValues {
        AlphaCurve {keys [<at 0 1 RCIM_Cubic>]}
        BlueCurve {keys [<at 0.0203 0>, <at 0.0654 0.0499>, {arriveTangent=-15.6046, time=0.1787, value=0}, <at 0.6131 0>, <at 0.984 0>]}
        GreenCurve {keys [<at 0.0203 0.8333>, <at 0.0654 0.2874>, {arriveTangent=-5.6465, time=0.1787, value=0.1667}, <at 0.6131 0.0833>, <at 0.984 0.0167>]}
        RedCurve {keys [<at 0.0203 2>, <at 0.0654 1>, {arriveTangent=-0.716, time=0.1787, value=1}, <at 0.6131 1>, <at 0.984 1>]}
       }}))
    ScaleA=true
    ScaleRGB=true
   )
  ]
  rendererClasses [&NiagaraMeshRendererProperties]
  renderers [{rendererClass=&NiagaraMeshRendererProperties, rendererIndex=0}]
 }
 Light {
  emitterUpdateScript [
   &EmitterState(
    Inactive Response=::NewEnumerator1
    Life Cycle Mode=::NewEnumerator1
    Loop Behavior=::NewEnumerator1
    Loop Duration Mode=::NewEnumerator1
    UseLoopDelay=false
   )
   &SpawnBurst_Instantaneous(Spawn Count=1, Use Spawn Probability=false)
  ]
  particleSpawnScript [&InitializeParticle(Color=<rgba 1 0.1667 0 1>, Position Offset=<10 0 0>, UsePositionOffset=true, Write Material Random=true) {module=InitializeParticle001}]
  particleUpdateScript [
   &ParticleState(Kill Particles When Lifetime Has Elapsed=false, Let Infinitely Lived Particles Die When Emitter Deactivates=true)
   &FN_Light_Attributes(
    Color=&MultiplyLinearColor(
     Linear Color=$Particles.Color
     Scale Factor=&MakeVector4_FromFloat(Float=&RandomRangeFloat(Minimum=10, Maximum=25, Evaluation Type=::NewEnumerator1))
    )
    Create Color Attribute=true
    Create Exponent Attribute=true
    Create Radius Attribute=true
    Exponent=3
    Radius=7
    Scale Color=&Multiply_Float(
     A=$User.FuelAmount
     B=&FloatFromCurve(
      FloatCurve=DataInterface() {propertyValues {Curve {keys [
          {arriveTangent=0.0801, interpMode=RCIM_Cubic, leaveTangent=0.0801, tangentMode=RCTM_User, time=0, value=0.25}
          {arriveTangent=0.097, interpMode=RCIM_Cubic, leaveTangent=0.097, tangentMode=RCTM_User, time=0.7, value=0.25}
          {arriveTangent=6.2083, interpMode=RCIM_Cubic, leaveTangent=6.2083, tangentMode=RCTM_User, time=1, value=1}
         ]}}}
      CurveIndex=$System.NormalizedVelocityRange
      Scale Curve=20
     )
    )
   )
  ]
  rendererClasses [&NiagaraLightRendererProperties]
  renderers [{rendererClass=&NiagaraLightRendererProperties, rendererIndex=0}]
 }
 Reload_Smoke {
  emitterUpdateScript [
   &EmitterState(Life Cycle Mode=::NewEnumerator1, Loop Behavior=::NewEnumerator1, Loop Duration Mode=::NewEnumerator1, UseLoopDelay=false)
   &SpawnRate(Spawn Probability=$User.FuelAmount, SpawnRate=128, Use Spawn Probability=true)
  ]
  particleSpawnScript [
   &InitializeParticle(
    Color=<rgba 0.0417 0.0378 0.0373 1>
    Lifetime=&FloatFromCurve(
     FloatCurve=DataInterface() {propertyValues {Curve {keys [
         {arriveTangent=-1.7792, interpMode=RCIM_Cubic, leaveTangent=-1.7792, tangentMode=RCTM_User, time=0, value=0.6}
         {arriveTangent=-0.0413, interpMode=RCIM_Cubic, leaveTangent=-0.0413, tangentMode=RCTM_User, time=1, value=0}
        ]}}}
     CurveIndex=$System.NormalizedVelocityRange
     Scale Curve=0.5
    )
    Position Offset=<10 0 0>
    Position Offset Coordinate Space=::Local
    Ribbon Width=25
    Ribbon Width Mode=::NewEnumerator1
    UsePositionOffset=true
    Write Material Random=true
   )
   &AddVelocity(Velocity=&MakeVector(X=-200, Y=75, Z=75))
  ]
  particleUpdateScript [
   &ParticleState()
   &ScaleColor(
    Scale Alpha=&Multiply_Float(
     A=&FloatFromCurve(FloatCurve=DataInterface() {propertyValues {Curve {keys [
          {arriveTangent=9.451, interpMode=RCIM_Cubic, leaveTangent=9.451, tangentMode=RCTM_User, time=0, value=0}
          {arriveTangent=3.9066, interpMode=RCIM_Cubic, leaveTangent=3.9066, time=0.08, value=0.5744}
          {arriveTangent=-1.5974, interpMode=RCIM_Cubic, leaveTangent=-1.5974, tangentMode=RCTM_User, time=0.145, value=0.5664}
          {interpMode=RCIM_Cubic, tangentMode=RCTM_User, time=1, value=0}
         ]}}})
     B=&FloatFromCurve(
      FloatCurve=DataInterface() {propertyValues {Curve {keys [
          {arriveTangent=-0.1327, interpMode=RCIM_Cubic, leaveTangent=-0.1327, tangentMode=RCTM_User, time=0, value=1}
          {arriveTangent=-0.0413, interpMode=RCIM_Cubic, leaveTangent=-0.0413, tangentMode=RCTM_User, time=1, value=0}
         ]}}}
      CurveIndex=$System.NormalizedVelocityRange
      Scale Curve=$User.FuelAmount
     )
    )
    ScaleA=true
    ScaleRGB=true
   )
   &DynamicMaterialParameters(
    Index 0 Param 1=&FloatFromCurve(FloatCurve=DataInterface() {propertyValues {Curve {keys [
         <at 0 1 RCIM_Cubic>
         {arriveTangent=-1.7387, arriveTangentWeight=1.9926, interpMode=RCIM_Cubic, leaveTangent=-1.7387, tangentMode=RCTM_User, time=1, value=0.125}
        ]}}})
    Index 0 Param 3=-1
    Index 1 Param 1=&FloatFromCurve(FloatCurve=DataInterface() {propertyValues {Curve {keys [<at 0 1 RCIM_Cubic>, {arriveTangent=45.3807, interpMode=RCIM_Cubic, leaveTangent=45.3807, tangentMode=RCTM_User, time=0.5, value=10}]}}})
    Index 1 Param 2=&FloatFromCurve(FloatCurve=DataInterface() {propertyValues {Curve {keys [{arriveTangent=-2.2405, interpMode=RCIM_Cubic, leaveTangent=-2.2405, tangentMode=RCTM_User, time=0, value=1}, <at 1 0 RCIM_Cubic>]}}})
    Module.Param0WriteEnabled=true
    Module.Param0WriteEnabled1=true
    Module.Param1WriteEnabled=false
    Module.Param1WriteEnabled1=true
    Module.Param2WriteEnabled=true
    Module.Param2WriteEnabled1=false
    Module.Param3WriteEnabled=true
    Module.Param3WriteEnabled1=false
    Write Parameter Index 1=true
   )
   &CurlNoiseForce(Mask Curl Noise=true, Noise Frequency=5, Noise Strength=500, Pan Noise=true, Pan Noise Field=<0 0 -10>)
   &AerodynamicDrag(
    Aerodynamic Drag:NiagaraFloat=&FloatFromCurve(
     FloatCurve=DataInterface() {propertyValues {Curve {keys [
         {interpMode=RCIM_Cubic, tangentMode=RCTM_User, time=0, value=0}
         {arriveTangent=0.7388, interpMode=RCIM_Cubic, leaveTangent=0.7388, tangentMode=RCTM_User, time=1, value=1}
        ]}}}
     Scale Curve=0.3
    )
    Aerodynamic Rotational Drag=0.1
    Lift Contribution=2
    Use Linear Drag=true
    Use Rotational Drag=true
   )
   &GravityForce(Gravity=<0 0 980>)
   &InheritSourceMovement(
    Applied Position Falloff Distance=500
    Applied Position Scale=&VectorFromFloat(Value=&FloatFromCurve(FloatCurve=DataInterface() {propertyValues {Curve {keys [
          {arriveTangent=-3.6856, interpMode=RCIM_Cubic, leaveTangent=-3.6856, tangentMode=RCTM_User, time=0, value=1}
          {interpMode=RCIM_Cubic, tangentMode=RCTM_User, time=1, value=0}
         ]}}}))
    ApplyForce=false
    ApplyPosition=true
    ApplyPositionFalloff=true
    ApplyPositionFalloffCurve=true
   )
   &SolveForcesAndVelocity(Clamp Velocity=false, Limit Acceleration=false)
   &AlignSpriteToMeshOrientation(SetAlignment=true, SetFacing=true)
  ]
  rendererClasses [&NiagaraRibbonRendererProperties]
  renderers [{rendererClass=&NiagaraRibbonRendererProperties}]
 }
 Smoke_Wide_Large {
  emitterUpdateScript [
   &EmitterState(Delay First Loop Only=true, Life Cycle Mode=::NewEnumerator1, Loop Delay=0.25, UseLoopDelay=true)
   &SpawnRate(
    Spawn Probability=$User.FuelAmount
    SpawnRate=&Lerp_Float(
     A=2.5
     B=7.5
     Alpha=&FloatFromCurve(
      FloatCurve=DataInterface() {propertyValues {Curve {keys [
          {interpMode=RCIM_Cubic, tangentMode=RCTM_Break, time=0, value=0}
          {arriveTangent=1.4927, interpMode=RCIM_Cubic, tangentMode=RCTM_Break, time=1, value=1}
         ]}}}
      CurveIndex=$User.FuelAmount
      Scale Curve=0.25
     )
    )
    Use Spawn Probability=true
   )
  ]
  particleSpawnScript [
   /SetVariables(Particles.RandCurve:NiagaraFloat=&FloatFromCurve(
     FloatCurve=DataInterface() {propertyValues {Curve {keys [
         {interpMode=RCIM_Cubic, tangentMode=RCTM_User, time=0, value=0}
         {arriveTangent=0.7625, interpMode=RCIM_Cubic, leaveTangent=0.7625, tangentMode=RCTM_User, time=0.8, value=0.2}
         {arriveTangent=7.0993, interpMode=RCIM_Cubic, leaveTangent=7.0993, tangentMode=RCTM_User, time=1, value=1}
        ]}}}
     CurveIndex=&RandomRangeFloat(Minimum=0, Maximum=1)
    )) {isSetParameters=true}
   &InitializeParticle(
    Color=&LerpLinearColors(
     StartColor=<rgba 0.0417 0.0348 0.0339 0.075>
     EndColor=<rgba 0 0 0 0.075>
     LerpFactor=&FloatFromCurve(
      FloatCurve=DataInterface() {propertyValues {Curve {keys [
          {interpMode=RCIM_Cubic, tangentMode=RCTM_User, time=0, value=0}
          {arriveTangent=0.7625, interpMode=RCIM_Cubic, leaveTangent=0.7625, tangentMode=RCTM_User, time=0.9, value=0.2}
          {arriveTangent=12.0097, interpMode=RCIM_Cubic, leaveTangent=12.0097, tangentMode=RCTM_User, time=1, value=1}
         ]}}}
      CurveIndex=&RandomRangeFloat(Minimum=0, Maximum=1)
     )
    )
    Lifetime Max=1.5
    Lifetime Mode=::NewEnumerator1
    Position Offset=<5 0 0>
    Position Offset Coordinate Space=::Local
    Sprite Rotation Mode=::NewEnumerator2
    Sprite Size Mode=::NewEnumerator3
    Uniform Sprite Size=&MakeFloatFromVector2d(Vector 2D=&Multiply_Vector2D_ByFloat(Float=&Lerp_Float(A=1, B=2, Alpha=$Particles.RandCurve), Vector2D=<30 30>))
    UsePositionOffset=true
    Write Material Random=true
   )
   /SetVariables(Particles.InitSpriteRotRate=&UniformRangedFloat(Minimum=-3, Maximum=3)) {isSetParameters=true}
   &AddVelocityInCone(
    Cone Angle=4
    Cone Axis=<1 0 -0.25>
    Module.Use Velocity Falloff On Cone Axis=false
    Velocity Strength=&Multiply_Float(A=&RandomRangeFloat(Minimum=-75, Maximum=-100), B=1)
   )
   &CameraOffset(Camera Offset Amount=5)
  ]
  particleUpdateScript [
   &ParticleState()
   &SubUVAnimation(SubUV Lookup Index:NiagaraFloat=&Multiply_Float(A=$Particles.NormalizedAge, B=0.25))
   &SpriteRotationRate(Rotation Rate=&Multiply_Float(
     A=$Particles.InitSpriteRotRate
     B=&FloatFromCurve(
      FloatCurve=DataInterface() {propertyValues {Curve {keys [
          {arriveTangent=-1.9214, interpMode=RCIM_Cubic, leaveTangent=-1.9214, tangentMode=RCTM_User, time=-0.0046, value=1.0113}
          {arriveTangent=-0.2496, interpMode=RCIM_Cubic, leaveTangent=-0.2496, tangentMode=RCTM_User, time=0.9896, value=0.16}
         ]}}}
      CurveIndex=$Emitter.Age
      Scale Curve=150
     )
    ))
   &ScaleColor(
    Scale Alpha=&FloatFromCurve(
     FloatCurve=DataInterface() {propertyValues {Curve {keys [
         {interpMode=RCIM_Cubic, tangentMode=RCTM_User, time=0, value=0}
         {interpMode=RCIM_Cubic, tangentMode=RCTM_User, time=0.4, value=1}
         {interpMode=RCIM_Cubic, tangentMode=RCTM_User, time=0.7, value=1}
         {interpMode=RCIM_Cubic, tangentMode=RCTM_User, time=1, value=0}
        ]}}}
     Scale Curve=&Lerp_Float(
      A=1
      B=1.5
      Alpha=&FloatFromCurve(
       FloatCurve=DataInterface() {propertyValues {Curve {keys [
           {interpMode=RCIM_Cubic, tangentMode=RCTM_User, time=0, value=0}
           {arriveTangent=0.7625, interpMode=RCIM_Cubic, leaveTangent=0.7625, tangentMode=RCTM_User, time=0.9, value=0.2}
           {arriveTangent=12.0097, interpMode=RCIM_Cubic, leaveTangent=12.0097, tangentMode=RCTM_User, time=1, value=1}
          ]}}}
       CurveIndex=&RandomRangeFloat(Minimum=0, Maximum=1)
       Scale Curve=$User.FuelAmount
      )
     )
    )
    ScaleA=true
    ScaleRGB=true
   )
   &Drag(
    Drag=&FloatFromCurve(FloatCurve=DataInterface() {propertyValues {Curve {keys [<at 0 1>,<at 1 0.5>]}}}, Scale Curve=7.5)
    Use Linear Drag=true
    Use Rotational Drag=false
   )
   &GravityForce(Gravity=&Lerp_Vector(A=<0 0 250>, B=<0 0 50>, Alpha=$Particles.MaterialRandom))
   /Niagara/Modules/Update/Forces/CurlNoiseForce(Noise Frequency=0.01, Noise Strength=25, Pan Noise=true, Pan Noise Field=<0 0 1>)
   &InheritSourceMovement(Applied Position Scale=<0.9 0.9 0.9>, ApplyForce=false, ApplyPosition=true, ApplyPositionFalloff=false)
   &SolveForcesAndVelocity(Clamp Velocity=false, Limit Acceleration=false)
   &ScaleSpriteSize(Scale Factor=&Vector2DFromFloat(Value=&FloatFromCurve(FloatCurve=DataInterface() {propertyValues {Curve {keys [
          {arriveTangent=7.4766, interpMode=RCIM_Cubic, leaveTangent=5.0379, tangentMode=RCTM_User, time=0, value=0.3}
          <at 0.2 0.7>
          <at 1 1 RCIM_Cubic>
         ]}}})))
   &DynamicMaterialParameters(
    Index 0 Param 1=2
    Module.Index0Param0Write=true
    Module.Index0Param1Write=true
    Module.Index0Param2Write=true
    Module.Index0Param3Write=true
   )
  ]
  rendererClasses [&NiagaraSpriteRendererProperties]
  renderers [{rendererClass=&NiagaraSpriteRendererProperties}]
 }
}
summary {
 systemName=NS_RocketDrill_Thruster_Idle
 emitters [
  {emitterName=FireMesh, rendererClasses [&NiagaraMeshRendererProperties]}
  {emitterName=Light, rendererClasses [&NiagaraLightRendererProperties]}
  {emitterName=FakeLighting, rendererClasses [&NiagaraSpriteRendererProperties]}
  {emitterName=Smoke_Wide_Large, rendererClasses [&NiagaraSpriteRendererProperties]}
  {emitterName=Reload_Smoke, rendererClasses [&NiagaraRibbonRendererProperties]}
  {emitterName=Distort, rendererClasses [&NiagaraSpriteRendererProperties]}
 ]
 userVariables [User.FuelAmount {type=NiagaraFloat, defaultValue=1, description=""}]
}
